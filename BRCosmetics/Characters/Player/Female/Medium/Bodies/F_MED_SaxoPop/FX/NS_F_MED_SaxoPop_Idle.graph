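# Niagara idle effect for the SaxoPop female character body cosmetic, adding twinkling flare sprites tinted by time of day.
# The enabled Blings_Low001 emitter places particles on cylinders around nine bones or sockets, scaling flare color by normal-to-camera dot product.
# Disabled GPU emitters instead sample skeletal-mesh UVs against mask textures to cull particles, with channels chosen by User.Variant.
# Hand and foot hidden flags kill particles in those regions.
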NiagaraSystem /BRCosmetics/Characters/Player/Female/Medium/Bodies/F_MED_SaxoPop/FX/NS_F_MED_SaxoPop_Idle
dependencies {
 usedDataInterfaces [
  /Script/Niagara.NiagaraDataInterfaceTexture
  &NiagaraDataInterfaceCurve
  &NiagaraDataInterfaceArrayFloat
  &NiagaraDataInterfaceArrayFloat3
  &NiagaraDataInterfaceCamera
  &NiagaraDataInterfaceSkeletalMesh
  &NiagaraDataInterfaceArrayColor
  &NiagaraDataInterfaceColorCurve
 ]
 usedDynamicInputs [&MakeVector2D, &SelectLinearColorFromArray]
 usedModules [
  &self:SystemSpawnScript.SystemScriptSource.SystemScriptGraph.NiagaraNodeAssignment_0.SetVariables_#1
  &self:SystemSpawnScript.SystemScriptSource.SystemScriptGraph.NiagaraNodeAssignment_1.SetVariables_#2
  &self:SystemSpawnScript.SystemScriptSource.SystemScriptGraph.NiagaraNodeAssignment_3.SetVariables_#3
  &SystemState
  /BRCosmetics/Characters/Player/Female/Medium/Bodies/F_MED_PiperShelf_Pearl/FX/TimeOfDayGlobal
 ]
}
emitters {
 Blings {
  bEnabled=false
  simTarget=GPUComputeSim
  emitterUpdateScript [
   &EmitterState(Inactive Response=::NewEnumerator1, Life Cycle Mode=::NewEnumerator1, Loop Duration=15, UseLoopDelay=false)
   &SpawnRate(Spawn Probability=0.75, SpawnRate=30, Use Spawn Probability=true)
  ]
  particleSpawnScript [
   &InitializeParticle(
    Color=<rgba 5 5 5 1>
    Lifetime Max=0.7
    Lifetime Min=0.5
    Lifetime Mode=::NewEnumerator1
    Sprite Size Mode=::NewEnumerator3
    Uniform Sprite Size=&RandomRangeFloat(Minimum=15, Maximum=25)
    UsePositionOffset=false
    Write Material Random=true
   )
   &CameraOffset(Camera Offset Amount=5)
  ]
  particleUpdateScript [
   &ParticleState()
   &SkeletalMeshLocation(
    MaskSpawnGroup=false
    Mesh Sampling Type=::NewEnumerator7
    OffsetAlongNormal=false
    Orientation Sampling=::NewEnumerator1
    OverrideRandomSeed=false
    Skeletal Mesh=$System.Skeletal Mesh
   )
   /Niagara/Modules/Update/Textures/SampleTexture(
    Sampled RGBA Mask:Vector4f=<1 1 1 1>
    Texture=DataInterface() {propertyValues {
      Texture=&/T_SaxoPop_Idle_FX
      TextureUserParameter {parameter {name=None, typeDefHandle {registeredTypeIndex=2}, varData []}}
     }}
    UV:Vector2f=$Particles.SkeletalMeshLocation.SampledUV
   ) {module=SampleTexture001}
   &SampleTexture(
    Sampled RGBA Mask:Vector4f=<1 1 1 1>
    Texture=DataInterface() {propertyValues {
      Texture=&/SaxoPop_Body_FX_2
      TextureUserParameter {parameter {name=None, typeDefHandle {registeredTypeIndex=2}, varData []}}
     }}
    UV:Vector2f=$Particles.SkeletalMeshLocation.SampledUV
   )
   &KillParticles(Kill Particles=&SetBoolByFloatComparison(
     A=&MaskFloat(
      Mask Source=1
      Mask Parameter=&MakeIntFromFloat(Float=$User.Variant)
      Mask=0
      Masked Float=&MakeFloatFromLinearColor(LinearColor=$Particles.SampleTexture001.SampledColor, Channel:ENiagaraFloatFromLinearColorOptions=::NewEnumerator0)
      Passthrough Float=&MaskFloat(
       Mask Source=1
       Mask Parameter=&MakeIntFromFloat(Float=$User.Variant)
       Mask=1
       Masked Float=&MakeFloatFromLinearColor(LinearColor=$Particles.SampleTexture001.SampledColor, Channel:ENiagaraFloatFromLinearColorOptions=::NewEnumerator1)
       Passthrough Float=&MaskFloat(
        Mask Source=1
        Mask Parameter=&MakeIntFromFloat(Float=$User.Variant)
        Mask=2
        Masked Float=&MakeFloatFromLinearColor(LinearColor=$Particles.SampleTexture001.SampledColor, Channel:ENiagaraFloatFromLinearColorOptions=::NewEnumerator2)
        Passthrough Float=&MaskFloat(
         Mask Source=1
         Mask Parameter=&MakeIntFromFloat(Float=$User.Variant)
         Mask=3
         Masked Float=&MakeFloatFromLinearColor(LinearColor=$Particles.SampleTexture001.SampledColor, Channel:ENiagaraFloatFromLinearColorOptions=::NewEnumerator3)
         Passthrough Float=&MaskFloat(
          Mask Source=1
          Mask Parameter=&MakeIntFromFloat(Float=$User.Variant)
          Mask=4
          Masked Float=&MakeFloatFromLinearColor(LinearColor=$Particles.SampleTexture.SampledColor, Channel:ENiagaraFloatFromLinearColorOptions=::NewEnumerator0)
          Passthrough Float=&MaskFloat(
           Mask Source=1
           Mask Parameter=&MakeIntFromFloat(Float=$User.Variant)
           Mask=5
           Masked Float=&MakeFloatFromLinearColor(LinearColor=$Particles.SampleTexture.SampledColor, Channel:ENiagaraFloatFromLinearColorOptions=::NewEnumerator1)
           Passthrough Float=1
          )
         )
        )
       )
      )
     )
     B=0.1
     Comparison Type=::NewEnumerator6
    )) {module=KillParticles001}
   &KillParticles(Kill Particles=/Niagara/DynamicInputs/Helpers/MaskBool(
     Mask Source:NiagaraInt32=1
     Mask Parameter:NiagaraInt32=&MakeCustomIntFromBool(Bool=$User.bisFootHidden, True INT=0, False INT=1)
     Mask:NiagaraInt32=0
     Masked Bool:NiagaraBool=&SetBoolByFloatComparison(
      A=&MakeFloatFromLinearColor(LinearColor=$Particles.SampleTexture.SampledColor, Channel:ENiagaraFloatFromLinearColorOptions=::NewEnumerator2)
      B=0.9
      Comparison Type=::NewEnumerator6
     )
     Passthrough Bool:NiagaraBool=false
    ))
   &KillParticles(Kill Particles=&MaskBool(
     Mask Source:NiagaraInt32=1
     Mask Parameter:NiagaraInt32=&MakeCustomIntFromBool(Bool=$User.bisHandHidden, True INT=0, False INT=1)
     Mask:NiagaraInt32=0
     Masked Bool:NiagaraBool=&SetBoolByFloatComparison(
      A=&MakeFloatFromLinearColor(LinearColor=$Particles.SampleTexture.SampledColor, Channel:ENiagaraFloatFromLinearColorOptions=::NewEnumerator3)
      B=0.9
      Comparison Type=::NewEnumerator6
     )
     Passthrough Bool:NiagaraBool=false
    )) {module=KillParticles002}
   &ScaleSpriteSize(Uniform Curve Scale=&FNTimeOfDayScalar(Night Scale=1.25, Morning Scale=1, Evening Scale=1)) {module=ScaleSpriteSize001}
   &ScaleColor(
    Scale Alpha=&Multiply_Float(
     A=&FloatFromCurve(
      FloatCurve=DataInterface() {propertyValues {Curve {keys [
          {interpMode=RCIM_Cubic, tangentMode=RCTM_User, time=0, value=0}
          {arriveTangent=-2.4058, interpMode=RCIM_Cubic, leaveTangent=-2.4058, tangentMode=RCTM_User, time=0.26, value=0.8421}
          {arriveTangent=-1.1775, interpMode=RCIM_Cubic, leaveTangent=-1.1775, time=0.4613, value=0.0645}
          {arriveTangent=-0.1197, interpMode=RCIM_Cubic, leaveTangent=-0.1197, time=0.5898, value=0.4537}
          <at 1 0 RCIM_Cubic>
         ]}}}
      Scale Curve=&RandomRangeFloat(Minimum=0.1, Maximum=0.5)
     )
     B=&FloatFromCurve(
      FloatCurve=DataInterface() {propertyValues {Curve {keys [{tangentMode=RCTM_User, time=0, value=0}, {tangentMode=RCTM_User, time=1, value=1}]}}}
      Scale Curve=&FloatFromCurve(
       FloatCurve=DataInterface() {propertyValues {Curve {keys [
           {tangentMode=RCTM_User, time=0, value=0}
           {arriveTangentWeight=0.1667, interpMode=RCIM_Cubic, leaveTangentWeight=0.1667, tangentMode=RCTM_User, time=1, value=1}
          ]}}}
       CurveIndex=$Emitter.Age
      )
     )
    )
    Scale RGB=&VectorFromFloat(Value=&RandomRangeFloat(Minimum=0.7, Maximum=0.9, Evaluation Type=::NewEnumerator1))
    ScaleA=true
    ScaleRGB=true
   ) {module=ScaleColor001}
   &ScaleColor(
    Scale Alpha=$System.TODAlpha
    Scale RGB=&MakeVectorFromLinearColorRGB(Color=$System.TODColor)
    ScaleA=true
    ScaleRGB=true
   )
   /SetVariables(
    Particles.CenterFlareColor:LinearColor=&MultiplyLinearColor(
     Linear Color=$Particles.Color
     Scale Factor=&MakeVector4_FromLinearColor(LinearColor=&MaskLinearColor(
       Mask Source=1
       Mask Parameter=&MakeIntFromFloat(Float=$User.Variant)
       Mask=0
       Masked Linear Color=<rgba 8.0099 9.8007 10 1>
       Passthrough Linear Color=&MaskLinearColor(
        Mask Source=1
        Mask Parameter=&MakeIntFromFloat(Float=$User.Variant)
        Mask=1
        Masked Linear Color=<rgba 10 0 2.5293 1>
        Passthrough Linear Color=&MaskLinearColor(
         Mask Source=1
         Mask Parameter=&MakeIntFromFloat(Float=$User.Variant)
         Mask=2
         Masked Linear Color=<rgba 6.8097 6.5104 10 1>
         Passthrough Linear Color=&MaskLinearColor(
          Mask Source=1
          Mask Parameter=&MakeIntFromFloat(Float=$User.Variant)
          Mask=3
          Masked Linear Color=<rgba 10 9.3031 3.947 1>
          Passthrough Linear Color=&MaskLinearColor(
           Mask Source=1
           Mask Parameter=&MakeIntFromFloat(Float=$User.Variant)
           Mask=4
           Masked Linear Color=<rgba 8.0099 9.8007 10 1>
           Passthrough Linear Color=<rgba 8.0099 9.8007 10 1>
          )
         )
        )
       )
      ))
    )
    Particles.CenterFlareSize:Vector2f=&Multiply_Vector2(A=$Particles.SpriteSize, B=&Vector2DFromFloat(Value=0.5))
    Particles.CoronaFlareColor:LinearColor=&MultiplyLinearColor(
     Linear Color=$Particles.Color
     Scale Factor=&MakeVector4_FromLinearColor(LinearColor=<rgba 0.15 0.15 0.15 1>)
    )
    Particles.CoronaFlareSize:Vector2f=&Multiply_Vector2(A=$Particles.SpriteSize, B=&Vector2DFromFloat(Value=0.75))
    Particles.FlickerFlareColor:LinearColor=&MultiplyLinearColor(
     Linear Color=$Particles.Color
     Scale Factor=&MakeVector4_FromLinearColor(LinearColor=&MaskLinearColor(
       Mask Source=1
       Mask Parameter=&MakeIntFromFloat(Float=$User.Variant)
       Mask=0
       Masked Linear Color=<rgba 0.9441 0.8079 1 3>
       Passthrough Linear Color=&MaskLinearColor(
        Mask Source=1
        Mask Parameter=&MakeIntFromFloat(Float=$User.Variant)
        Mask=1
        Masked Linear Color=<rgba 0.3 0.0487 0.1459 3>
        Passthrough Linear Color=&MaskLinearColor(
         Mask Source=1
         Mask Parameter=&MakeIntFromFloat(Float=$User.Variant)
         Mask=2
         Masked Linear Color=<rgba 0.1182 0.0826 0.3 5>
         Passthrough Linear Color=&MaskLinearColor(
          Mask Source=1
          Mask Parameter=&MakeIntFromFloat(Float=$User.Variant)
          Mask=3
          Masked Linear Color=<rgba 0.2 0.1376 0.02 3>
          Passthrough Linear Color=&MaskLinearColor(
           Mask Source=1
           Mask Parameter=&MakeIntFromFloat(Float=$User.Variant)
           Mask=4
           Masked Linear Color=<rgba 0.75 0.75 0.75 1>
           Passthrough Linear Color=<rgba 0.6059 0.6561 0.75 1>
          )
         )
        )
       )
      ))
    )
   ) {isSetParameters=true, module=SetVariables_#4}
   /SetVariables(Particles.SpriteSize=&Multiply_Vector2D(
     A=$Particles.SpriteSize
     B=&Multiply_Vector2D_ByFloat(
      Float=&MaskFloat(
       Mask Source=1
       Mask Parameter=&MakeIntFromFloat(Float=$User.Variant)
       Mask=0
       Masked Float=&MakeFloatFromLinearColor(LinearColor=$Particles.SampleTexture001.SampledColor, Channel:ENiagaraFloatFromLinearColorOptions=::NewEnumerator0)
       Passthrough Float=&MaskFloat(
        Mask Source=1
        Mask Parameter=&MakeIntFromFloat(Float=$User.Variant)
        Mask=1
        Masked Float=&MakeFloatFromLinearColor(LinearColor=$Particles.SampleTexture001.SampledColor, Channel:ENiagaraFloatFromLinearColorOptions=::NewEnumerator1)
        Passthrough Float=&MaskFloat(
         Mask Source=1
         Mask Parameter=&MakeIntFromFloat(Float=$User.Variant)
         Mask=2
         Masked Float=&MakeFloatFromLinearColor(LinearColor=$Particles.SampleTexture001.SampledColor, Channel:ENiagaraFloatFromLinearColorOptions=::NewEnumerator2)
         Passthrough Float=&MaskFloat(
          Mask Source=1
          Mask Parameter=&MakeIntFromFloat(Float=$User.Variant)
          Mask=3
          Masked Float=&MakeFloatFromLinearColor(LinearColor=$Particles.SampleTexture001.SampledColor, Channel:ENiagaraFloatFromLinearColorOptions=::NewEnumerator3)
          Passthrough Float=1
         )
        )
       )
      )
      Vector2D=<1 1>
     )
    )) {isSetParameters=true, module=SetVariables_#5}
  ]
  rendererClasses [&NiagaraSpriteRendererProperties, &NiagaraMeshRendererProperties]
  renderers [
   {rendererClass=&NiagaraSpriteRendererProperties}
   {rendererClass=&NiagaraSpriteRendererProperties, rendererIndex=1}
   {rendererClass=&NiagaraSpriteRendererProperties, rendererIndex=2}
   {rendererClass=&NiagaraMeshRendererProperties, rendererIndex=3}
  ]
 }
 Blings_Low {
  bEnabled=false
  emitterUpdateScript [
   &EmitterState(Inactive Response=::NewEnumerator1, Life Cycle Mode=::NewEnumerator1, Loop Duration=15, UseLoopDelay=false)
   &SpawnRate(Spawn Probability=0.75, SpawnRate=2.5, Use Spawn Probability=true)
  ]
  particleSpawnScript [
   &InitializeParticle(
    Color=<rgba 5 5 5 1>
    Lifetime Max=0.7
    Lifetime Min=0.5
    Lifetime Mode=::NewEnumerator1
    Sprite Size Mode=::NewEnumerator3
    Uniform Sprite Size=&RandomRangeFloat(Minimum=15, Maximum=25)
    UsePositionOffset=false
    Write Material Random=true
   )
   &CameraOffset(Camera Offset Amount=10)
   /SetVariables(Particles.BoneIndex:NiagaraInt32=&RandomRangeInt(Minimum=0, Maximum=8)) {isSetParameters=true, module=SetVariables_#6}
  ]
  particleUpdateScript [
   &ParticleState()
   &SkeletalMeshLocation(Bone / Socket Index=$Particles.BoneIndex, Bone Sampling Mode=::NewEnumerator4, MaskSpawnGroup=false, Skeletal Mesh=$System.Skeletal Mesh)
   &ShapeLocation(
    Additional Quaternion Rotation=true
    Additional Yaw / Pitch / Roll=&SelectVectorFromArray(
     Direct Array Index=$Particles.BoneIndex
     Vector Selection Array=DataInterface() {propertyValues {FloatData [<0 90 0>, <0 90 0>, <0 90 0>, <0 -70 0>, <0 110 0>, <5 85 0>, <5 85 0>, <0 90 0>, <0 90 0>]}}
    )
    Cylinder Distribution=::NewEnumerator2
    Cylinder Height=&SelectFloatFromArray(Direct Array Index=$Particles.BoneIndex, Float Selection Array=DataInterface() {propertyValues {FloatData [25,25,30,15,15,30,30,22,22]}})
    Cylinder Height Midpoint:NiagaraFloat=&SelectFloatFromArray(Direct Array Index=$Particles.BoneIndex, Float Selection Array=DataInterface() {propertyValues {FloatData [1,0,0,0.5,0.5,-0.6,1.6,1,0]}})
    Cylinder Radius=&SelectFloatFromArray(Direct Array Index=$Particles.BoneIndex, Float Selection Array=DataInterface() {propertyValues {FloatData [3,3,7.5,10,10,4,4,2.5,2.5]}})
    MaskSpawnGroup=false
    Non Uniform Scale=&SelectVectorFromArray(
     Direct Array Index=$Particles.BoneIndex
     Vector Selection Array=DataInterface() {propertyValues {FloatData [<1 1 1>, <1 1 1>, <1 1 1>, <0.6 1 1>, <0.6 1 1>, <1 1 1>, <1 1 1>]}}
    )
    Rotation Angle Type:ENiagara_AngleInput=::NewEnumerator0
    Rotation Mode=::NewEnumerator3
    Rotation Quaternion=$Particles.SkeletalMeshLocation.SampledRotation
    Shape Origin=&Add_Vector(
     A=Linked() {linkedVariable {name=Particles.SkeletalMeshLocation.SampledPosition, type=NiagaraPosition}}
     B=&SelectVectorFromArray(
      Direct Array Index=$Particles.BoneIndex
      Vector Selection Array=DataInterface() {propertyValues {FloatData [<0 0 0>, <0 0 0>, <0 0 0>, <0 0 0>, <0 0 0>, <3 0 0>, <-3 0 0>, <0 0 0>, <0 0 0>]}}
     )
    )
    Shape Primitive=::NewEnumerator1
    U Position=&Lerp_Float(
     A=&RandomRangeFloat(Minimum=0, Maximum=1)
     B=&RandomRangeFloat(Minimum=-0.3, Maximum=0.25)
     Alpha=&MakeCustomFloatFromBool(
      Bool=&SetBoolByFloatComparison(A=&MakeFloatFromInt(Int=$Particles.BoneIndex), B=2)
      True Float=&MakeCustomFloatFromBool(
       Bool=&SetBoolByFloatComparison(A=&MakeFloatFromInt(Int=$Particles.BoneIndex), B=5, Comparison Type=::NewEnumerator6)
       True Float=1
       False Float=0
      )
      False Float=0
     )
    )
    UseOwnerScale=true
    V Position=&RandomRangeFloat(Minimum=0, Maximum=1)
   )
   &KillParticles(Kill Particles=&MaskBool(
     Mask Source:NiagaraInt32=1
     Mask Parameter:NiagaraInt32=&MakeCustomIntFromBool(Bool=$User.bisHandHidden, True INT=1, False INT=0)
     Mask:NiagaraInt32=1
     Masked Bool:NiagaraBool=&SetBoolByFloatComparison(A=&MakeFloatFromInt(Int=$Particles.BoneIndex), B=6)
     Passthrough Bool:NiagaraBool=false
    ))
   &KillParticles(Kill Particles=&MaskBool(
     Mask Source:NiagaraInt32=1
     Mask Parameter:NiagaraInt32=&MakeCustomIntFromBool(Bool=$User.bisFootHidden, True INT=1, False INT=0)
     Mask:NiagaraInt32=1
     Masked Bool:NiagaraBool=&SetBoolByFloatComparison(A=&MakeFloatFromInt(Int=$Particles.BoneIndex), B=6, Comparison Type=::NewEnumerator4)
     Passthrough Bool:NiagaraBool=false
    )) {module=KillParticles001}
   &KillParticles(Kill Particles=&MaskBool(
     Mask Source:NiagaraInt32=1
     Mask Parameter:NiagaraInt32=&MakeCustomIntFromBool(Bool=$User.bisFootHidden, True INT=1, False INT=0)
     Mask:NiagaraInt32=1
     Masked Bool:NiagaraBool=&SetBoolByFloatComparison(A=&MakeFloatFromInt(Int=$Particles.BoneIndex), B=5, Comparison Type=::NewEnumerator4)
     Passthrough Bool:NiagaraBool=false
    )) {module=KillParticles002}
   &ScaleSpriteSize(Uniform Curve Scale=&FNTimeOfDayScalar(Night Scale=1.25, Morning Scale=1, Evening Scale=1)) {module=ScaleSpriteSize001}
   &ScaleColor(
    Scale Alpha=&Multiply_Float(
     A=&FloatFromCurve(
      FloatCurve=DataInterface() {propertyValues {Curve {keys [
          {interpMode=RCIM_Cubic, tangentMode=RCTM_User, time=0, value=0}
          {arriveTangent=-2.4058, interpMode=RCIM_Cubic, leaveTangent=-2.4058, tangentMode=RCTM_User, time=0.26, value=0.8421}
          {arriveTangent=-1.1775, interpMode=RCIM_Cubic, leaveTangent=-1.1775, time=0.4613, value=0.0645}
          {arriveTangent=-0.1197, interpMode=RCIM_Cubic, leaveTangent=-0.1197, time=0.5898, value=0.4537}
          <at 1 0 RCIM_Cubic>
         ]}}}
      Scale Curve=&RandomRangeFloat(Minimum=0.1, Maximum=0.5)
     )
     B=&FloatFromCurve(
      FloatCurve=DataInterface() {propertyValues {Curve {keys [{tangentMode=RCTM_User, time=0, value=0}, {tangentMode=RCTM_User, time=1, value=1}]}}}
      Scale Curve=&FloatFromCurve(
       FloatCurve=DataInterface() {propertyValues {Curve {keys [
           {tangentMode=RCTM_User, time=0, value=0}
           {arriveTangentWeight=0.1667, interpMode=RCIM_Cubic, leaveTangentWeight=0.1667, tangentMode=RCTM_User, time=1, value=1}
          ]}}}
       CurveIndex=$Emitter.Age
      )
     )
    )
    Scale RGB=&VectorFromFloat(Value=&RandomRangeFloat(Minimum=0.7, Maximum=0.9, Evaluation Type=::NewEnumerator1))
    ScaleA=true
    ScaleRGB=true
   ) {module=ScaleColor001}
   &ScaleColor(
    Scale Alpha=$System.TODAlpha
    Scale RGB=&MakeVectorFromLinearColorRGB(Color=$System.TODColor)
    ScaleA=true
    ScaleRGB=true
   )
   /SetVariables(
    Particles.CenterFlareColor:LinearColor=&MultiplyLinearColor(
     Linear Color=$Particles.Color
     Scale Factor=&MakeVector4_FromLinearColor(LinearColor=&MaskLinearColor(
       Mask Source=1
       Mask Parameter=&MakeIntFromFloat(Float=$User.Variant)
       Mask=0
       Masked Linear Color=<rgba 8.0099 9.8007 10 1>
       Passthrough Linear Color=&MaskLinearColor(
        Mask Source=1
        Mask Parameter=&MakeIntFromFloat(Float=$User.Variant)
        Mask=1
        Masked Linear Color=<rgba 10 0 2.5293 1>
        Passthrough Linear Color=&MaskLinearColor(
         Mask Source=1
         Mask Parameter=&MakeIntFromFloat(Float=$User.Variant)
         Mask=2
         Masked Linear Color=<rgba 6.8097 6.5104 10 1>
         Passthrough Linear Color=&MaskLinearColor(
          Mask Source=1
          Mask Parameter=&MakeIntFromFloat(Float=$User.Variant)
          Mask=3
          Masked Linear Color=<rgba 10 9.3031 3.947 1>
          Passthrough Linear Color=&MaskLinearColor(
           Mask Source=1
           Mask Parameter=&MakeIntFromFloat(Float=$User.Variant)
           Mask=4
           Masked Linear Color=<rgba 8.0099 9.8007 10 1>
           Passthrough Linear Color=<rgba 8.0099 9.8007 10 1>
          )
         )
        )
       )
      ))
    )
    Particles.CenterFlareSize:Vector2f=&Multiply_Vector2(A=$Particles.SpriteSize, B=&Vector2DFromFloat(Value=0.5))
    Particles.CoronaFlareColor:LinearColor=&MultiplyLinearColor(
     Linear Color=$Particles.Color
     Scale Factor=&MakeVector4_FromLinearColor(LinearColor=<rgba 0.15 0.15 0.15 1>)
    )
    Particles.CoronaFlareSize:Vector2f=&Multiply_Vector2(A=$Particles.SpriteSize, B=&Vector2DFromFloat(Value=0.75))
    Particles.FlickerFlareColor:LinearColor=&MultiplyLinearColor(
     Linear Color=$Particles.Color
     Scale Factor=&MakeVector4_FromLinearColor(LinearColor=&MaskLinearColor(
       Mask Source=1
       Mask Parameter=&MakeIntFromFloat(Float=$User.Variant)
       Mask=0
       Masked Linear Color=<rgba 0.9441 0.8079 1 1>
       Passthrough Linear Color=&MaskLinearColor(
        Mask Source=1
        Mask Parameter=&MakeIntFromFloat(Float=$User.Variant)
        Mask=1
        Masked Linear Color=<rgba 0.3 0.0487 0.1459 3>
        Passthrough Linear Color=&MaskLinearColor(
         Mask Source=1
         Mask Parameter=&MakeIntFromFloat(Float=$User.Variant)
         Mask=2
         Masked Linear Color=<rgba 0.1182 0.0826 0.3 5>
         Passthrough Linear Color=&MaskLinearColor(
          Mask Source=1
          Mask Parameter=&MakeIntFromFloat(Float=$User.Variant)
          Mask=3
          Masked Linear Color=<rgba 0.2 0.1376 0.02 3>
          Passthrough Linear Color=&MaskLinearColor(
           Mask Source=1
           Mask Parameter=&MakeIntFromFloat(Float=$User.Variant)
           Mask=4
           Masked Linear Color=<rgba 0.75 0.75 0.75 1>
           Passthrough Linear Color=<rgba 0.6059 0.6561 0.75 1>
          )
         )
        )
       )
      ))
    )
   ) {isSetParameters=true, module=SetVariables_#4}
   /SetVariables(Particles.SpriteSize=&Multiply_Vector2D(
     A=$Particles.SpriteSize
     B=&Multiply_Vector2D_ByFloat(
      Float=&MaskFloat(
       Mask Source=1
       Mask Parameter=&MakeIntFromFloat(Float=$User.Variant)
       Mask=0
       Masked Float=&MakeFloatFromLinearColor(LinearColor=$Particles.SampleTexture001.SampledColor, Channel:ENiagaraFloatFromLinearColorOptions=::NewEnumerator0)
       Passthrough Float=&MaskFloat(
        Mask Source=1
        Mask Parameter=&MakeIntFromFloat(Float=$User.Variant)
        Mask=1
        Masked Float=&MakeFloatFromLinearColor(LinearColor=$Particles.SampleTexture001.SampledColor, Channel:ENiagaraFloatFromLinearColorOptions=::NewEnumerator1)
        Passthrough Float=&MaskFloat(
         Mask Source=1
         Mask Parameter=&MakeIntFromFloat(Float=$User.Variant)
         Mask=2
         Masked Float=&MakeFloatFromLinearColor(LinearColor=$Particles.SampleTexture001.SampledColor, Channel:ENiagaraFloatFromLinearColorOptions=::NewEnumerator2)
         Passthrough Float=&MaskFloat(
          Mask Source=1
          Mask Parameter=&MakeIntFromFloat(Float=$User.Variant)
          Mask=3
          Masked Float=&MakeFloatFromLinearColor(LinearColor=$Particles.SampleTexture001.SampledColor, Channel:ENiagaraFloatFromLinearColorOptions=::NewEnumerator3)
          Passthrough Float=1
         )
        )
       )
      )
      Vector2D=<1 1>
     )
    )) {isSetParameters=true, module=SetVariables_#5}
  ]
  rendererClasses [&NiagaraSpriteRendererProperties]
  renderers [
   {rendererClass=&NiagaraSpriteRendererProperties}
   {rendererClass=&NiagaraSpriteRendererProperties, rendererIndex=1}
  ]
 }
 Blings_Low001 {
  emitterUpdateScript [
   &EmitterState(Inactive Response=::NewEnumerator1, Life Cycle Mode=::NewEnumerator1, Loop Duration=15, UseLoopDelay=false)
   &SpawnRate(Spawn Probability=0.75, SpawnRate=10, Use Spawn Probability=true)
  ]
  particleSpawnScript [
   &InitializeParticle(
    Color=<rgba 5 5 5 1>
    Lifetime Max=0.7
    Lifetime Min=0.5
    Lifetime Mode=::NewEnumerator1
    Sprite Size Mode=::NewEnumerator3
    Uniform Sprite Size=&RandomRangeFloat(Minimum=15, Maximum=25)
    UsePositionOffset=false
    Write Material Random=true
   )
   &CameraOffset(Camera Offset Amount=10)
   /SetVariables(Particles.BoneIndex:NiagaraInt32=&RandomRangeInt(Minimum=0, Maximum=8)) {isSetParameters=true, module=SetVariables_#6}
  ]
  particleUpdateScript [
   &ParticleState()
   &SkeletalMeshLocation(Bone / Socket Index=$Particles.BoneIndex, Bone Sampling Mode=::NewEnumerator4, MaskSpawnGroup=false, Skeletal Mesh=$System.Skeletal Mesh)
   &ShapeLocation(
    Additional Quaternion Rotation=true
    Additional Yaw / Pitch / Roll=&SelectVectorFromArray(
     Direct Array Index=$Particles.BoneIndex
     Vector Selection Array=DataInterface() {propertyValues {FloatData [<0 90 0>, <0 90 0>, <0 90 0>, <0 -70 0>, <0 110 0>, <0 85 0>, <0 -95 0>, <0 90 0>, <0 90 0>]}}
    )
    Cylinder Distribution=::NewEnumerator2
    Cylinder Height=&SelectFloatFromArray(Direct Array Index=$Particles.BoneIndex, Float Selection Array=DataInterface() {propertyValues {FloatData [25,25,30,15,15,30,30,22,22]}})
    Cylinder Height Midpoint:NiagaraFloat=&SelectFloatFromArray(Direct Array Index=$Particles.BoneIndex, Float Selection Array=DataInterface() {propertyValues {FloatData [1,0,0,0.5,0.5,0,0,1,0]}})
    Cylinder Radius=&SelectFloatFromArray(Direct Array Index=$Particles.BoneIndex, Float Selection Array=DataInterface() {propertyValues {FloatData [3,3,7.5,10,10,3.5,3.5,2.5,2.5]}})
    MaskSpawnGroup=false
    Non Uniform Scale=&SelectVectorFromArray(
     Direct Array Index=$Particles.BoneIndex
     Vector Selection Array=DataInterface() {propertyValues {FloatData [<1 1 1>, <1 1 1>, <1 1 1>, <0.6 1 1>, <0.6 1 1>, <1 1 1>, <1 1 1>]}}
    )
    Rotation Angle Type:ENiagara_AngleInput=::NewEnumerator0
    Rotation Mode=::NewEnumerator3
    Rotation Quaternion=$Particles.SkeletalMeshLocation.SampledRotation
    Shape Origin=&Add_Vector(
     A=Linked() {linkedVariable {name=Particles.SkeletalMeshLocation.SampledPosition, type=NiagaraPosition}}
     B=&SelectVectorFromArray(
      Direct Array Index=$Particles.BoneIndex
      Vector Selection Array=DataInterface() {propertyValues {FloatData [<0 0 0>, <0 0 0>, <0 0 0>, <0 0 0>, <0 0 0>, <2 -1 0>, <-2 -1 0>, <0 0 0>, <0 0 0>]}}
     )
    )
    Shape Primitive=::NewEnumerator1
    U Position=&Lerp_Float(
     A=&RandomRangeFloat(Minimum=0, Maximum=1)
     B=&RandomRangeFloat(Minimum=-0.3, Maximum=0.25)
     Alpha=&MakeCustomFloatFromBool(
      Bool=&SetBoolByFloatComparison(A=&MakeFloatFromInt(Int=$Particles.BoneIndex), B=2)
      True Float=&MakeCustomFloatFromBool(
       Bool=&SetBoolByFloatComparison(A=&MakeFloatFromInt(Int=$Particles.BoneIndex), B=5, Comparison Type=::NewEnumerator6)
       True Float=1
       False Float=0
      )
      False Float=0
     )
    )
    UseOwnerScale=true
    V Position=&RandomRangeFloat(Minimum=0, Maximum=1)
   )
   &KillParticles(Kill Particles=&MaskBool(
     Mask Source:NiagaraInt32=1
     Mask Parameter:NiagaraInt32=&MakeCustomIntFromBool(Bool=$User.bisHandHidden, True INT=1, False INT=0)
     Mask:NiagaraInt32=1
     Masked Bool:NiagaraBool=&SetBoolByFloatComparison(A=&MakeFloatFromInt(Int=$Particles.BoneIndex), B=6)
     Passthrough Bool:NiagaraBool=false
    ))
   &KillParticles(Kill Particles=&MaskBool(
     Mask Source:NiagaraInt32=1
     Mask Parameter:NiagaraInt32=&MakeCustomIntFromBool(Bool=$User.bisFootHidden, True INT=1, False INT=0)
     Mask:NiagaraInt32=1
     Masked Bool:NiagaraBool=&SetBoolByFloatComparison(A=&MakeFloatFromInt(Int=$Particles.BoneIndex), B=6, Comparison Type=::NewEnumerator4)
     Passthrough Bool:NiagaraBool=&MaskBool(
      Mask Source:NiagaraInt32=1
      Mask Parameter:NiagaraInt32=&MakeIntFromFloat(Float=$User.Variant)
      Mask:NiagaraInt32=3
      Masked Bool:NiagaraBool=&SetBoolByFloatComparison(A=&MakeFloatFromInt(Int=$Particles.BoneIndex), B=6, Comparison Type=::NewEnumerator4)
      Passthrough Bool:NiagaraBool=false
     )
    )) {module=KillParticles001}
   &KillParticles(Kill Particles=&MaskBool(
     Mask Source:NiagaraInt32=1
     Mask Parameter:NiagaraInt32=&MakeCustomIntFromBool(Bool=$User.bisFootHidden, True INT=1, False INT=0)
     Mask:NiagaraInt32=1
     Masked Bool:NiagaraBool=&SetBoolByFloatComparison(A=&MakeFloatFromInt(Int=$Particles.BoneIndex), B=5, Comparison Type=::NewEnumerator4)
     Passthrough Bool:NiagaraBool=&MaskBool(
      Mask Source:NiagaraInt32=1
      Mask Parameter:NiagaraInt32=&MakeIntFromFloat(Float=$User.Variant)
      Mask:NiagaraInt32=3
      Masked Bool:NiagaraBool=&SetBoolByFloatComparison(A=&MakeFloatFromInt(Int=$Particles.BoneIndex), B=5, Comparison Type=::NewEnumerator4)
      Passthrough Bool:NiagaraBool=false
     )
    )) {module=KillParticles002}
   &ScaleSpriteSize(Uniform Curve Scale=&FNTimeOfDayScalar(Night Scale=1.25, Morning Scale=1, Evening Scale=1)) {module=ScaleSpriteSize001}
   &ScaleColor(
    Scale Alpha=&Multiply_Float(
     A=&FloatFromCurve(
      FloatCurve=DataInterface() {propertyValues {Curve {keys [
          {interpMode=RCIM_Cubic, tangentMode=RCTM_User, time=0, value=0}
          {arriveTangent=-2.4058, interpMode=RCIM_Cubic, leaveTangent=-2.4058, tangentMode=RCTM_User, time=0.26, value=0.8421}
          {arriveTangent=-1.1775, interpMode=RCIM_Cubic, leaveTangent=-1.1775, time=0.4613, value=0.0645}
          {arriveTangent=-0.1197, interpMode=RCIM_Cubic, leaveTangent=-0.1197, time=0.5898, value=0.4537}
          <at 1 0 RCIM_Cubic>
         ]}}}
      Scale Curve=&RandomRangeFloat(Minimum=0.1, Maximum=0.5)
     )
     B=&FloatFromCurve(
      FloatCurve=DataInterface() {propertyValues {Curve {keys [{tangentMode=RCTM_User, time=0, value=0}, {tangentMode=RCTM_User, time=1, value=1}]}}}
      Scale Curve=&FloatFromCurve(
       FloatCurve=DataInterface() {propertyValues {Curve {keys [
           {tangentMode=RCTM_User, time=0, value=0}
           {arriveTangentWeight=0.1667, interpMode=RCIM_Cubic, leaveTangentWeight=0.1667, tangentMode=RCTM_User, time=1, value=1}
          ]}}}
       CurveIndex=$Emitter.Age
      )
     )
    )
    Scale RGB=&VectorFromFloat(Value=&RandomRangeFloat(Minimum=0.7, Maximum=0.9, Evaluation Type=::NewEnumerator1))
    ScaleA=true
    ScaleRGB=true
   ) {module=ScaleColor001}
   &ScaleColor(
    Scale Alpha=$System.TODAlpha
    Scale RGB=&MakeVectorFromLinearColorRGB(Color=$System.TODColor)
    ScaleA=true
    ScaleRGB=true
   )
   /SetVariables(Particles.EdgeAlpha:NiagaraFloat=&DotProduct(
     Vector A=&Multiply_VectorByFloat(Vector=$Particles.ShapeLocation.ShapeNormal, Float=1)
     Vector B=&Multiply_VectorByFloat(
      Vector=&ReturnCameraVector(
       Camera Vectors=::NewEnumerator1
       Coordinate Space=::Simulation
       CameraQuery=DataInterface() {propertyValues {bRequireCurrentFrameData=false}}
      )
      Float=-1
     )
    )) {isSetParameters=true}
   /SetVariables(
    Particles.CenterFlareColor:LinearColor=&Multiply_LinearColorByFloat(
     Linear Color=&MultiplyLinearColor(
      Linear Color=$Particles.Color
      Scale Factor=&MakeVector4_FromLinearColor(LinearColor=$System.BlingsCenterFlare)
     )
     Float=$Particles.EdgeAlpha
    )
    Particles.CenterFlareSize:Vector2f=&Multiply_Vector2(A=$Particles.SpriteSize, B=&Vector2DFromFloat(Value=0.5))
    Particles.CoronaFlareColor:LinearColor=&MultiplyLinearColor(
     Linear Color=$Particles.Color
     Scale Factor=&MakeVector4_FromLinearColor(LinearColor=<rgba 0.15 0.15 0.15 1>)
    )
    Particles.CoronaFlareSize:Vector2f=&Multiply_Vector2(A=$Particles.SpriteSize, B=&Vector2DFromFloat(Value=0.75))
    Particles.FlickerFlareColor:LinearColor=&Multiply_LinearColorByFloat(
     Linear Color=&MultiplyLinearColor(
      Linear Color=$Particles.Color
      Scale Factor=&MakeVector4_FromLinearColor(LinearColor=$System.BlingsFlickerFlare)
     )
     Float=$Particles.EdgeAlpha
    )
   ) {isSetParameters=true, module=SetVariables_#4}
  ]
  rendererClasses [&NiagaraSpriteRendererProperties]
  renderers [
   {rendererClass=&NiagaraSpriteRendererProperties}
   {rendererClass=&NiagaraSpriteRendererProperties, rendererIndex=1}
  ]
 }
 Sparkles {
  bEnabled=false
  simTarget=GPUComputeSim
  emitterUpdateScript [
   &EmitterState(Inactive Response=::NewEnumerator1, Life Cycle Mode=::NewEnumerator1, Loop Duration=15, UseLoopDelay=false)
   &SpawnRate(SpawnRate=50, Use Spawn Probability=true)
  ]
  particleSpawnScript [
   &InitializeParticle(
    Lifetime Max=0.4
    Lifetime Min=0.6
    Lifetime Mode=::NewEnumerator1
    Sprite Size Mode=::NewEnumerator3
    Uniform Sprite Size=&RandomRangeFloat(Minimum=15, Maximum=25)
    UsePositionOffset=false
    Write Material Random=true
   )
   &CameraOffset(Camera Offset Amount=5)
  ]
  particleUpdateScript [
   &ParticleState()
   &SkeletalMeshLocation(
    MaskSpawnGroup=false
    Mesh Sampling Type=::NewEnumerator7
    OffsetAlongNormal=false
    Orientation Sampling=::NewEnumerator1
    OverrideRandomSeed=false
    Skeletal Mesh=$System.Skeletal Mesh
   )
   &SampleTexture(
    Sampled RGBA Mask:Vector4f=<1 1 1 1>
    Texture=DataInterface() {propertyValues {
      Texture=&T_SaxoPop_Idle_FX
      TextureUserParameter {parameter {name=None, typeDefHandle {registeredTypeIndex=2}, varData []}}
     }}
    UV:Vector2f=$Particles.SkeletalMeshLocation.SampledUV
   ) {module=SampleTexture001}
   &SampleTexture(
    Sampled RGBA Mask:Vector4f=<1 1 1 1>
    Texture=DataInterface() {propertyValues {
      Texture=&SaxoPop_Body_FX_2
      TextureUserParameter {parameter {name=None, typeDefHandle {registeredTypeIndex=2}, varData []}}
     }}
    UV:Vector2f=$Particles.SkeletalMeshLocation.SampledUV
   )
   &KillParticles(Kill Particles=&SetBoolByFloatComparison(
     A=&MaskFloat(
      Mask Source=1
      Mask Parameter=&MakeIntFromFloat(Float=$User.Variant)
      Mask=0
      Masked Float=&MakeFloatFromLinearColor(LinearColor=$Particles.SampleTexture001.SampledColor, Channel:ENiagaraFloatFromLinearColorOptions=::NewEnumerator0)
      Passthrough Float=&MaskFloat(
       Mask Source=1
       Mask Parameter=&MakeIntFromFloat(Float=$User.Variant)
       Mask=1
       Masked Float=&MakeFloatFromLinearColor(LinearColor=$Particles.SampleTexture001.SampledColor, Channel:ENiagaraFloatFromLinearColorOptions=::NewEnumerator1)
       Passthrough Float=&MaskFloat(
        Mask Source=1
        Mask Parameter=&MakeIntFromFloat(Float=$User.Variant)
        Mask=2
        Masked Float=&MakeFloatFromLinearColor(LinearColor=$Particles.SampleTexture001.SampledColor, Channel:ENiagaraFloatFromLinearColorOptions=::NewEnumerator2)
        Passthrough Float=&MaskFloat(
         Mask Source=1
         Mask Parameter=&MakeIntFromFloat(Float=$User.Variant)
         Mask=3
         Masked Float=&MakeFloatFromLinearColor(LinearColor=$Particles.SampleTexture001.SampledColor, Channel:ENiagaraFloatFromLinearColorOptions=::NewEnumerator3)
         Passthrough Float=&MaskFloat(
          Mask Source=1
          Mask Parameter=&MakeIntFromFloat(Float=$User.Variant)
          Mask=4
          Masked Float=&MakeFloatFromLinearColor(LinearColor=$Particles.SampleTexture.SampledColor, Channel:ENiagaraFloatFromLinearColorOptions=::NewEnumerator0)
          Passthrough Float=&MaskFloat(
           Mask Source=1
           Mask Parameter=&MakeIntFromFloat(Float=$User.Variant)
           Mask=5
           Masked Float=&MakeFloatFromLinearColor(LinearColor=$Particles.SampleTexture.SampledColor, Channel:ENiagaraFloatFromLinearColorOptions=::NewEnumerator1)
           Passthrough Float=1
          )
         )
        )
       )
      )
     )
     B=0.1
     Comparison Type=::NewEnumerator6
    )) {module=KillParticles001}
   &KillParticles(Kill Particles=&MaskBool(
     Mask Source:NiagaraInt32=1
     Mask Parameter:NiagaraInt32=&MakeCustomIntFromBool(Bool=$User.bisFootHidden, True INT=0, False INT=1)
     Mask:NiagaraInt32=0
     Masked Bool:NiagaraBool=&SetBoolByFloatComparison(
      A=&MakeFloatFromLinearColor(LinearColor=$Particles.SampleTexture.SampledColor, Channel:ENiagaraFloatFromLinearColorOptions=::NewEnumerator2)
      B=0.9
      Comparison Type=::NewEnumerator6
     )
     Passthrough Bool:NiagaraBool=false
    ))
   &KillParticles(Kill Particles=&MaskBool(
     Mask Source:NiagaraInt32=1
     Mask Parameter:NiagaraInt32=&MakeCustomIntFromBool(Bool=$User.bisHandHidden, True INT=0, False INT=1)
     Mask:NiagaraInt32=0
     Masked Bool:NiagaraBool=&SetBoolByFloatComparison(
      A=&MakeFloatFromLinearColor(LinearColor=$Particles.SampleTexture.SampledColor, Channel:ENiagaraFloatFromLinearColorOptions=::NewEnumerator3)
      B=0.9
      Comparison Type=::NewEnumerator6
     )
     Passthrough Bool:NiagaraBool=false
    )) {module=KillParticles002}
   &ScaleSpriteSize(Uniform Curve Scale=&FNTimeOfDayScalar(Night Scale=1.25, Morning Scale=1, Evening Scale=1)) {module=ScaleSpriteSize001}
   &ScaleColor(
    Scale Alpha=&Multiply_Float(
     A=&FloatFromCurve(
      FloatCurve=DataInterface() {propertyValues {Curve {keys [
          {interpMode=RCIM_Cubic, tangentMode=RCTM_User, time=0, value=0}
          {arriveTangent=-2.4058, interpMode=RCIM_Cubic, leaveTangent=-2.4058, tangentMode=RCTM_User, time=0.26, value=0.8421}
          {arriveTangent=-1.1775, interpMode=RCIM_Cubic, leaveTangent=-1.1775, time=0.4613, value=0.0645}
          {arriveTangent=-0.1197, interpMode=RCIM_Cubic, leaveTangent=-0.1197, time=0.5898, value=0.4537}
          <at 1 0 RCIM_Cubic>
         ]}}}
      Scale Curve=&RandomRangeFloat(Minimum=0.1, Maximum=0.5)
     )
     B=&FloatFromCurve(
      FloatCurve=DataInterface() {propertyValues {Curve {keys [{tangentMode=RCTM_User, time=0, value=0}, {tangentMode=RCTM_User, time=1, value=1}]}}}
      Scale Curve=&FloatFromCurve(
       FloatCurve=DataInterface() {propertyValues {Curve {keys [
           {tangentMode=RCTM_User, time=0, value=0}
           {arriveTangentWeight=0.1667, interpMode=RCIM_Cubic, leaveTangentWeight=0.1667, tangentMode=RCTM_User, time=1, value=1}
          ]}}}
       CurveIndex=$Emitter.Age
      )
     )
    )
    Scale RGB=&VectorFromFloat(Value=&RandomRangeFloat(Minimum=0.7, Maximum=0.9, Evaluation Type=::NewEnumerator1))
    ScaleA=true
    ScaleRGB=true
   ) {module=ScaleColor001}
   &ScaleColor(
    Scale Alpha=$System.TODAlpha
    Scale RGB=&MakeVectorFromLinearColorRGB(Color=$System.TODColor)
    ScaleA=true
    ScaleRGB=true
   )
   /SetVariables(
    Particles.CenterFlareColor:LinearColor=&MultiplyLinearColor(
     Linear Color=$Particles.Color
     Scale Factor=&MakeVector4_FromLinearColor(LinearColor=&MaskLinearColor(
       Mask Source=1
       Mask Parameter=&MakeIntFromFloat(Float=$User.Variant)
       Mask=0
       Masked Linear Color=<rgba 8.0099 9.8007 10 1>
       Passthrough Linear Color=&MaskLinearColor(
        Mask Source=1
        Mask Parameter=&MakeIntFromFloat(Float=$User.Variant)
        Mask=1
        Masked Linear Color=<rgba 10 0 2.5293 1>
        Passthrough Linear Color=&MaskLinearColor(
         Mask Source=1
         Mask Parameter=&MakeIntFromFloat(Float=$User.Variant)
         Mask=2
         Masked Linear Color=<rgba 6.8097 6.5104 10 1>
         Passthrough Linear Color=&MaskLinearColor(
          Mask Source=1
          Mask Parameter=&MakeIntFromFloat(Float=$User.Variant)
          Mask=3
          Masked Linear Color=<rgba 10 9.3031 3.947 1>
          Passthrough Linear Color=&MaskLinearColor(
           Mask Source=1
           Mask Parameter=&MakeIntFromFloat(Float=$User.Variant)
           Mask=4
           Masked Linear Color=<rgba 8.0099 9.8007 10 1>
           Passthrough Linear Color=<rgba 8.0099 9.8007 10 1>
          )
         )
        )
       )
      ))
    )
    Particles.CenterFlareSize:Vector2f=&Multiply_Vector2(A=$Particles.SpriteSize, B=&Vector2DFromFloat(Value=0.5))
    Particles.CoronaFlareColor:LinearColor=&MultiplyLinearColor(
     Linear Color=$Particles.Color
     Scale Factor=&MakeVector4_FromLinearColor(LinearColor=<rgba 0.15 0.15 0.15 1>)
    )
    Particles.CoronaFlareSize:Vector2f=&Multiply_Vector2(A=$Particles.SpriteSize, B=&Vector2DFromFloat(Value=0.75))
    Particles.FlickerFlareColor:LinearColor=&MultiplyLinearColor(
     Linear Color=$Particles.Color
     Scale Factor=&MakeVector4_FromLinearColor(LinearColor=&MaskLinearColor(
       Mask Source=1
       Mask Parameter=&MakeIntFromFloat(Float=$User.Variant)
       Mask=0
       Masked Linear Color=<rgba 0.9441 0.8079 1 1>
       Passthrough Linear Color=&MaskLinearColor(
        Mask Source=1
        Mask Parameter=&MakeIntFromFloat(Float=$User.Variant)
        Mask=1
        Masked Linear Color=<rgba 0.3 0.0562 0.1112 50>
        Passthrough Linear Color=&MaskLinearColor(
         Mask Source=1
         Mask Parameter=&MakeIntFromFloat(Float=$User.Variant)
         Mask=2
         Masked Linear Color=<rgba 0.0314 0.0322 0.1 50>
         Passthrough Linear Color=&MaskLinearColor(
          Mask Source=1
          Mask Parameter=&MakeIntFromFloat(Float=$User.Variant)
          Mask=3
          Masked Linear Color=<rgba 0.05 0.0344 0.005 50>
          Passthrough Linear Color=&MaskLinearColor(
           Mask Source=1
           Mask Parameter=&MakeIntFromFloat(Float=$User.Variant)
           Mask=4
           Masked Linear Color=<rgba 0.5 0.5 0.5 1>
           Passthrough Linear Color=<rgba 0.404 0.4624 0.5 1>
          )
         )
        )
       )
      ))
    )
   ) {isSetParameters=true, module=SetVariables_#4}
   /SetVariables(Particles.SpriteSize=&Multiply_Vector2D(
     A=$Particles.SpriteSize
     B=&Multiply_Vector2D_ByFloat(
      Float=&MaskFloat(
       Mask Source=1
       Mask Parameter=&MakeIntFromFloat(Float=$User.Variant)
       Mask=0
       Masked Float=&MakeFloatFromLinearColor(LinearColor=$Particles.SampleTexture001.SampledColor, Channel:ENiagaraFloatFromLinearColorOptions=::NewEnumerator0)
       Passthrough Float=&MaskFloat(
        Mask Source=1
        Mask Parameter=&MakeIntFromFloat(Float=$User.Variant)
        Mask=1
        Masked Float=&MakeFloatFromLinearColor(LinearColor=$Particles.SampleTexture001.SampledColor, Channel:ENiagaraFloatFromLinearColorOptions=::NewEnumerator1)
        Passthrough Float=&MaskFloat(
         Mask Source=1
         Mask Parameter=&MakeIntFromFloat(Float=$User.Variant)
         Mask=2
         Masked Float=&MakeFloatFromLinearColor(LinearColor=$Particles.SampleTexture001.SampledColor, Channel:ENiagaraFloatFromLinearColorOptions=::NewEnumerator2)
         Passthrough Float=&MaskFloat(
          Mask Source=1
          Mask Parameter=&MakeIntFromFloat(Float=$User.Variant)
          Mask=3
          Masked Float=&MakeFloatFromLinearColor(LinearColor=$Particles.SampleTexture001.SampledColor, Channel:ENiagaraFloatFromLinearColorOptions=::NewEnumerator3)
          Passthrough Float=1
         )
        )
       )
      )
      Vector2D=<1 1>
     )
    )) {isSetParameters=true, module=SetVariables_#7}
  ]
  rendererClasses [&NiagaraSpriteRendererProperties, &NiagaraMeshRendererProperties]
  renderers [
   {rendererClass=&NiagaraSpriteRendererProperties}
   {rendererClass=&NiagaraSpriteRendererProperties, rendererIndex=1}
   {rendererClass=&NiagaraSpriteRendererProperties, rendererIndex=2}
   {rendererClass=&NiagaraMeshRendererProperties, rendererIndex=3}
  ]
 }
 Sparkles_Low {
  bEnabled=false
  emitterUpdateScript [
   &EmitterState(Inactive Response=::NewEnumerator1, Life Cycle Mode=::NewEnumerator1, Loop Duration=15, UseLoopDelay=false)
   &SpawnRate(SpawnRate=15, Use Spawn Probability=true)
  ]
  particleSpawnScript [
   &InitializeParticle(
    Lifetime Max=0.4
    Lifetime Min=0.6
    Lifetime Mode=::NewEnumerator1
    Sprite Size Mode=::NewEnumerator3
    Uniform Sprite Size=&RandomRangeFloat(Minimum=15, Maximum=25)
    UsePositionOffset=false
    Write Material Random=true
   )
   &CameraOffset(Camera Offset Amount=10)
   /SetVariables(Particles.BoneIndex:NiagaraInt32=&RandomRangeInt(Minimum=0, Maximum=8)) {isSetParameters=true, module=SetVariables_#8}
  ]
  particleUpdateScript [
   &ParticleState()
   &SkeletalMeshLocation(Bone / Socket Index=$Particles.BoneIndex, Bone Sampling Mode=::NewEnumerator4, MaskSpawnGroup=false, Skeletal Mesh=$System.Skeletal Mesh)
   &KillParticles(Kill Particles=&MaskBool(
     Mask Source:NiagaraInt32=1
     Mask Parameter:NiagaraInt32=&MakeCustomIntFromBool(Bool=$User.bisHandHidden, True INT=1, False INT=0)
     Mask:NiagaraInt32=1
     Masked Bool:NiagaraBool=&SetBoolByFloatComparison(A=&MakeFloatFromInt(Int=$Particles.BoneIndex), B=6)
     Passthrough Bool:NiagaraBool=false
    ))
   &KillParticles(Kill Particles=&MaskBool(
     Mask Source:NiagaraInt32=1
     Mask Parameter:NiagaraInt32=&MakeCustomIntFromBool(Bool=$User.bisFootHidden, True INT=1, False INT=0)
     Mask:NiagaraInt32=1
     Masked Bool:NiagaraBool=&SetBoolByFloatComparison(A=&MakeFloatFromInt(Int=$Particles.BoneIndex), B=6, Comparison Type=::NewEnumerator4)
     Passthrough Bool:NiagaraBool=false
    )) {module=KillParticles001}
   &KillParticles(Kill Particles=&MaskBool(
     Mask Source:NiagaraInt32=1
     Mask Parameter:NiagaraInt32=&MakeCustomIntFromBool(Bool=$User.bisFootHidden, True INT=1, False INT=0)
     Mask:NiagaraInt32=1
     Masked Bool:NiagaraBool=&SetBoolByFloatComparison(A=&MakeFloatFromInt(Int=$Particles.BoneIndex), B=5, Comparison Type=::NewEnumerator4)
     Passthrough Bool:NiagaraBool=false
    )) {module=KillParticles002}
   &ShapeLocation(
    Additional Quaternion Rotation=true
    Additional Yaw / Pitch / Roll=&SelectVectorFromArray(
     Direct Array Index=$Particles.BoneIndex
     Vector Selection Array=DataInterface() {propertyValues {FloatData [<0 90 0>, <0 90 0>, <0 90 0>, <0 -70 0>, <0 110 0>, <5 85 0>, <5 85 0>, <0 90 0>, <0 90 0>]}}
    )
    Cylinder Distribution=::NewEnumerator2
    Cylinder Height=&SelectFloatFromArray(Direct Array Index=$Particles.BoneIndex, Float Selection Array=DataInterface() {propertyValues {FloatData [25,25,30,15,15,30,30,22,22]}})
    Cylinder Height Midpoint:NiagaraFloat=&SelectFloatFromArray(Direct Array Index=$Particles.BoneIndex, Float Selection Array=DataInterface() {propertyValues {FloatData [1,0,0,0.5,0.5,-0.6,1.6,1,0]}})
    Cylinder Radius=&SelectFloatFromArray(Direct Array Index=$Particles.BoneIndex, Float Selection Array=DataInterface() {propertyValues {FloatData [3,3,7.5,10,10,4,4,2.5,2.5]}})
    MaskSpawnGroup=false
    Non Uniform Scale=&SelectVectorFromArray(
     Direct Array Index=$Particles.BoneIndex
     Vector Selection Array=DataInterface() {propertyValues {FloatData [<1 1 1>, <1 1 1>, <1 1 1>, <0.6 1 1>, <0.6 1 1>, <1 1 1>, <1 1 1>]}}
    )
    Rotation Angle Type:ENiagara_AngleInput=::NewEnumerator0
    Rotation Mode=::NewEnumerator3
    Rotation Quaternion=$Particles.SkeletalMeshLocation.SampledRotation
    Shape Origin=&Add_Vector(
     A=Linked() {linkedVariable {name=Particles.SkeletalMeshLocation.SampledPosition, type=NiagaraPosition}}
     B=&SelectVectorFromArray(
      Direct Array Index=$Particles.BoneIndex
      Vector Selection Array=DataInterface() {propertyValues {FloatData [<0 0 0>, <0 0 0>, <0 0 0>, <0 0 0>, <0 0 0>, <3 0 0>, <-3 0 0>, <0 0 0>, <0 0 0>]}}
     )
    )
    Shape Primitive=::NewEnumerator1
    U Position=&Lerp_Float(
     A=&RandomRangeFloat(Minimum=0, Maximum=1)
     B=&RandomRangeFloat(Minimum=-0.3, Maximum=0.25)
     Alpha=&MakeCustomFloatFromBool(
      Bool=&SetBoolByFloatComparison(A=&MakeFloatFromInt(Int=$Particles.BoneIndex), B=2)
      True Float=&MakeCustomFloatFromBool(
       Bool=&SetBoolByFloatComparison(A=&MakeFloatFromInt(Int=$Particles.BoneIndex), B=5, Comparison Type=::NewEnumerator6)
       True Float=1
       False Float=0
      )
      False Float=0
     )
    )
    UseOwnerScale=true
    V Position=&RandomRangeFloat(Minimum=0, Maximum=1)
   )
   &ScaleSpriteSize(Uniform Curve Scale=&FNTimeOfDayScalar(Night Scale=1.25, Morning Scale=1, Evening Scale=1)) {module=ScaleSpriteSize001}
   &ScaleColor(
    Scale Alpha=&Multiply_Float(
     A=&FloatFromCurve(
      FloatCurve=DataInterface() {propertyValues {Curve {keys [
          {interpMode=RCIM_Cubic, tangentMode=RCTM_User, time=0, value=0}
          {arriveTangent=-2.4058, interpMode=RCIM_Cubic, leaveTangent=-2.4058, tangentMode=RCTM_User, time=0.26, value=0.8421}
          {arriveTangent=-1.1775, interpMode=RCIM_Cubic, leaveTangent=-1.1775, time=0.4613, value=0.0645}
          {arriveTangent=-0.1197, interpMode=RCIM_Cubic, leaveTangent=-0.1197, time=0.5898, value=0.4537}
          <at 1 0 RCIM_Cubic>
         ]}}}
      Scale Curve=&RandomRangeFloat(Minimum=0.1, Maximum=0.5)
     )
     B=&FloatFromCurve(
      FloatCurve=DataInterface() {propertyValues {Curve {keys [{tangentMode=RCTM_User, time=0, value=0}, {tangentMode=RCTM_User, time=1, value=1}]}}}
      Scale Curve=&FloatFromCurve(
       FloatCurve=DataInterface() {propertyValues {Curve {keys [
           {tangentMode=RCTM_User, time=0, value=0}
           {arriveTangentWeight=0.1667, interpMode=RCIM_Cubic, leaveTangentWeight=0.1667, tangentMode=RCTM_User, time=1, value=1}
          ]}}}
       CurveIndex=$Emitter.Age
      )
     )
    )
    Scale RGB=&VectorFromFloat(Value=&RandomRangeFloat(Minimum=0.7, Maximum=0.9, Evaluation Type=::NewEnumerator1))
    ScaleA=true
    ScaleRGB=true
   ) {module=ScaleColor001}
   &ScaleColor(
    Scale Alpha=$System.TODAlpha
    Scale RGB=&MakeVectorFromLinearColorRGB(Color=$System.TODColor)
    ScaleA=true
    ScaleRGB=true
   )
   /SetVariables(
    Particles.CenterFlareColor:LinearColor=&MultiplyLinearColor(
     Linear Color=$Particles.Color
     Scale Factor=&MakeVector4_FromLinearColor(LinearColor=&MaskLinearColor(
       Mask Source=1
       Mask Parameter=&MakeIntFromFloat(Float=$User.Variant)
       Mask=0
       Masked Linear Color=<rgba 8.0099 9.8007 10 1>
       Passthrough Linear Color=&MaskLinearColor(
        Mask Source=1
        Mask Parameter=&MakeIntFromFloat(Float=$User.Variant)
        Mask=1
        Masked Linear Color=<rgba 10 0 2.5293 1>
        Passthrough Linear Color=&MaskLinearColor(
         Mask Source=1
         Mask Parameter=&MakeIntFromFloat(Float=$User.Variant)
         Mask=2
         Masked Linear Color=<rgba 6.8097 6.5104 10 1>
         Passthrough Linear Color=&MaskLinearColor(
          Mask Source=1
          Mask Parameter=&MakeIntFromFloat(Float=$User.Variant)
          Mask=3
          Masked Linear Color=<rgba 10 9.3031 3.947 1>
          Passthrough Linear Color=&MaskLinearColor(
           Mask Source=1
           Mask Parameter=&MakeIntFromFloat(Float=$User.Variant)
           Mask=4
           Masked Linear Color=<rgba 8.0099 9.8007 10 1>
           Passthrough Linear Color=<rgba 8.0099 9.8007 10 1>
          )
         )
        )
       )
      ))
    )
    Particles.CenterFlareSize:Vector2f=&Multiply_Vector2(A=$Particles.SpriteSize, B=&Vector2DFromFloat(Value=0.5))
    Particles.CoronaFlareColor:LinearColor=&MultiplyLinearColor(
     Linear Color=$Particles.Color
     Scale Factor=&MakeVector4_FromLinearColor(LinearColor=<rgba 0.15 0.15 0.15 1>)
    )
    Particles.CoronaFlareSize:Vector2f=&Multiply_Vector2(A=$Particles.SpriteSize, B=&Vector2DFromFloat(Value=0.75))
    Particles.FlickerFlareColor:LinearColor=&MultiplyLinearColor(
     Linear Color=$Particles.Color
     Scale Factor=&MakeVector4_FromLinearColor(LinearColor=&MaskLinearColor(
       Mask Source=1
       Mask Parameter=&MakeIntFromFloat(Float=$User.Variant)
       Mask=0
       Masked Linear Color=<rgba 0.9441 0.8079 1 1>
       Passthrough Linear Color=&MaskLinearColor(
        Mask Source=1
        Mask Parameter=&MakeIntFromFloat(Float=$User.Variant)
        Mask=1
        Masked Linear Color=<rgba 0.3 0.0562 0.1112 50>
        Passthrough Linear Color=&MaskLinearColor(
         Mask Source=1
         Mask Parameter=&MakeIntFromFloat(Float=$User.Variant)
         Mask=2
         Masked Linear Color=<rgba 0.0314 0.0322 0.1 50>
         Passthrough Linear Color=&MaskLinearColor(
          Mask Source=1
          Mask Parameter=&MakeIntFromFloat(Float=$User.Variant)
          Mask=3
          Masked Linear Color=<rgba 0.05 0.0344 0.005 50>
          Passthrough Linear Color=&MaskLinearColor(
           Mask Source=1
           Mask Parameter=&MakeIntFromFloat(Float=$User.Variant)
           Mask=4
           Masked Linear Color=<rgba 0.5 0.5 0.5 1>
           Passthrough Linear Color=<rgba 0.404 0.4624 0.5 1>
          )
         )
        )
       )
      ))
    )
   ) {isSetParameters=true, module=SetVariables_#4}
   /SetVariables(Particles.SpriteSize=&Multiply_Vector2D(
     A=$Particles.SpriteSize
     B=&Multiply_Vector2D_ByFloat(
      Float=&MaskFloat(
       Mask Source=1
       Mask Parameter=&MakeIntFromFloat(Float=$User.Variant)
       Mask=0
       Masked Float=&MakeFloatFromLinearColor(LinearColor=$Particles.SampleTexture001.SampledColor, Channel:ENiagaraFloatFromLinearColorOptions=::NewEnumerator0)
       Passthrough Float=&MaskFloat(
        Mask Source=1
        Mask Parameter=&MakeIntFromFloat(Float=$User.Variant)
        Mask=1
        Masked Float=&MakeFloatFromLinearColor(LinearColor=$Particles.SampleTexture001.SampledColor, Channel:ENiagaraFloatFromLinearColorOptions=::NewEnumerator1)
        Passthrough Float=&MaskFloat(
         Mask Source=1
         Mask Parameter=&MakeIntFromFloat(Float=$User.Variant)
         Mask=2
         Masked Float=&MakeFloatFromLinearColor(LinearColor=$Particles.SampleTexture001.SampledColor, Channel:ENiagaraFloatFromLinearColorOptions=::NewEnumerator2)
         Passthrough Float=&MaskFloat(
          Mask Source=1
          Mask Parameter=&MakeIntFromFloat(Float=$User.Variant)
          Mask=3
          Masked Float=&MakeFloatFromLinearColor(LinearColor=$Particles.SampleTexture001.SampledColor, Channel:ENiagaraFloatFromLinearColorOptions=::NewEnumerator3)
          Passthrough Float=1
         )
        )
       )
      )
      Vector2D=<1 1>
     )
    )) {isSetParameters=true, module=SetVariables_#7}
  ]
  rendererClasses [&NiagaraSpriteRendererProperties]
  renderers [
   {rendererClass=&NiagaraSpriteRendererProperties}
   {rendererClass=&NiagaraSpriteRendererProperties, rendererIndex=1}
  ]
 }
 Sparkles_Low001 {
  emitterUpdateScript [
   &EmitterState(Inactive Response=::NewEnumerator1, Life Cycle Mode=::NewEnumerator1, Loop Duration=15, UseLoopDelay=false)
   &SpawnRate(SpawnRate=20, Use Spawn Probability=true)
  ]
  particleSpawnScript [
   &InitializeParticle(
    Lifetime Max=0.4
    Lifetime Min=0.6
    Lifetime Mode=::NewEnumerator1
    Sprite Size Mode=::NewEnumerator3
    Uniform Sprite Size=&RandomRangeFloat(Minimum=15, Maximum=25)
    UsePositionOffset=false
    Write Material Random=true
   )
   &CameraOffset(Camera Offset Amount=10)
   /SetVariables(Particles.BoneIndex:NiagaraInt32=&RandomRangeInt(Minimum=0, Maximum=8)) {isSetParameters=true, module=SetVariables_#8}
  ]
  particleUpdateScript [
   &ParticleState()
   &SkeletalMeshLocation(Bone / Socket Index=$Particles.BoneIndex, Bone Sampling Mode=::NewEnumerator4, MaskSpawnGroup=false, Skeletal Mesh=$System.Skeletal Mesh)
   &ShapeLocation(
    Additional Quaternion Rotation=true
    Additional Yaw / Pitch / Roll=&SelectVectorFromArray(
     Direct Array Index=$Particles.BoneIndex
     Vector Selection Array=DataInterface() {propertyValues {FloatData [<0 90 0>, <0 90 0>, <0 90 0>, <0 -70 0>, <0 110 0>, <0 85 0>, <0 -95 0>, <0 90 0>, <0 90 0>]}}
    )
    Cylinder Distribution=::NewEnumerator2
    Cylinder Height=&SelectFloatFromArray(Direct Array Index=$Particles.BoneIndex, Float Selection Array=DataInterface() {propertyValues {FloatData [25,25,30,15,15,30,30,22,22]}})
    Cylinder Height Midpoint:NiagaraFloat=&SelectFloatFromArray(Direct Array Index=$Particles.BoneIndex, Float Selection Array=DataInterface() {propertyValues {FloatData [1,0,0,0.5,0.5,0,0,1,0]}})
    Cylinder Radius=&SelectFloatFromArray(Direct Array Index=$Particles.BoneIndex, Float Selection Array=DataInterface() {propertyValues {FloatData [3,3,7.5,10,10,3.5,3.5,2.5,2.5]}})
    MaskSpawnGroup=false
    Non Uniform Scale=&SelectVectorFromArray(
     Direct Array Index=$Particles.BoneIndex
     Vector Selection Array=DataInterface() {propertyValues {FloatData [<1 1 1>, <1 1 1>, <1 1 1>, <0.6 1 1>, <0.6 1 1>, <1 1 1>, <1 1 1>]}}
    )
    Rotation Angle Type:ENiagara_AngleInput=::NewEnumerator0
    Rotation Mode=::NewEnumerator3
    Rotation Quaternion=$Particles.SkeletalMeshLocation.SampledRotation
    Shape Origin=&Add_Vector(
     A=Linked() {linkedVariable {name=Particles.SkeletalMeshLocation.SampledPosition, type=NiagaraPosition}}
     B=&SelectVectorFromArray(
      Direct Array Index=$Particles.BoneIndex
      Vector Selection Array=DataInterface() {propertyValues {FloatData [<0 0 0>, <0 0 0>, <0 0 0>, <0 0 0>, <0 0 0>, <2 -1 0>, <-2 -1 0>, <0 0 0>, <0 0 0>]}}
     )
    )
    Shape Primitive=::NewEnumerator1
    U Position=&Lerp_Float(
     A=&RandomRangeFloat(Minimum=0, Maximum=1)
     B=&RandomRangeFloat(Minimum=-0.3, Maximum=0.25)
     Alpha=&MakeCustomFloatFromBool(
      Bool=&SetBoolByFloatComparison(A=&MakeFloatFromInt(Int=$Particles.BoneIndex), B=2)
      True Float=&MakeCustomFloatFromBool(
       Bool=&SetBoolByFloatComparison(A=&MakeFloatFromInt(Int=$Particles.BoneIndex), B=5, Comparison Type=::NewEnumerator6)
       True Float=1
       False Float=0
      )
      False Float=0
     )
    )
    UseOwnerScale=true
    V Position=&RandomRangeFloat(Minimum=0, Maximum=1)
   )
   &KillParticles(Kill Particles=&MaskBool(
     Mask Source:NiagaraInt32=1
     Mask Parameter:NiagaraInt32=&MakeCustomIntFromBool(Bool=$User.bisHandHidden, True INT=1, False INT=0)
     Mask:NiagaraInt32=1
     Masked Bool:NiagaraBool=&SetBoolByFloatComparison(A=&MakeFloatFromInt(Int=$Particles.BoneIndex), B=6)
     Passthrough Bool:NiagaraBool=false
    ))
   &KillParticles(Kill Particles=&MaskBool(
     Mask Source:NiagaraInt32=1
     Mask Parameter:NiagaraInt32=&MakeCustomIntFromBool(Bool=$User.bisFootHidden, True INT=1, False INT=0)
     Mask:NiagaraInt32=1
     Masked Bool:NiagaraBool=&SetBoolByFloatComparison(A=&MakeFloatFromInt(Int=$Particles.BoneIndex), B=6, Comparison Type=::NewEnumerator4)
     Passthrough Bool:NiagaraBool=&MaskBool(
      Mask Source:NiagaraInt32=1
      Mask Parameter:NiagaraInt32=&MakeIntFromFloat(Float=$User.Variant)
      Mask:NiagaraInt32=3
      Masked Bool:NiagaraBool=&SetBoolByFloatComparison(A=&MakeFloatFromInt(Int=$Particles.BoneIndex), B=6, Comparison Type=::NewEnumerator4)
      Passthrough Bool:NiagaraBool=false
     )
    )) {module=KillParticles001}
   &KillParticles(Kill Particles=&MaskBool(
     Mask Source:NiagaraInt32=1
     Mask Parameter:NiagaraInt32=&MakeCustomIntFromBool(Bool=$User.bisFootHidden, True INT=1, False INT=0)
     Mask:NiagaraInt32=1
     Masked Bool:NiagaraBool=&SetBoolByFloatComparison(A=&MakeFloatFromInt(Int=$Particles.BoneIndex), B=5, Comparison Type=::NewEnumerator4)
     Passthrough Bool:NiagaraBool=&MaskBool(
      Mask Source:NiagaraInt32=1
      Mask Parameter:NiagaraInt32=&MakeIntFromFloat(Float=$User.Variant)
      Mask:NiagaraInt32=3
      Masked Bool:NiagaraBool=&SetBoolByFloatComparison(A=&MakeFloatFromInt(Int=$Particles.BoneIndex), B=5, Comparison Type=::NewEnumerator4)
      Passthrough Bool:NiagaraBool=false
     )
    )) {module=KillParticles002}
   &ScaleSpriteSize(Uniform Curve Scale=&FNTimeOfDayScalar(Night Scale=1.25, Morning Scale=1, Evening Scale=1)) {module=ScaleSpriteSize001}
   &ScaleColor(
    Scale Alpha=&Multiply_Float(
     A=&FloatFromCurve(
      FloatCurve=DataInterface() {propertyValues {Curve {keys [
          {interpMode=RCIM_Cubic, tangentMode=RCTM_User, time=0, value=0}
          {arriveTangent=-2.4058, interpMode=RCIM_Cubic, leaveTangent=-2.4058, tangentMode=RCTM_User, time=0.26, value=0.8421}
          {arriveTangent=-1.1775, interpMode=RCIM_Cubic, leaveTangent=-1.1775, time=0.4613, value=0.0645}
          {arriveTangent=-0.1197, interpMode=RCIM_Cubic, leaveTangent=-0.1197, time=0.5898, value=0.4537}
          <at 1 0 RCIM_Cubic>
         ]}}}
      Scale Curve=&RandomRangeFloat(Minimum=0.1, Maximum=0.5)
     )
     B=&FloatFromCurve(
      FloatCurve=DataInterface() {propertyValues {Curve {keys [{tangentMode=RCTM_User, time=0, value=0}, {tangentMode=RCTM_User, time=1, value=1}]}}}
      Scale Curve=&FloatFromCurve(
       FloatCurve=DataInterface() {propertyValues {Curve {keys [
           {tangentMode=RCTM_User, time=0, value=0}
           {arriveTangentWeight=0.1667, interpMode=RCIM_Cubic, leaveTangentWeight=0.1667, tangentMode=RCTM_User, time=1, value=1}
          ]}}}
       CurveIndex=$Emitter.Age
      )
     )
    )
    Scale RGB=&VectorFromFloat(Value=&RandomRangeFloat(Minimum=0.7, Maximum=0.9, Evaluation Type=::NewEnumerator1))
    ScaleA=true
    ScaleRGB=true
   ) {module=ScaleColor001}
   &ScaleColor(
    Scale Alpha=$System.TODAlpha
    Scale RGB=&MakeVectorFromLinearColorRGB(Color=$System.TODColor)
    ScaleA=true
    ScaleRGB=true
   )
   /SetVariables(Particles.EdgeAlpha:NiagaraFloat=&DotProduct(
     Vector A=&Multiply_VectorByFloat(Vector=$Particles.ShapeLocation.ShapeNormal, Float=1)
     Vector B=&Multiply_VectorByFloat(
      Vector=&ReturnCameraVector(
       Camera Vectors=::NewEnumerator1
       Coordinate Space=::Simulation
       CameraQuery=DataInterface() {propertyValues {bRequireCurrentFrameData=false}}
      )
      Float=-1
     )
    )) {isSetParameters=true}
   /SetVariables(
    Particles.CenterFlareColor:LinearColor=&Multiply_LinearColorByFloat(
     Linear Color=&MultiplyLinearColor(
      Linear Color=$Particles.Color
      Scale Factor=&MakeVector4_FromLinearColor(LinearColor=$System.SparklesCenterFlare)
     )
     Float=$Particles.EdgeAlpha
    )
    Particles.CenterFlareSize:Vector2f=&Multiply_Vector2(A=$Particles.SpriteSize, B=&Vector2DFromFloat(Value=0.5))
    Particles.CoronaFlareColor:LinearColor=&MultiplyLinearColor(
     Linear Color=$Particles.Color
     Scale Factor=&MakeVector4_FromLinearColor(LinearColor=<rgba 0.15 0.15 0.15 1>)
    )
    Particles.CoronaFlareSize:Vector2f=&Multiply_Vector2(A=$Particles.SpriteSize, B=&Vector2DFromFloat(Value=0.75))
    Particles.FlickerFlareColor:LinearColor=&Multiply_LinearColorByFloat(
     Linear Color=&MultiplyLinearColor(
      Linear Color=$Particles.Color
      Scale Factor=&MakeVector4_FromLinearColor(LinearColor=$System.SparklesFlickerFlare)
     )
     Float=$Particles.EdgeAlpha
    )
   ) {isSetParameters=true, module=SetVariables_#4}
  ]
  rendererClasses [&NiagaraSpriteRendererProperties]
  renderers [
   {rendererClass=&NiagaraSpriteRendererProperties}
   {rendererClass=&NiagaraSpriteRendererProperties, rendererIndex=1}
  ]
 }
}
summary {
 systemName=NS_F_MED_SaxoPop_Idle
 emitters [
  {
   bEnabled=false
   emitterName=Blings
   simTarget=GPUComputeSim
   rendererClasses [&NiagaraSpriteRendererProperties, &NiagaraMeshRendererProperties]
  }
  {
   bEnabled=false
   emitterName=Sparkles
   simTarget=GPUComputeSim
   rendererClasses [&NiagaraSpriteRendererProperties, &NiagaraMeshRendererProperties]
  }
  {emitterName=Blings_Low001, rendererClasses [&NiagaraSpriteRendererProperties]}
  {emitterName=Sparkles_Low001, rendererClasses [&NiagaraSpriteRendererProperties]}
  {bEnabled=false, emitterName=Blings_Low, rendererClasses [&NiagaraSpriteRendererProperties]}
  {bEnabled=false, emitterName=Sparkles_Low, rendererClasses [&NiagaraSpriteRendererProperties]}
 ]
 userVariables [
  User.bisFootHidden {type=NiagaraBool}
  User.bisHandHidden {type=NiagaraBool}
  User.Lifetime_Scale {type=NiagaraFloat, defaultValue=0, description=""}
  User.Variant {type=NiagaraFloat, defaultValue=0, description=""}
 ]
}
