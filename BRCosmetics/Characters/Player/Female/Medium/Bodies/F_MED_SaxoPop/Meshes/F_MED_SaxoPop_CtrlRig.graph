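# Control Rig for the SaxoPop female medium body that procedurally corrects deform bones for the jacket, pockets, collar, belt and feet.
# After the shared DeformRig function, a sequence runs weighted parent and rotation constraints plus thigh twist aims.
# Spherical pose readers on arms, feet, thighs and spine drive corrective offsets.
# Jacket flaps aim at opposite-named targets pulled by thigh readers, and remapped pocket-point distances add further flap bends.
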ControlRigBlueprint /BRCosmetics/Characters/Player/Female/Medium/Bodies/F_MED_SaxoPop/Meshes/F_MED_SaxoPop_CtrlRig
graphs [
 New Function_ContainedGraph {nodes [
   n1 = RigVMFunctionEntryNode()
   RigUnit_GetTransform = &RigUnit_GetTransform(bInitial=False, Item=@n1.bone 1)
   RigVMFunction_MathVectorDistance = &RigVMFunction_MathVectorDistance(
    A=@RigUnit_GetTransform.Transform.Translation
    B=&RigUnit_GetTransform(Item=(Type=Bone,Name="jacket_flap_C_l_pocket_pnt"), bInitial=False).Transform.Translation
   )
   Remap = &RigVMFunction_MathDoubleRemap(
    TargetMaximum=1
    TargetMinimum=0
    bClamp=False
    Value=@RigVMFunction_MathVectorDistance.Result
    SourceMinimum=RigVMVariableNode() {variable=min}.Value
    SourceMaximum=RigVMVariableNode() {variable=max}.Value
   )
   > RigVMFunctionReturnNode(value=@Remap.Result, distance=@RigVMFunction_MathVectorDistance.Result)
  ]}
 RigVMFunctionLibrary {nodes [Distance Between Bones = RigVMCollapseNode(ExecuteContext=(), bone 1=(Type=None,Name="None"), "bone 2 "=(Type=None,Name="None"))]}
 RigVMFunction_Sequence_1_ContainedGraph {nodes [
   n1 = RigVMFunctionEntryNode()
   RigVMFunction_Sequence = &RigVMFunction_Sequence(ExecuteContext=@n1.ExecuteContext)
   n2 = &RigVMFunction_Sequence(ExecuteContext=@RigVMFunction_Sequence.B)
   n3 = &RigVMFunction_Sequence(ExecuteContext=@n2.B)
   n4 = &RigVMFunction_Sequence(ExecuteContext=@n3.B)
   n5 = &RigVMFunction_Sequence(ExecuteContext=@n4.B)
   n6 = &RigVMFunction_Sequence(ExecuteContext=@n5.B)
   n7 = &RigVMFunction_Sequence(ExecuteContext=@n6.B)
   n8 = &RigVMFunction_Sequence(ExecuteContext=@n7.B)
   n9 = &RigVMFunction_Sequence(ExecuteContext=@n8.B)
   n10 = &RigVMFunction_Sequence(ExecuteContext=@n9.B)
   n11 = &RigVMFunction_Sequence(ExecuteContext=@n10.B)
   n12 = &RigVMFunction_Sequence(ExecuteContext=@n11.B)
   n13 = &RigVMFunction_Sequence(ExecuteContext=@n12.B)
   n14 = &RigVMFunction_Sequence(ExecuteContext=@n13.B)
   RigVMFunctionReturnNode(
    A=@RigVMFunction_Sequence.A
    B=@n2.A
    K=@n11.A
    L=@n12.A
    M=@n13.A
    N=@n14.A
    O=@n14.B
    C=@n3.A
    D=@n4.A
    E=@n5.A
    F=@n6.A
    G=@n7.A
    H=@n8.A
    I=@n9.A
    J=@n10.A
   )
  ]}
 RigVMModel {nodes [
   RigUnit_BeginExecution = &RigUnit_BeginExecution()
   FN_DeformRig_FNC = RigVMFunctionReferenceNode(
    DeformRig_ Curve="(deformRig_l_upperarm_up_alpha_72_#1=False,deformRig_r_upperarm_up_alpha_73_#2=False,deformRig_l_upperarm_fr_alpha_74_#3=False,deformRig_r_upperarm_fr_alpha_75_CB51B8 ...(truncated)"
    DeformRig_Bool="(upperarm_1_#4=True,elbow_3_#5=True,wrist_7_#6=True,groin_8_#7=True,glute_17_#8=True,knee_9_40C408324CA261 ...(truncated)"
    DeformRig_Debug=(upperarm_debug_18_#4=False,wrist_debug_20_#6=False,lat_debug_26_#9=False,groinGlute_debug_27_#7=False)
    DeformRig_Value="(upperarm_up_18_#10=(Rotation=(X=0,Y=0,Z=0,W=1),Translation=(X=0,Y=0,Z=8),Scale3D=(X=0.7,Y=1,Z=1)),upperarm_fwd_fr_19_#11=(Rotation=(X=0,Y=0,Z=0,W=1),Translation=(X=0,Y=10,Z=0),Scale3D=(X=0.5,Y=1,Z= ...(truncated)"
    ExecuteContext=@RigUnit_BeginExecution.ExecutePin
   )
   > RigVMFunction_Sequence = RigVMAggregateNode()
   &RigUnit_ParentConstraint(
    AdvancedSettings=(InterpolationType=Average,RotationOrderForFilter=XZY)
    Child=(Type=Bone,Name="pockets_r_deform_base")
    Filter=(TranslationFilter=(bX=True,bY=True,bZ=True),RotationFilter=(bX=True,bY=True,bZ=True),ScaleFilter=(bX=false,bY=false,bZ=false))
    Parents="((Item=(Type=Bone,Name=\"clavicle_r\"),Weight=0.059),(Item=(Type=Bone,Name=\"pelvis\"),Weight=0.009),(Item=(Type=Bone,Name=\"spine_01\"),Weight=0.033),(Item=(Type=Bone,Name=\"spine_02\"),Weight=0.166),(Item=(Type=Bone,Name=\"spine_03\"),Weight=0.166),(Item=(Type=Bon ...(truncated)"
    Weight=1
    ExecutePin=@RigVMFunction_Sequence.A
   )
   > &RigUnit_ParentConstraint(
    AdvancedSettings=(InterpolationType=Average,RotationOrderForFilter=XZY)
    Child=(Type=Bone,Name="pockets_l_deform_base")
    Filter=(TranslationFilter=(bX=True,bY=True,bZ=True),RotationFilter=(bX=True,bY=True,bZ=True),ScaleFilter=(bX=false,bY=false,bZ=false))
    Parents="((Item=(Type=Bone,Name=\"clavicle_l\"),Weight=0.059),(Item=(Type=Bone,Name=\"pelvis\"),Weight=0.009),(Item=(Type=Bone,Name=\"spine_01\"),Weight=0.033),(Item=(Type=Bone,Name=\"spine_02\"),Weight=0.166),(Item=(Type=Bone,Name=\"spine_03\"),Weight=0.166),(Item=(Type=Bon ...(truncated)"
    Weight=1
   )
   > n1 = &RigUnit_SphericalPoseReader(
    ActiveRegionScaleFactors=(PositiveWidth=1,NegativeWidth=1,PositiveHeight=1,NegativeHeight=1)
    ActiveRegionSize=0.15
    Debug=(bDrawDebug=false,bDraw2D=False,bDrawLocalAxes=False,DebugScale=15,DebugSegments=20,DebugThickness=0.1)
    DriverAxis=(X=-1,Y=0,Z=0)
    DriverItem=(Type=Bone,Name="upperarm_r")
    FalloffRegionScaleFactors=(PositiveWidth=1,NegativeWidth=1,PositiveHeight=1,NegativeHeight=1)
    FalloffSize=0.4
    FlipHeightScaling=False
    FlipWidthScaling=False
    OptionalParentItem=(Type=Bone,Name="spine_02")
    RotationOffset=(X=-90,Y=0,Z=0)
   )
   > &RigUnit_OffsetTransformForItem(
    Item=(Type=Bone,Name="pockets_r_deform")
    OffsetTransform=(Rotation=(X=0,Y=0,Z=0,W=1),Translation=(X=0,Y=1,Z=0),Scale3D=(X=1,Y=0.9,Z=1))
    bPropagateToChildren=True
    Weight=@n1.OutputParam
   )
   n2 = &RigUnit_SphericalPoseReader(
    ActiveRegionScaleFactors=(PositiveWidth=1,NegativeWidth=1,PositiveHeight=1,NegativeHeight=1)
    ActiveRegionSize=0.1
    Debug=(bDrawDebug=false,bDraw2D=False,bDrawLocalAxes=False,DebugScale=10,DebugSegments=20,DebugThickness=0.05)
    DriverAxis=(X=-1,Y=0,Z=0)
    DriverItem=(Type=Bone,Name="upperarm_r")
    FalloffRegionScaleFactors=(PositiveWidth=1,NegativeWidth=1,PositiveHeight=1,NegativeHeight=1)
    FalloffSize=0.5
    FlipHeightScaling=False
    FlipWidthScaling=False
    OptionalParentItem=(Type=Bone,Name="None")
    RotationOffset=(X=-120,Y=0,Z=20)
    ExecutePin=@RigVMFunction_Sequence.B
   )
   > &RigUnit_OffsetTransformForItem(
    Item=(Type=Bone,Name="MSTR_jacket_collar_dyn_r")
    OffsetTransform=(Rotation=(X=-0.0253,Y=-0.0068,Z=0.2587,W=0.9656),Translation=(X=0,Y=0,Z=0),Scale3D=(X=1,Y=1,Z=1))
    bPropagateToChildren=True
    Weight=@n2.OutputParam
   )
   > &RigUnit_OffsetTransformForItem(
    Item=(Type=Bone,Name="collar_for_r_deform")
    OffsetTransform=(Rotation=(X=-0.0961,Y=0.8423,Z=-0.0842,W=0.5237),Translation=(X=0,Y=0,Z=0),Scale3D=(X=2,Y=1,Z=1))
    bPropagateToChildren=True
    Weight=@n2.OutputParam
   )
   &RigUnit_ParentConstraint(
    AdvancedSettings=(InterpolationType=Average,RotationOrderForFilter=XZY)
    Child=(Type=Bone,Name="belt_deform_base")
    Parents=((Item=(Type=Bone,Name="pelvis"),Weight=0.199),(Item=(Type=Bone,Name="spine_01"),Weight=0.033),(Item=(Type=Bone,Name="spine_02"),Weight=0.43),(Item=(Type=Bone,Name="spine_03"),Weight=0.096),(Item=(Type=Bone,Name="spine_04"),Weight=0.242))
    Weight=1
    ExecutePin=@RigVMFunction_Sequence.D
   )
   > &RigUnit_ParentConstraint(
    AdvancedSettings=(InterpolationType=Average,RotationOrderForFilter=XZY)
    Child=(Type=Bone,Name="star_l_base")
    Parents=((Item=(Type=Bone,Name="spine_05"),Weight=0.17),(Item=(Type=Bone,Name="clavicle_l"),Weight=0.0037),(Item=(Type=Bone,Name="spine_04"),Weight=0.5982),(Item=(Type=Bone,Name="deform_pec_l"),Weight=0.2281))
    Weight=1
   )
   > &RigUnit_ParentConstraint(
    AdvancedSettings=(InterpolationType=Average,RotationOrderForFilter=XZY)
    Child=(Type=Bone,Name="star_r_base")
    Parents=((Item=(Type=Bone,Name="spine_05"),Weight=0.1637),(Item=(Type=Bone,Name="deform_pec_r"),Weight=0.1964),(Item=(Type=Bone,Name="spine_04"),Weight=0.6393),(Item=(Type=Bone,Name="clavicle_r"),Weight=0.001))
    Weight=1
   )
   &RigUnit_RotationConstraintLocalSpaceOffset(
    AdvancedSettings=(InterpolationType=Average,RotationOrderForFilter=XZY)
    Child=(Type=Bone,Name="foot_r_deform_base")
    Filter=(bX=True,bY=True,bZ=True)
    Parents=((Item=(Type=Bone,Name="calf_twist_01_r"),Weight=0.3),(Item=(Type=Bone,Name="foot_r"),Weight=2))
    Weight=1
    ExecutePin=@RigVMFunction_Sequence.E
   )
   > n3 = &RigUnit_SphericalPoseReader(
    ActiveRegionScaleFactors=(PositiveWidth=1,NegativeWidth=1,PositiveHeight=1,NegativeHeight=1)
    ActiveRegionSize=0.08
    Debug=(bDrawDebug=false,bDraw2D=False,bDrawLocalAxes=False,DebugScale=15,DebugSegments=20,DebugThickness=0.1)
    DriverAxis=(X=1,Y=0,Z=0)
    DriverItem=(Type=Bone,Name="foot_r")
    FalloffRegionScaleFactors=(PositiveWidth=1,NegativeWidth=1,PositiveHeight=1,NegativeHeight=1)
    FalloffSize=0.5
    FlipHeightScaling=False
    FlipWidthScaling=False
    OptionalParentItem=(Type=Bone,Name="calf_twist_01_r")
    RotationOffset=(X=90,Y=0,Z=0)
   )
   > &RigUnit_OffsetTransformForItem(
    Item=(Type=Bone,Name="foot_r_deform_back")
    OffsetTransform=(Rotation=(X=0,Y=0,Z=0,W=1),Translation=(X=9,Y=0,Z=0),Scale3D=(X=1,Y=1,Z=1))
    bPropagateToChildren=True
    Weight=@n3.OutputParam
   )
   > &RigUnit_OffsetTransformForItem(
    Item=(Type=Bone,Name="foot_r_deform_for")
    OffsetTransform=(Rotation=(X=0,Y=0,Z=0,W=1),Translation=(X=-6,Y=0,Z=0),Scale3D=(X=1,Y=1,Z=1))
    bPropagateToChildren=True
    Weight=@n3.OutputParam
   )
   > n4 = &RigUnit_SphericalPoseReader(
    ActiveRegionScaleFactors=(PositiveWidth=1,NegativeWidth=1,PositiveHeight=1,NegativeHeight=1)
    ActiveRegionSize=0.26
    Debug=(bDrawDebug=false,bDraw2D=False,bDrawLocalAxes=False,DebugScale=25,DebugSegments=20,DebugThickness=0.25)
    DriverAxis=(X=1,Y=0,Z=0)
    DriverItem=(Type=Bone,Name="foot_r")
    FalloffRegionScaleFactors=(PositiveWidth=1,NegativeWidth=1,PositiveHeight=1,NegativeHeight=1)
    FalloffSize=0.5
    FlipHeightScaling=False
    FlipWidthScaling=False
    OptionalParentItem=(Type=Bone,Name="None")
    RotationOffset=(X=-90,Y=0,Z=0)
   )
   > &RigUnit_OffsetTransformForItem(
    Item=(Type=Bone,Name="foot_r_deform_back")
    OffsetTransform=(Rotation=(X=0,Y=0,Z=0,W=1),Translation=(X=-4,Y=0,Z=0),Scale3D=(X=1,Y=1,Z=1))
    bPropagateToChildren=True
    Weight=@n4.OutputParam
   )
   > &RigUnit_OffsetTransformForItem(
    Item=(Type=Bone,Name="foot_r_deform_for")
    OffsetTransform=(Rotation=(X=0,Y=0,Z=0,W=1),Translation=(X=0,Y=0,Z=0),Scale3D=(X=1,Y=1,Z=1))
    bPropagateToChildren=True
    Weight=@n4.OutputParam
   )
   &RigUnit_RotationConstraintLocalSpaceOffset(
    AdvancedSettings=(InterpolationType=Average,RotationOrderForFilter=XZY)
    Child=(Type=Bone,Name="foot_l_deform_base")
    Filter=(bX=true,bY=true,bZ=true)
    Parents=((Item=(Type=Bone,Name="calf_twist_01_l"),Weight=0.3),(Item=(Type=Bone,Name="foot_l"),Weight=2))
    Weight=1
    ExecutePin=@RigVMFunction_Sequence.F
   )
   > n5 = &RigUnit_SphericalPoseReader(
    ActiveRegionScaleFactors=(PositiveWidth=1,NegativeWidth=1,PositiveHeight=1,NegativeHeight=1)
    ActiveRegionSize=0.08
    Debug=(bDrawDebug=false,bDraw2D=False,bDrawLocalAxes=False,DebugScale=25,DebugSegments=20,DebugThickness=0.25)
    DriverAxis=(X=-1,Y=0,Z=0)
    DriverItem=(Type=Bone,Name="foot_l")
    FalloffRegionScaleFactors=(PositiveWidth=1,NegativeWidth=1,PositiveHeight=1,NegativeHeight=1)
    FalloffSize=0.5
    FlipHeightScaling=False
    FlipWidthScaling=False
    OptionalParentItem=(Type=Bone,Name="calf_twist_01_l")
    RotationOffset=(X=-90,Y=0,Z=0)
   )
   > &RigUnit_OffsetTransformForItem(
    Item=(Type=Bone,Name="foot_l_deform_back")
    OffsetTransform=(Rotation=(X=0,Y=0,Z=0,W=1),Translation=(X=9,Y=0,Z=0),Scale3D=(X=1,Y=1,Z=1))
    bPropagateToChildren=True
    Weight=@n5.OutputParam
   )
   > &RigUnit_OffsetTransformForItem(
    Item=(Type=Bone,Name="foot_l_deform_for")
    OffsetTransform=(Rotation=(X=0,Y=0,Z=0,W=1),Translation=(X=-6,Y=0,Z=0),Scale3D=(X=1,Y=1,Z=1))
    bPropagateToChildren=True
    Weight=@n5.OutputParam
   )
   > n6 = &RigUnit_SphericalPoseReader(
    ActiveRegionScaleFactors=(PositiveWidth=1,NegativeWidth=1,PositiveHeight=1,NegativeHeight=1)
    ActiveRegionSize=0.26
    Debug=(bDrawDebug=false,bDraw2D=False,bDrawLocalAxes=False,DebugScale=25,DebugSegments=20,DebugThickness=0.25)
    DriverAxis=(X=-1,Y=0,Z=0)
    DriverItem=(Type=Bone,Name="foot_l")
    FalloffRegionScaleFactors=(PositiveWidth=1,NegativeWidth=1,PositiveHeight=1,NegativeHeight=1)
    FalloffSize=0.5
    FlipHeightScaling=False
    FlipWidthScaling=False
    OptionalParentItem=(Type=Bone,Name="None")
    RotationOffset=(X=90,Y=0,Z=0)
   )
   > &RigUnit_OffsetTransformForItem(
    Item=(Type=Bone,Name="foot_l_deform_back")
    OffsetTransform=(Rotation=(X=0,Y=0,Z=0,W=1),Translation=(X=-4,Y=0,Z=0),Scale3D=(X=1,Y=1,Z=1))
    bPropagateToChildren=True
    Weight=@n6.OutputParam
   )
   > &RigUnit_OffsetTransformForItem(
    Item=(Type=Bone,Name="foot_l_deform_for")
    OffsetTransform=(Rotation=(X=0,Y=0,Z=0,W=1),Translation=(X=0,Y=0,Z=0),Scale3D=(X=1,Y=1,Z=1))
    bPropagateToChildren=True
    Weight=@n6.OutputParam
   )
   &RigUnit_ParentConstraint(
    AdvancedSettings=(InterpolationType=Average,RotationOrderForFilter=XZY)
    Child=(Type=Bone,Name="MSTR_jacket_flap_A_l")
    Filter=(TranslationFilter=(bX=True,bY=True,bZ=True),RotationFilter=(bX=false,bY=false,bZ=false),ScaleFilter=(bX=True,bY=True,bZ=True))
    Parents=((Item=(Type=Bone,Name="spine_02"),Weight=0.3696),(Item=(Type=Bone,Name="spine_03"),Weight=0.0418),(Item=(Type=Bone,Name="spine_04"),Weight=0.0842),(Item=(Type=Bone,Name="pelvis"),Weight=0.4862),(Item=(Type=Bone,Name="spine_01"),Weight=0.0182))
    Weight=1
    ExecutePin=@RigVMFunction_Sequence.G
   )
   > &RigUnit_ParentConstraint(
    AdvancedSettings=(InterpolationType=Average,RotationOrderForFilter=XZY)
    Child=(Type=Bone,Name="MSTR_jacket_flap_B_l")
    Filter=(TranslationFilter=(bX=True,bY=True,bZ=True),RotationFilter=(bX=false,bY=false,bZ=false),ScaleFilter=(bX=True,bY=True,bZ=True))
    Parents=((Item=(Type=Bone,Name="spine_02"),Weight=0.3993),(Item=(Type=Bone,Name="spine_03"),Weight=0.0479),(Item=(Type=Bone,Name="spine_04"),Weight=0.1016),(Item=(Type=Bone,Name="pelvis"),Weight=0.4266),(Item=(Type=Bone,Name="spine_01"),Weight=0.0246))
    Weight=1
   )
   > &RigUnit_ParentConstraint(
    AdvancedSettings=(InterpolationType=Average,RotationOrderForFilter=XZY)
    Child=(Type=Bone,Name="MSTR_jacket_flap_C_l")
    Filter=(TranslationFilter=(bX=True,bY=True,bZ=True),RotationFilter=(bX=false,bY=false,bZ=false),ScaleFilter=(bX=True,bY=True,bZ=True))
    Parents=((Item=(Type=Bone,Name="spine_02"),Weight=0.459),(Item=(Type=Bone,Name="spine_03"),Weight=0.0498),(Item=(Type=Bone,Name="spine_04"),Weight=0.0953),(Item=(Type=Bone,Name="pelvis"),Weight=0.3575),(Item=(Type=Bone,Name="spine_01"),Weight=0.0384))
    Weight=1
   )
   > &RigUnit_ParentConstraint(
    AdvancedSettings=(InterpolationType=Average,RotationOrderForFilter=XZY)
    Child=(Type=Bone,Name="MSTR_jacket_flap_A_r")
    Filter=(TranslationFilter=(bX=True,bY=True,bZ=True),RotationFilter=(bX=false,bY=false,bZ=false),ScaleFilter=(bX=True,bY=True,bZ=True))
    Parents=((Item=(Type=Bone,Name="spine_02"),Weight=0.3696),(Item=(Type=Bone,Name="spine_03"),Weight=0.0418),(Item=(Type=Bone,Name="spine_04"),Weight=0.0842),(Item=(Type=Bone,Name="pelvis"),Weight=0.4862),(Item=(Type=Bone,Name="spine_01"),Weight=0.0182))
    Weight=1
   )
   > &RigUnit_ParentConstraint(
    AdvancedSettings=(InterpolationType=Average,RotationOrderForFilter=XZY)
    Child=(Type=Bone,Name="MSTR_jacket_flap_B_r")
    Filter=(TranslationFilter=(bX=True,bY=True,bZ=True),RotationFilter=(bX=false,bY=false,bZ=false),ScaleFilter=(bX=True,bY=True,bZ=True))
    Parents=((Item=(Type=Bone,Name="spine_02"),Weight=0.3748),(Item=(Type=Bone,Name="spine_03"),Weight=0.0384),(Item=(Type=Bone,Name="spine_04"),Weight=0.078),(Item=(Type=Bone,Name="pelvis"),Weight=0.4879),(Item=(Type=Bone,Name="spine_01"),Weight=0.0209))
    Weight=1
   )
   > &RigUnit_ParentConstraint(
    AdvancedSettings=(InterpolationType=Average,RotationOrderForFilter=XZY)
    Child=(Type=Bone,Name="MSTR_jacket_flap_C_r")
    Filter=(TranslationFilter=(bX=True,bY=True,bZ=True),RotationFilter=(bX=false,bY=false,bZ=false),ScaleFilter=(bX=True,bY=True,bZ=True))
    Parents=((Item=(Type=Bone,Name="spine_02"),Weight=0.4459),(Item=(Type=Bone,Name="spine_03"),Weight=0.0417),(Item=(Type=Bone,Name="spine_04"),Weight=0.0771),(Item=(Type=Bone,Name="pelvis"),Weight=0.4011),(Item=(Type=Bone,Name="spine_01"),Weight=0.0342))
    Weight=1
   )
   &RigUnit_AimConstraintLocalSpaceOffset(
    AimAxis=(X=-1,Y=0,Z=0)
    Child=(Type=Bone,Name="thigh_l_twist")
    Filter=(bX=True,bY=True,bZ=True)
    Parents=((Item=(Type=Bone,Name="calf_l"),Weight=1))
    UpAxis=(X=0,Y=1,Z=0)
    Weight=1
    WorldUp=(Target=(X=0,Y=1,Z=0),Kind=Location,Space=(Type=Bone,Name="thigh_l_aim_up_pnt"))
    ExecutePin=@RigVMFunction_Sequence.H
   )
   > &RigUnit_AimConstraintLocalSpaceOffset(
    Child=(Type=Bone,Name="thigh_r_twist")
    Filter=(bX=True,bY=True,bZ=True)
    Parents=((Item=(Type=Bone,Name="calf_r"),Weight=1))
    UpAxis=(X=0,Y=-1,Z=0)
    Weight=1
    WorldUp=(Target=(X=0,Y=-1,Z=0),Kind=Location,Space=(Type=Bone,Name="thigh_r_aim_up_pnt"))
   )
   n7 = &RigUnit_SphericalPoseReader(
    ActiveRegionScaleFactors=(PositiveWidth=3,NegativeWidth=3,PositiveHeight=3,NegativeHeight=3)
    ActiveRegionSize=0.1
    Debug=(bDrawDebug=false,bDraw2D=False,bDrawLocalAxes=false,DebugScale=5,DebugSegments=20,DebugThickness=0.05)
    DriverAxis=(X=-1,Y=0,Z=0)
    DriverItem=(Type=Bone,Name="thigh_l")
    FalloffRegionScaleFactors=(PositiveWidth=0.5,NegativeWidth=1,PositiveHeight=0.5,NegativeHeight=0.5)
    FalloffSize=0.5
    FlipHeightScaling=False
    FlipWidthScaling=False
    OptionalParentItem=(Type=Bone,Name="None")
    RotationOffset=(X=90,Y=0,Z=140)
    ExecutePin=@RigVMFunction_Sequence.I
   )
   > &RigUnit_ParentConstraint(
    AdvancedSettings=(InterpolationType=Average,RotationOrderForFilter=XZY)
    Child=(Type=Bone,Name="jacket_flap_A_r_aim")
    Filter=(TranslationFilter=(bX=True,bY=True,bZ=True),RotationFilter=(bX=false,bY=false,bZ=false),ScaleFilter=(bX=false,bY=false,bZ=false))
    Parents=((Item=(Type=Bone,Name="pelvis"),Weight=1))
    Weight=@n7.OutputParam
   )
   > &RigUnit_ParentConstraint(
    AdvancedSettings=(InterpolationType=Average,RotationOrderForFilter=XZY)
    Child=(Type=Bone,Name="jacket_flap_B_r_aim")
    Filter=(TranslationFilter=(bX=True,bY=True,bZ=True),RotationFilter=(bX=false,bY=false,bZ=false),ScaleFilter=(bX=false,bY=false,bZ=false))
    Parents=((Item=(Type=Bone,Name="pelvis"),Weight=1))
    Weight=@n7.OutputParam
   )
   > &RigUnit_ParentConstraint(
    AdvancedSettings=(InterpolationType=Average,RotationOrderForFilter=XZY)
    Child=(Type=Bone,Name="jacket_flap_C_r_aim")
    Filter=(TranslationFilter=(bX=True,bY=True,bZ=True),RotationFilter=(bX=false,bY=false,bZ=false),ScaleFilter=(bX=false,bY=false,bZ=false))
    Parents=((Item=(Type=Bone,Name="pelvis"),Weight=1))
    Weight=@n7.OutputParam
   )
   > n8 = &RigUnit_SphericalPoseReader(
    ActiveRegionScaleFactors=(PositiveWidth=3,NegativeWidth=3,PositiveHeight=3,NegativeHeight=3)
    ActiveRegionSize=0.1
    Debug=(bDrawDebug=false,bDraw2D=False,bDrawLocalAxes=false,DebugScale=5,DebugSegments=20,DebugThickness=0.05)
    DriverAxis=(X=1,Y=0,Z=0)
    DriverItem=(Type=Bone,Name="thigh_r")
    FalloffRegionScaleFactors=(PositiveWidth=0.5,NegativeWidth=1,PositiveHeight=0.5,NegativeHeight=0.5)
    FalloffSize=0.5
    FlipHeightScaling=False
    FlipWidthScaling=False
    OptionalParentItem=(Type=Bone,Name="None")
    RotationOffset=(X=90,Y=0,Z=-40)
   )
   > &RigUnit_ParentConstraint(
    AdvancedSettings=(InterpolationType=Average,RotationOrderForFilter=XZY)
    Child=(Type=Bone,Name="jacket_flap_A_l_aim")
    Filter=(TranslationFilter=(bX=True,bY=True,bZ=True),RotationFilter=(bX=false,bY=false,bZ=false),ScaleFilter=(bX=false,bY=false,bZ=false))
    Parents=((Item=(Type=Bone,Name="pelvis"),Weight=1))
    Weight=@n8.OutputParam
   )
   > &RigUnit_ParentConstraint(
    AdvancedSettings=(InterpolationType=Average,RotationOrderForFilter=XZY)
    Child=(Type=Bone,Name="jacket_flap_B_l_aim")
    Filter=(TranslationFilter=(bX=True,bY=True,bZ=True),RotationFilter=(bX=false,bY=false,bZ=false),ScaleFilter=(bX=false,bY=false,bZ=false))
    Parents=((Item=(Type=Bone,Name="pelvis"),Weight=1))
    Weight=@n8.OutputParam
   )
   > &RigUnit_ParentConstraint(
    AdvancedSettings=(InterpolationType=Average,RotationOrderForFilter=XZY)
    Child=(Type=Bone,Name="jacket_flap_C_l_aim")
    Filter=(TranslationFilter=(bX=True,bY=True,bZ=True),RotationFilter=(bX=false,bY=false,bZ=false),ScaleFilter=(bX=false,bY=false,bZ=false))
    Parents=((Item=(Type=Bone,Name="pelvis"),Weight=1))
    Weight=@n8.OutputParam
   )
   n9 = &RigUnit_SphericalPoseReader(
    ActiveRegionScaleFactors=(PositiveWidth=1,NegativeWidth=1,PositiveHeight=2,NegativeHeight=3)
    ActiveRegionSize=0.1
    Debug=(bDrawDebug=false,bDraw2D=False,bDrawLocalAxes=false,DebugScale=15,DebugSegments=20,DebugThickness=0.05)
    DriverAxis=(X=-1,Y=0,Z=0)
    DriverItem=(Type=Bone,Name="thigh_l")
    FalloffRegionScaleFactors=(PositiveWidth=1,NegativeWidth=1,PositiveHeight=1,NegativeHeight=1)
    FalloffSize=0.55
    FlipHeightScaling=False
    FlipWidthScaling=False
    OptionalParentItem=(Type=Bone,Name="None")
    RotationOffset=(X=120,Y=10,Z=-10)
    ExecutePin=@RigVMFunction_Sequence.J
   )
   > &RigUnit_OffsetTransformForItem(
    Item=(Type=Bone,Name="jacket_flap_A_r_aim")
    OffsetTransform=(Rotation=(X=0,Y=0,Z=0,W=1),Translation=(X=0,Y=10,Z=-10),Scale3D=(X=1,Y=1,Z=1))
    bPropagateToChildren=True
    Weight=@n9.OutputParam
   )
   > &RigUnit_OffsetTransformForItem(
    Item=(Type=Bone,Name="jacket_flap_B_r_aim")
    OffsetTransform=(Rotation=(X=0,Y=0,Z=0,W=1),Translation=(X=0,Y=-3,Z=1),Scale3D=(X=1,Y=1,Z=1))
    bPropagateToChildren=True
    Weight=@n9.OutputParam
   )
   > &RigUnit_OffsetTransformForItem(
    Item=(Type=Bone,Name="jacket_flap_C_r_aim")
    OffsetTransform=(Rotation=(X=0,Y=0,Z=0,W=1),Translation=(X=0,Y=-8,Z=-3),Scale3D=(X=1,Y=1,Z=1))
    bPropagateToChildren=True
    Weight=@n9.OutputParam
   )
   > n10 = &RigUnit_SphericalPoseReader(
    ActiveRegionScaleFactors=(PositiveWidth=2,NegativeWidth=2,PositiveHeight=2,NegativeHeight=3)
    ActiveRegionSize=0.1
    Debug=(bDrawDebug=false,bDraw2D=False,bDrawLocalAxes=false,DebugScale=15,DebugSegments=20,DebugThickness=0.05)
    DriverAxis=(X=1,Y=0,Z=0)
    DriverItem=(Type=Bone,Name="thigh_r")
    FalloffRegionScaleFactors=(PositiveWidth=1,NegativeWidth=1,PositiveHeight=1,NegativeHeight=1)
    FalloffSize=0.55
    FlipHeightScaling=False
    FlipWidthScaling=False
    OptionalParentItem=(Type=Bone,Name="None")
    RotationOffset=(X=-60,Y=10,Z=-10)
   )
   > &RigUnit_OffsetTransformForItem(
    Item=(Type=Bone,Name="jacket_flap_A_l_aim")
    OffsetTransform=(Rotation=(X=0,Y=0,Z=0,W=1),Translation=(X=0,Y=8,Z=-5),Scale3D=(X=1,Y=1,Z=1))
    bPropagateToChildren=True
    Weight=@n10.OutputParam
   )
   > &RigUnit_OffsetTransformForItem(
    Item=(Type=Bone,Name="jacket_flap_B_l_aim")
    OffsetTransform=(Rotation=(X=0,Y=0,Z=0,W=1),Translation=(X=0,Y=6,Z=-8),Scale3D=(X=1,Y=1,Z=1))
    bPropagateToChildren=True
    Weight=@n10.OutputParam
   )
   > &RigUnit_OffsetTransformForItem(
    Item=(Type=Bone,Name="jacket_flap_C_l_aim")
    OffsetTransform=(Rotation=(X=0,Y=0,Z=0,W=1),Translation=(X=0,Y=8,Z=-3),Scale3D=(X=1,Y=1,Z=1))
    bPropagateToChildren=True
    Weight=@n10.OutputParam
   )
   n11 = &RigUnit_SphericalPoseReader(
    ActiveRegionScaleFactors=(PositiveWidth=0.5,NegativeWidth=0.5,PositiveHeight=2,NegativeHeight=2)
    ActiveRegionSize=0.1
    Debug=(bDrawDebug=false,bDraw2D=False,bDrawLocalAxes=false,DebugScale=15,DebugSegments=20,DebugThickness=0.05)
    DriverAxis=(X=-1,Y=0,Z=0)
    DriverItem=(Type=Bone,Name="thigh_l")
    FalloffRegionScaleFactors=(PositiveWidth=0.4,NegativeWidth=1,PositiveHeight=1,NegativeHeight=1)
    FalloffSize=0.55
    FlipHeightScaling=False
    FlipWidthScaling=False
    OptionalParentItem=(Type=Bone,Name="None")
    RotationOffset=(X=120,Y=0,Z=0)
    ExecutePin=@RigVMFunction_Sequence.L
   )
   > &RigUnit_OffsetTransformForItem(
    Item=(Type=Bone,Name="jacket_flap_A_l_1")
    OffsetTransform=(Rotation=(X=0,Y=0.2164,Z=0,W=0.9763),Translation=(X=0,Y=0,Z=0),Scale3D=(X=0.75,Y=1,Z=1))
    bPropagateToChildren=True
    Weight=@n11.OutputParam
   )
   > &RigUnit_OffsetTransformForItem(
    Item=(Type=Bone,Name="jacket_flap_A_l_2")
    OffsetTransform=(Rotation=(X=-0.2362,Y=-0.21,Z=-0.0524,W=0.9473),Translation=(X=0,Y=0,Z=0),Scale3D=(X=1,Y=1,Z=1))
    bPropagateToChildren=True
    Weight=@n11.OutputParam
   )
   > &RigUnit_OffsetTransformForItem(
    Item=(Type=Bone,Name="jacket_flap_B_l_1")
    OffsetTransform=(Rotation=(X=0.0099,Y=0.1128,Z=-0.0866,W=0.9898),Translation=(X=0,Y=0,Z=0),Scale3D=(X=0.75,Y=1,Z=1))
    bPropagateToChildren=True
    Weight=@n11.OutputParam
   )
   > &RigUnit_OffsetTransformForItem(
    Item=(Type=Bone,Name="jacket_flap_B_l_2")
    OffsetTransform=(Rotation=(X=-0.3254,Y=-0.0114,Z=0.033,W=0.9449),Translation=(X=0,Y=0,Z=0),Scale3D=(X=1,Y=1,Z=1))
    bPropagateToChildren=True
    Weight=@n11.OutputParam
   )
   > &RigUnit_OffsetTransformForItem(
    Item=(Type=Bone,Name="jacket_flap_B_l_3")
    OffsetTransform=(Rotation=(X=0.0057,Y=-0.1304,Z=0.0432,W=0.9905),Translation=(X=0,Y=0,Z=0),Scale3D=(X=1,Y=1,Z=1))
    bPropagateToChildren=True
    Weight=@n11.OutputParam
   )
   > &RigUnit_OffsetTransformForItem(
    Item=(Type=Bone,Name="jacket_flap_C_l_1")
    OffsetTransform=(Rotation=(X=0.0057,Y=0.0432,Z=-0.1304,W=0.9905),Translation=(X=0,Y=0,Z=0),Scale3D=(X=1,Y=1,Z=1))
    bPropagateToChildren=True
    Weight=@n11.OutputParam
   )
   > &RigUnit_OffsetTransformForItem(
    Item=(Type=Bone,Name="jacket_flap_C_l_2")
    OffsetTransform=(Rotation=(X=0.0905,Y=0.083,Z=-0.0509,W=0.9911),Translation=(X=0,Y=0,Z=0),Scale3D=(X=0.8,Y=1,Z=1))
    bPropagateToChildren=True
    Weight=@n11.OutputParam
   )
   > &RigUnit_OffsetTransformForItem(
    Item=(Type=Bone,Name="jacket_flap_C_l_3")
    OffsetTransform=(Rotation=(X=0,Y=-0.0872,Z=0,W=0.9962),Translation=(X=0,Y=0,Z=0),Scale3D=(X=1,Y=1,Z=1))
    bPropagateToChildren=True
    Weight=@n11.OutputParam
   )
   > &RigUnit_OffsetTransformForItem(
    Item=(Type=Bone,Name="jacket_flap_C_l_4")
    OffsetTransform=(Rotation=(X=0,Y=0.2164,Z=0,W=0.9763),Translation=(X=0,Y=0,Z=0),Scale3D=(X=1,Y=1,Z=1))
    bPropagateToChildren=True
    Weight=@n11.OutputParam
   )
   > n12 = &RigUnit_SphericalPoseReader(
    ActiveRegionScaleFactors=(PositiveWidth=1,NegativeWidth=1,PositiveHeight=1,NegativeHeight=1)
    ActiveRegionSize=0.08
    Debug=(bDrawDebug=false,bDraw2D=False,bDrawLocalAxes=False,DebugScale=15,DebugSegments=20,DebugThickness=0.1)
    DriverAxis=(X=-1,Y=0,Z=0)
    DriverItem=(Type=Bone,Name="thigh_l")
    FalloffRegionScaleFactors=(PositiveWidth=1,NegativeWidth=1,PositiveHeight=1,NegativeHeight=1)
    FalloffSize=0.3
    FlipHeightScaling=False
    FlipWidthScaling=False
    OptionalParentItem=(Type=Bone,Name="None")
    RotationOffset=(X=-20,Y=65,Z=0)
   )
   > &RigUnit_OffsetTransformForItem(
    Item=(Type=Bone,Name="jacket_flap_B_l_1")
    OffsetTransform=(Rotation=(X=0,Y=-0.1305,Z=0,W=0.9914),Translation=(X=0,Y=0,Z=0),Scale3D=(X=1,Y=1,Z=1))
    bPropagateToChildren=True
    Weight=@n12.OutputParam
   )
   > &RigUnit_OffsetTransformForItem(
    Item=(Type=Bone,Name="MSTR_jacket_flap_C_l")
    OffsetTransform=(Rotation=(X=0,Y=-0.0872,Z=0,W=0.9962),Translation=(X=0,Y=0,Z=0),Scale3D=(X=1,Y=1,Z=1))
    bPropagateToChildren=True
    Weight=@n12.OutputParam
   )
   > &RigUnit_OffsetTransformForItem(
    Item=(Type=Bone,Name="jacket_flap_A_l_1")
    OffsetTransform=(Rotation=(X=0,Y=-0.0436,Z=0,W=0.999),Translation=(X=0,Y=0,Z=0),Scale3D=(X=1,Y=1,Z=1))
    bPropagateToChildren=True
    Weight=@n12.OutputParam
   )
   > n13 = &RigUnit_SphericalPoseReader(
    ActiveRegionScaleFactors=(PositiveWidth=1,NegativeWidth=1,PositiveHeight=1,NegativeHeight=1)
    ActiveRegionSize=0.02
    Debug=(bDrawDebug=false,bDraw2D=False,bDrawLocalAxes=False,DebugScale=15,DebugSegments=20,DebugThickness=0.1)
    DriverAxis=(X=1,Y=0,Z=0)
    DriverItem=(Type=Bone,Name="thigh_r")
    FalloffRegionScaleFactors=(PositiveWidth=1,NegativeWidth=1,PositiveHeight=1,NegativeHeight=1)
    FalloffSize=0.16
    FlipHeightScaling=False
    FlipWidthScaling=False
    OptionalParentItem=(Type=Bone,Name="None")
    RotationOffset=(X=-75,Y=0,Z=20)
   )
   > &RigUnit_OffsetTransformForItem(
    Item=(Type=Bone,Name="MSTR_jacket_flap_B_r")
    OffsetTransform=(Rotation=(X=0,Y=0.1478,Z=0,W=0.989),Translation=(X=0,Y=0,Z=0),Scale3D=(X=1,Y=1,Z=1))
    bPropagateToChildren=True
    Weight=@n13.OutputParam
   )
   > &RigUnit_OffsetTransformForItem(
    Item=(Type=Bone,Name="jacket_flap_B_r_3")
    OffsetTransform=(Rotation=(X=0,Y=-0.1305,Z=0,W=0.9914),Translation=(X=0,Y=0,Z=0),Scale3D=(X=1,Y=1,Z=1))
    bPropagateToChildren=True
    Weight=@n13.OutputParam
   )
   > &RigUnit_OffsetTransformForItem(
    Item=(Type=Bone,Name="jacket_flap_A_r_1")
    OffsetTransform=(Rotation=(X=0.0868,Y=0.0868,Z=-0.0076,W=0.9924),Translation=(X=0,Y=0,Z=0),Scale3D=(X=1,Y=1,Z=1))
    bPropagateToChildren=True
    Weight=@n13.OutputParam
   )
   > n14 = &RigUnit_SphericalPoseReader(
    ActiveRegionScaleFactors=(PositiveWidth=1,NegativeWidth=1.2,PositiveHeight=1,NegativeHeight=1)
    ActiveRegionSize=0.15
    Debug=(bDrawDebug=true,bDraw2D=False,bDrawLocalAxes=False,DebugScale=15,DebugSegments=20,DebugThickness=0.1)
    DriverAxis=(X=1,Y=0,Z=0)
    DriverItem=(Type=Bone,Name="thigh_r_twist")
    FalloffRegionScaleFactors=(PositiveWidth=1,NegativeWidth=1,PositiveHeight=1,NegativeHeight=1)
    FalloffSize=0.28
    FlipHeightScaling=False
    FlipWidthScaling=False
    OptionalParentItem=(Type=Bone,Name="spine_01")
    RotationOffset=(X=-45,Y=0,Z=0)
   )
   > &RigUnit_OffsetTransformForItem(
    Item=(Type=Bone,Name="jacket_flap_C_r_2")
    OffsetTransform=(Rotation=(X=0,Y=0.1305,Z=0,W=0.9914),Translation=(X=0,Y=0,Z=0),Scale3D=(X=1,Y=1,Z=1))
    bPropagateToChildren=True
    Weight=@n14.OutputParam
   )
   > n15 = &RigUnit_SphericalPoseReader(
    ActiveRegionScaleFactors=(PositiveWidth=1,NegativeWidth=1,PositiveHeight=1,NegativeHeight=1)
    ActiveRegionSize=0.27
    Debug=(bDrawDebug=false,bDraw2D=False,bDrawLocalAxes=False,DebugScale=15,DebugSegments=20,DebugThickness=0.1)
    DriverAxis=(X=1,Y=0,Z=0)
    DriverItem=(Type=Bone,Name="spine_04")
    FalloffRegionScaleFactors=(PositiveWidth=1,NegativeWidth=1,PositiveHeight=1,NegativeHeight=1)
    FalloffSize=0.43
    FlipHeightScaling=False
    FlipWidthScaling=False
    OptionalParentItem=(Type=Bone,Name="spine_01")
    RotationOffset=(X=90,Y=0,Z=15)
   )
   > &RigUnit_OffsetTransformForItem(
    Item=(Type=Bone,Name="belt_deform")
    OffsetTransform=(Rotation=(X=0,Y=-0.0872,Z=0,W=0.9962),Translation=(X=2,Y=0,Z=0.5),Scale3D=(X=1,Y=1,Z=0.9))
    bPropagateToChildren=True
    Weight=@n15.OutputParam
   )
   > &RigUnit_OffsetTransformForItem(
    Item=(Type=Bone,Name="beltSide_deform")
    OffsetTransform=(Rotation=(X=0,Y=0,Z=0,W=1),Translation=(X=0,Y=0,Z=0),Scale3D=(X=1,Y=1.1,Z=1.5))
    bPropagateToChildren=True
    Weight=@n15.OutputParam
   )
   > &RigUnit_OffsetTransformForItem(
    Item=(Type=Bone,Name="MSTR_jacket_flap_A_r")
    OffsetTransform=(Rotation=(X=0,Y=0,Z=0,W=1),Translation=(X=0,Y=0,Z=1),Scale3D=(X=1,Y=1,Z=1))
    bPropagateToChildren=True
    Weight=@n15.OutputParam
   )
   > &RigUnit_OffsetTransformForItem(
    Item=(Type=Bone,Name="MSTR_jacket_flap_A_l")
    OffsetTransform=(Rotation=(X=0,Y=0.0436,Z=0,W=0.999),Translation=(X=0,Y=0,Z=1),Scale3D=(X=1,Y=1,Z=1))
    bPropagateToChildren=True
    Weight=@n15.OutputParam
   )
   > &RigUnit_OffsetTransformForItem(
    Item=(Type=Bone,Name="MSTR_jacket_flap_C_l")
    OffsetTransform=(Rotation=(X=0,Y=-0.0523,Z=0,W=0.9986),Translation=(X=0,Y=0,Z=0),Scale3D=(X=1,Y=1,Z=1))
    bPropagateToChildren=True
    Weight=@n15.OutputParam
   )
   &RigUnit_SphericalPoseReader(
    ActiveRegionScaleFactors=(PositiveWidth=1,NegativeWidth=1,PositiveHeight=1,NegativeHeight=1)
    ActiveRegionSize=0.4
    Debug=(bDrawDebug=false,bDraw2D=False,bDrawLocalAxes=false,DebugScale=15,DebugSegments=20,DebugThickness=0.05)
    DriverAxis=(X=-1,Y=0,Z=0)
    DriverItem=(Type=Bone,Name="thigh_l")
    FalloffRegionScaleFactors=(PositiveWidth=1,NegativeWidth=1,PositiveHeight=1,NegativeHeight=1)
    FalloffSize=0.55
    FlipHeightScaling=False
    FlipWidthScaling=False
    OptionalParentItem=(Type=Bone,Name="None")
    RotationOffset=(X=120,Y=10,Z=0)
    ExecutePin=@RigVMFunction_Sequence.M
   )
   RigUnit_GetTransform = &RigUnit_GetTransform(Item=(Type=Bone,Name="thigh_l_pocket_pnt_1"), bInitial=False)
   RigVMFunction_MathVectorDistance = &RigVMFunction_MathVectorDistance(
    A=@RigUnit_GetTransform.Transform.Translation
    B=&RigUnit_GetTransform(Item=(Type=Bone,Name="jacket_B_pocket_pnt"), bInitial=False).Transform.Translation
   )
   Remap = &RigVMFunction_MathDoubleRemap(SourceMaximum=1, SourceMinimum=3, TargetMaximum=1, TargetMinimum=0, Value=@RigVMFunction_MathVectorDistance.Result)
   > RigUnit_OffsetTransformForItem = &RigUnit_OffsetTransformForItem(
    Item=(Type=Bone,Name="jacket_flap_B_l_2")
    OffsetTransform=(Rotation=(X=0,Y=0.1045,Z=0,W=0.9945),Translation=(X=0,Y=0,Z=0),Scale3D=(X=1,Y=1,Z=1))
    bPropagateToChildren=True
    Weight=@Remap.Result
   )
   n16 = &RigUnit_GetTransform(Item=(Type=Bone,Name="thigh_l_pocket_pnt_2"), bInitial=False)
   n17 = &RigUnit_GetTransform(Item=(Type=Bone,Name="pocket_jacket_C_pnt_1"), bInitial=False)
   > &RigUnit_OffsetTransformForItem(
    Item=(Type=Bone,Name="jacket_flap_C_l_3")
    OffsetTransform=(Rotation=(X=0.0068,Y=-0.1563,Z=0.0431,W=0.9867),Translation=(X=0,Y=0,Z=0),Scale3D=(X=1,Y=1,Z=1))
    bPropagateToChildren=True
    Weight=&RigVMFunction_MathDoubleRemap(
     SourceMaximum=3
     SourceMinimum=6
     TargetMaximum=1
     TargetMinimum=0
     Value=&RigVMFunction_MathVectorDistance(A=@n16.Transform.Translation, B=@n17.Transform.Translation).Result
    ).Result
   )
   > n18 = &RigUnit_SphericalPoseReader(
    ActiveRegionScaleFactors=(PositiveWidth=1,NegativeWidth=1,PositiveHeight=1,NegativeHeight=1)
    ActiveRegionSize=0.2
    Debug=(bDrawDebug=false,bDraw2D=False,bDrawLocalAxes=False,DebugScale=15,DebugSegments=20,DebugThickness=0.1)
    DriverAxis=(X=1,Y=0,Z=0)
    DriverItem=(Type=Bone,Name="thigh_r")
    FalloffRegionScaleFactors=(PositiveWidth=1,NegativeWidth=1,PositiveHeight=1,NegativeHeight=1)
    FalloffSize=0.5
    FlipHeightScaling=False
    FlipWidthScaling=False
    OptionalParentItem=(Type=Bone,Name="None")
    RotationOffset=(X=0,Y=0,Z=0)
   )
   > &RigUnit_OffsetTransformForItem(
    Item=(Type=Bone,Name="jacket_flap_C_r_1")
    OffsetTransform=(Rotation=(X=0,Y=-0.1736,Z=0,W=0.9848),Translation=(X=0,Y=0,Z=0),Scale3D=(X=1,Y=1,Z=1))
    bPropagateToChildren=True
    Weight=@n18.OutputParam
   )
   > &RigUnit_OffsetTransformForItem(
    Item=(Type=Bone,Name="jacket_flap_B_r_1")
    OffsetTransform=(Rotation=(X=0,Y=-0.1736,Z=0,W=0.9848),Translation=(X=0,Y=0,Z=0),Scale3D=(X=1,Y=1,Z=1))
    bPropagateToChildren=True
    Weight=@n18.OutputParam
   )
   n19 = &RigUnit_SphericalPoseReader(
    ActiveRegionScaleFactors=(PositiveWidth=1,NegativeWidth=1,PositiveHeight=1,NegativeHeight=1)
    ActiveRegionSize=0.2
    Debug=(bDrawDebug=false,bDraw2D=False,bDrawLocalAxes=False,DebugScale=25,DebugSegments=20,DebugThickness=0.1)
    DriverAxis=(X=-1,Y=0,Z=0)
    DriverItem=(Type=Bone,Name="thigh_l")
    FalloffRegionScaleFactors=(PositiveWidth=1,NegativeWidth=1,PositiveHeight=1,NegativeHeight=1)
    FalloffSize=0.42
    FlipHeightScaling=False
    FlipWidthScaling=False
    OptionalParentItem=(Type=Bone,Name="None")
    RotationOffset=(X=40,Y=0,Z=0)
    ExecutePin=@RigVMFunction_Sequence.N
   )
   > &RigUnit_OffsetTransformForItem(
    Item=(Type=Bone,Name="jacket_flap_C_l_1")
    OffsetTransform=(Rotation=(X=0,Y=-0.1392,Z=0,W=0.9903),Translation=(X=0,Y=0,Z=0),Scale3D=(X=1,Y=1,Z=1))
    bPropagateToChildren=True
    Weight=@n19.OutputParam
   )
   > &RigUnit_OffsetTransformForItem(
    Item=(Type=Bone,Name="jacket_flap_B_l_1")
    OffsetTransform=(Rotation=(X=0,Y=-0.0698,Z=0,W=0.9976),Translation=(X=0,Y=0,Z=0),Scale3D=(X=1,Y=1,Z=1))
    bPropagateToChildren=True
    Weight=@n19.OutputParam
   )
   &RigUnit_AimConstraintLocalSpaceOffset(
    Child=(Type=Bone,Name="MSTR_jacket_flap_A_l")
    Filter=(bX=True,bY=True,bZ=True)
    Parents=((Item=(Type=Bone,Name="jacket_flap_A_r_aim"),Weight=1))
    Weight=1
    WorldUp=(Target=(X=0,Y=0,Z=1),Kind=Location,Space=(Type=Bone,Name="pelvis"))
    ExecutePin=@RigVMFunction_Sequence.K
   )
   > &RigUnit_AimConstraintLocalSpaceOffset(
    Child=(Type=Bone,Name="MSTR_jacket_flap_B_l")
    Filter=(bX=True,bY=True,bZ=True)
    Parents=((Item=(Type=Bone,Name="jacket_flap_B_r_aim"),Weight=1))
    Weight=1
    WorldUp=(Target=(X=0,Y=0,Z=1),Kind=Location,Space=(Type=Bone,Name="pelvis"))
   )
   > &RigUnit_AimConstraintLocalSpaceOffset(
    Child=(Type=Bone,Name="MSTR_jacket_flap_C_l")
    Filter=(bX=True,bY=True,bZ=True)
    Parents=((Item=(Type=Bone,Name="jacket_flap_C_r_aim"),Weight=1))
    Weight=1
    WorldUp=(Target=(X=0,Y=0,Z=1),Kind=Location,Space=(Type=Bone,Name="pelvis"))
   )
   > &RigUnit_AimConstraintLocalSpaceOffset(
    Child=(Type=Bone,Name="MSTR_jacket_flap_A_r")
    Filter=(bX=True,bY=True,bZ=True)
    Parents=((Item=(Type=Bone,Name="jacket_flap_A_l_aim"),Weight=1))
    Weight=1
    WorldUp=(Target=(X=0,Y=0,Z=1),Kind=Location,Space=(Type=Bone,Name="pelvis"))
   )
   > &RigUnit_AimConstraintLocalSpaceOffset(
    Child=(Type=Bone,Name="MSTR_jacket_flap_B_r")
    Filter=(bX=True,bY=True,bZ=True)
    Parents=((Item=(Type=Bone,Name="jacket_flap_B_l_aim"),Weight=1))
    Weight=1
    WorldUp=(Target=(X=0,Y=0,Z=1),Kind=Location,Space=(Type=Bone,Name="pelvis"))
   )
   > &RigUnit_AimConstraintLocalSpaceOffset(
    Child=(Type=Bone,Name="MSTR_jacket_flap_C_r")
    Filter=(bX=True,bY=True,bZ=True)
    Parents=((Item=(Type=Bone,Name="jacket_flap_C_l_aim"),Weight=1))
    Weight=1
    WorldUp=(Target=(X=0,Y=0,Z=1),Kind=Location,Space=(Type=Bone,Name="pelvis"))
   )
  ]}
]
hierarchy [
 root {loc=<0 0 0>, rot=<0 0 0 1>}
 pelvis {parent=root, loc=<0 -0.5301 98.0571>, rot=<0.0127 0.707 -0.0127 -0.707>}
 spine_01 {parent=pelvis, loc=<6.1208 -0.9328 0>, rot=<0 0 0.034 -0.9994>}
 spine_02 {parent=spine_01, loc=<6.5997 -0.2095 0>, rot=<0 0 0.0238 -0.9997>}
 spine_03 {parent=spine_02, loc=<6.4914 -0.1237 0>, rot=<0 0 0.0149 -0.9999>}
 spine_04 {parent=spine_03, loc=<6.9598 0.0794 0>, rot=<0 0 -0.0132 -0.9999>}
 spine_05 {parent=spine_04, loc=<6.9987 0 0>, rot=<0 0 -0.0015 -1>}
 neck_01 {parent=spine_05, loc=<11.3138 0.1529 0>, rot=<0 0 -0.2003 -0.9797>}
 neck_02 {parent=neck_01, loc=<4.5627 0 0>, rot=<0 0 0.0504 -0.9987>}
 head {parent=neck_02, loc=<5.3914 0 0>, rot=<0 0 0.115 -0.9934>}
 transform1 {parent=head, loc=<-74.5785 -85.2756 24.3176>, rot=<0.3856 -0.683 -0.1931 -0.5895>, scale=<1 0.9998 0.9996>}
 MSTR_hair_A {parent=head, loc=<4.8173 5.1934 -6.9544>, rot=<0.0178 -0.9914 -0.1259 0.03>}
 hair_A {parent=MSTR_hair_A, loc=<0 0 0>, rot=<0 0 0 1>}
 MSTR_hair_B {parent=head, loc=<6.5784 8.7766 -5.3275>, rot=<0.1198 -0.9149 -0.3603 -0.1368>}
 hair_B {parent=MSTR_hair_B, loc=<0 0 0>, rot=<0 0 0 1>}
 MSTR_hair_C {parent=head, loc=<7.4362 10.1608 2.0174>, rot=<-0.0946 -0.704 -0.6778 0.1899>}
 hair_C {parent=MSTR_hair_C, loc=<0 0 0>, rot=<0 0 0 1>}
 MSTR_hair_D {parent=head, loc=<8.8829 9.6455 -1.7174>, rot=<-0.0744 -0.7995 -0.5957 -0.0206>}
 hair_D {parent=MSTR_hair_D, loc=<0 0 0>, rot=<0 0 0 1>}
 MSTR_hair_E {parent=head, loc=<4.3503 5.2282 7.2156>, rot=<-0.026 0.2577 0.964 -0.0596>}
 hair_E_1 {parent=MSTR_hair_E, loc=<0 0 0>, rot=<0 0 0 1>}
 hair_E_2 {parent=hair_E_1, loc=<2.1291 0 0>, rot=<0 0 0.1886 -0.9821>}
 MSTR_hair_G {parent=head, loc=<4.0411 -0.307 -9.0782>, rot=<0.377 -0.9017 -0.0187 -0.211>}
 hair_G_1 {parent=MSTR_hair_G, loc=<0 0 0>, rot=<0 0 0 1>}
 hair_G_2 {parent=hair_G_1, loc=<2.8515 0 0>, rot=<-0.0396 0.0166 0.1589 -0.9864>}
 MSTR_hair_I {parent=head, loc=<5.7391 -3.6196 -9.1918>, rot=<0.3834 -0.9003 0.0943 -0.1829>}
 hair_I_1 {parent=MSTR_hair_I, loc=<0 0 0>, rot=<0 0 0 1>}
 hair_I_2 {parent=hair_I_1, loc=<2.7177 0 0>, rot=<-0.0298 0.0262 0.1687 -0.9849>}
 MSTR_hair_K {parent=head, loc=<4.9478 -3.1552 8.8722>, rot=<0.4737 -0.3939 0.7374 -0.2772>}
 hair_K_1 {parent=MSTR_hair_K, loc=<0 0 0>, rot=<0 0 0 1>}
 hair_K_2 {parent=hair_K_1, loc=<2.0988 0 0>, rot=<0.0394 0.1601 -0.0294 -0.9859>}
 MSTR_hair_L {parent=head, loc=<4.9165 8.2517 6.19>, rot=<0.1028 0.3638 0.7818 -0.4959>}
 hair_L {parent=MSTR_hair_L, loc=<0 0 0>, rot=<0 0 0 1>}
 MSTR_hair_M {parent=head, loc=<3.9861 2.6434 -8.2828>, rot=<0.0178 -0.9914 -0.1259 0.03>}
 hair_M {parent=MSTR_hair_M, loc=<0 0 0>, rot=<0 0 0 1>}
 MSTR_hair_N {parent=head, loc=<2.7091 2.3063 8.2559>, rot=<0.1344 0.1151 0.9755 -0.1308>}
 hair_N {parent=MSTR_hair_N, loc=<0 0 0>, rot=<0 0 0 1>}
 MSTR_hair_Z {parent=head, loc=<-4.2399 -4.6139 -7.6628>, rot=<0.02 -0.9206 0.243 -0.3049>}
 hair_Z {parent=MSTR_hair_Z, loc=<0 0 0>, rot=<0 0 0 1>}
 MSTR_hair_dyn_V {parent=head, loc=<3.9055 5.4213 -6.8835>, rot=<0.1186 -0.7376 -0.6592 -0.0857>}
 hair_dyn_V {parent=MSTR_hair_dyn_V, loc=<0 0 0>, rot=<0 0 0 1>}
 MSTR_hair_dyn_Y {parent=head, loc=<6.165 7.8393 -4.6422>, rot=<0.1198 -0.9149 -0.3603 -0.1368>}
 hair_dyn_Y {parent=MSTR_hair_dyn_Y, loc=<0 0 0>, rot=<0 0 0 1>}
 MSTR_hair_dyn_W {parent=head, loc=<6.2053 9.4268 2.1734>, rot=<-0.1207 -0.6881 -0.6272 0.3444>}
 hair_dyn_W {parent=MSTR_hair_dyn_W, loc=<0 0 0>, rot=<0 0 0 1>}
 MSTR_hair_dyn_X {parent=head, loc=<8.2957 9.1867 -1.9652>, rot=<-0.2156 -0.7728 -0.5816 0.1342>}
 hair_dyn_X {parent=MSTR_hair_dyn_X, loc=<0 0 0>, rot=<0 0 0 1>}
 MSTR_hair_dyn_R_1 {parent=head, loc=<4.8267 4.7229 6.8946>, rot=<0.1772 -0.5883 0.7773 -0.1355>}
 hair_dyn_R_1 {parent=MSTR_hair_dyn_R_1, loc=<0 0 0>, rot=<0 0 0 1>}
 hair_dyn_R_2 {parent=hair_dyn_R_1, loc=<4.0542 0 0>, rot=<0 0 0.2642 -0.9645>}
 MSTR_hair_dyn_P {parent=head, loc=<3.4997 -9.3305 4.6958>, rot=<0.6693 -0.1729 0.6941 -0.2008>}
 hair_dyn_P {parent=MSTR_hair_dyn_P, loc=<0 0 0>, rot=<0 0 0 1>}
 MSTR_hair_dyn_N_1 {parent=head, loc=<3.41 -5.4503 8.0015>, rot=<0.5501 -0.2559 0.7727 -0.1865>}
 hair_dyn_N_1 {parent=MSTR_hair_dyn_N_1, loc=<0 0 0>, rot=<0 0 0 1>}
 hair_dyn_N_2 {parent=hair_dyn_N_1, loc=<2.6721 0 0>, rot=<-0.1035 0.1615 0.1105 -0.9752>}
 MSTR_hair_dyn_M {parent=head, loc=<1.8927 1.4373 -7.9894>, rot=<0.0444 -0.8963 0.4328 -0.086>}
 hair_dyn_M_1 {parent=MSTR_hair_dyn_M, loc=<0 0 0>, rot=<0 0 0 1>}
 hair_dyn_M_2 {parent=hair_dyn_M_1, loc=<2.8515 0 0>, rot=<0.9336 0.0198 -0.3354 -0.1244>}
 hair_dyn_M_3 {parent=hair_dyn_M_2, loc=<2.1661 0 0>, rot=<0 0.4347 0 -0.9006>}
 MSTR_hair_dyn_L {parent=head, loc=<-2.8296 -1.72 -8.2345>, rot=<-0.3254 -0.6595 -0.6104 -0.2943>}
 hair_dyn_L {parent=MSTR_hair_dyn_L, loc=<0 0 0>, rot=<0 0 0 1>}
 MSTR_hair_dyn_K_1 {parent=head, loc=<-5.9255 -6.5046 -7.9188>, rot=<-0.1571 -0.834 -0.3164 -0.4238>}
 hair_dyn_K_1 {parent=MSTR_hair_dyn_K_1, loc=<0 0 0>, rot=<0 0 0 1>}
 hair_dyn_K_2 {parent=hair_dyn_K_1, loc=<3.2127 0 0>, rot=<0.0035 0.3182 -0.0189 -0.9478>}
 MSTR_hair_dyn_J {parent=head, loc=<-0.2078 -5.95 -7.9603>, rot=<-0.1644 -0.8115 -0.1017 -0.5515>}
 hair_dyn_J {parent=MSTR_hair_dyn_J, loc=<0 0 0>, rot=<0 0 0 1>}
 MSTR_hair_dyn_I {parent=head, loc=<-1.0771 -8.7093 -6.678>, rot=<0.2492 -0.7181 0.3651 -0.5375>}
 hair_dyn_I {parent=MSTR_hair_dyn_I, loc=<0 0 0>, rot=<0 0 0 1>}
 MSTR_hair_dyn_E_1 {parent=head, loc=<5.0584 -0.5472 -8.4946>, rot=<0.232 -0.7011 0.2556 -0.624>}
 hair_dyn_E_1 {parent=MSTR_hair_dyn_E_1, loc=<0 0 0>, rot=<0 0 0 1>}
 hair_dyn_E_2 {parent=hair_dyn_E_1, loc=<3.522 0 0>, rot=<-0.0112 0.1769 0.0199 -0.984>}
 MSTR_hair_dyn_D {parent=head, loc=<4.0129 -6.1716 -7.6006>, rot=<0.1816 -0.6911 0.1854 -0.6746>}
 hair_dyn_D {parent=MSTR_hair_dyn_D, loc=<0 0 0>, rot=<0 0 0 1>}
 MSTR_hair_dyn_C {parent=head, loc=<3.1103 -10.2449 -2.8387>, rot=<0.267 -0.7313 0.3825 -0.4976>}
 hair_dyn_C {parent=MSTR_hair_dyn_C, loc=<0 0 0>, rot=<0 0 0 1>}
 MSTR_hair_dyn_B_1 {parent=head, loc=<4.9445 -0.4084 9.8034>, rot=<0.648 -0.2649 0.6639 -0.2631>}
 hair_dyn_B_1 {parent=MSTR_hair_dyn_B_1, loc=<0 0 0>, rot=<0 0 0 1>}
 hair_dyn_B_2 {parent=hair_dyn_B_1, loc=<2.0988 0 0>, rot=<0.0437 0.2397 -0.0704 -0.9673>}
 MSTR_hair_dyn_A {parent=head, loc=<-0.2473 -8.0289 6.5048>, rot=<0.4605 -0.3848 0.7596 -0.2509>}
 hair_dyn_A {parent=MSTR_hair_dyn_A, loc=<0 0 0>, rot=<0 0 0 1>}
 MSTR_hair_dyn_Q {parent=head, loc=<-2.7746 -4.7075 9.9482>, rot=<0.6613 0.1663 0.7152 0.1533>}
 hair_dyn_Q {parent=MSTR_hair_dyn_Q, loc=<0 0 0>, rot=<0 0 0 1>}
 hair_mid_deform_base {parent=head, loc=<0.3206 -7.9608 0>, rot=<0 0 0.6666 -0.7454>}
 hair_mid_deform {parent=hair_mid_deform_base, loc=<0 0 0>, rot=<0 0 0 1>}
 MSTR_hair_F {parent=head, loc=<4.0411 -0.3071 -9.0782>, rot=<0.377 -0.9017 -0.0187 -0.211>}
 hair_F_1 {parent=MSTR_hair_F, loc=<0 0 0>, rot=<0 0 0 -1>}
 hair_down_deform_base {parent=neck_01, loc=<2.0149 -8.7289 0>, rot=<0 0 0.9991 -0.0419>}
 hair_down_deform {parent=hair_down_deform_base, loc=<0 0 0>, rot=<0 0 0 1>}
 clavicle_l {parent=spine_05, loc=<5.4831 -0.1053 -3.2946>, rot=<0.0465 -0.7289 -0.009 -0.6829>}
 upperarm_l {parent=clavicle_l, loc=<9.8658 -1.2136 -0.4105>, rot=<0.0164 -0.3245 -0.0269 -0.9454>}
 lowerarm_l {parent=upperarm_l, loc=<26.465 -0.8345 0.259>, rot=<0.0036 0.0161 -0.2298 -0.9731>}
 hand_l {parent=lowerarm_l, loc=<22.7355 -0.3203 -0.6105>, rot=<-0.6784 -0.0223 0.0735 -0.7307>}
 thumb_01_l {parent=hand_l, loc=<1.3118 -0.3191 -2.459>, rot=<0.6546 -0.4042 0.0664 -0.6354>}
 thumb_02_l {parent=thumb_01_l, loc=<5.9336 -0.0448 0.0401>, rot=<-0.0037 0.0213 0.298 -0.9543>}
 thumb_03_l {parent=thumb_02_l, loc=<2.2223 0.3682 -0.0907>, rot=<0.0118 -0.013 0.0066 -0.9998>}
 index_metacarpal_l {parent=hand_l, loc=<1.6535 0.0183 -2.96>, rot=<0.0158 -0.0124 -0.0173 -0.9996>}
 index_01_l {parent=index_metacarpal_l, loc=<8.3008 0.0357 -0.0524>, rot=<-0.0044 0.0485 0.0924 -0.9945>}
 index_02_l {parent=index_01_l, loc=<3.3587 -0.3004 0.0452>, rot=<-0.002 0.0027 0.3225 -0.9466>}
 index_03_l {parent=index_02_l, loc=<2.5578 -0.123 -0.0252>, rot=<-0.002 0.0067 0.2558 -0.9667>}
 middle_metacarpal_l {parent=hand_l, loc=<2.1856 -0.1725 -0.5984>, rot=<0.006 0.0168 -0.0329 -0.9993>}
 middle_01_l {parent=middle_metacarpal_l, loc=<7.3275 -0.2447 0.0129>, rot=<-0.0208 0.0042 0.0858 -0.9961>}
 middle_02_l {parent=middle_01_l, loc=<4.0123 0.1401 0.1175>, rot=<-0.0218 0.0258 0.3512 -0.9357>}
 middle_03_l {parent=middle_02_l, loc=<2.6077 0.0534 0.1373>, rot=<0.0082 0.0188 0.1708 -0.9851>}
 ring_metacarpal_l {parent=hand_l, loc=<2.2074 -0.3625 1.3638>, rot=<-0.0284 0.0722 -0.0458 -0.9959>}
 ring_01_l {parent=ring_metacarpal_l, loc=<7.0295 -0.6197 -0.1446>, rot=<-0.0489 -0.0252 0.1892 -0.9804>}
 ring_02_l {parent=ring_01_l, loc=<3.6174 0.4647 -0.1027>, rot=<0 0 0.2806 -0.9598>}
 ring_03_l {parent=ring_02_l, loc=<2.4152 0.114 0.0199>, rot=<0 0 0.1189 -0.9929>}
 pinky_metacarpal_l {parent=hand_l, loc=<1.7616 -0.9753 3.3814>, rot=<-0.0397 0.0814 -0.069 -0.9935>}
 pinky_01_l {parent=pinky_metacarpal_l, loc=<7.1709 -0.4831 -0.1094>, rot=<-0.0654 -0.0035 0.2242 -0.9723>}
 pinky_02_l {parent=pinky_01_l, loc=<2.8274 0.0893 -0.1043>, rot=<0.0037 0.0011 0.2854 -0.9584>}
 pinky_03_l {parent=pinky_02_l, loc=<2.2046 0.0307 -0.0078>, rot=<-0.0002 0.0098 0.1445 -0.9895>}
 deform_wrist_dn_base_l {parent=hand_l, loc=<0 -1.5326 0>, rot=<0 0 0 -1>}
 deform_wrist_dn_l {parent=deform_wrist_dn_base_l, loc=<0 0 0>, rot=<0 0 0 1>}
 deform_wrist_up_base_l {parent=hand_l, loc=<0 2.3009 0>, rot=<0 0 0 -1>}
 deform_wrist_up_l {parent=deform_wrist_up_base_l, loc=<0 0 0>, rot=<0 0 0 1>}
 hand_attach_l {parent=hand_l, loc=<8.5 -3 0>, rot=<0.7071 0.7071 0 0>}
 weapon_l {parent=hand_l, loc=<-0.1974 -3.4111 0.3811>, rot=<0 0 0 1>}
 lowerarm_twist_02_l {parent=lowerarm_l, loc=<10.6707 0 0>, rot=<0 0 0 1>}
 MSTR_jacket_l_flap {parent=lowerarm_twist_02_l, loc=<4.4391 -3.5265 1.9873>, rot=<0.6769 0.1587 0.678 0.2387>}
 leather_l_1 {parent=MSTR_jacket_l_flap, loc=<0 0 0>, rot=<0 0 0 1>}
 leather_l_2 {parent=leather_l_1, loc=<3.1902 -0.04 -0.0003>, rot=<0 0 0.2047 -0.9788>}
 leather_l_3 {parent=leather_l_2, loc=<1.9907 -0.0002 0.0002>, rot=<0 0 0 1>}
 lowerarm_twist_01_l {parent=lowerarm_l, loc=<18.316 0 0>, rot=<0 0 0 1>}
 deform_elbow_in_base_l {parent=lowerarm_l, loc=<0.1834 3.6062 -0.2572>, rot=<0 0 0 -1>}
 deform_elbow_in_l {parent=deform_elbow_in_base_l, loc=<0 0 0>, rot=<0 0 0 1>}
 deform_elbow_out_base_l {parent=lowerarm_l, loc=<0 -4 0>, rot=<0 0 0 1>}
 deform_elbow_out_l {parent=deform_elbow_out_base_l, loc=<0 0 0>, rot=<0 0 0 1>}
 upperarm_twist_01_l {parent=upperarm_l, loc=<7.2135 -0.1134 0>, rot=<0 0 0 1>}
 deform_upperarm_bk_base_l {parent=upperarm_twist_01_l, loc=<-6.4053 -4.3746 0.314>, rot=<0 0 0 -1>}
 deform_upperarm_bk_l {parent=deform_upperarm_bk_base_l, loc=<0 0 0>, rot=<0 0 0 1>}
 deform_upperarm_fr_base_l {parent=upperarm_twist_01_l, loc=<-5.6312 5.092 0.3354>, rot=<0 0 0 -1>}
 deform_upperarm_fr_l {parent=deform_upperarm_fr_base_l, loc=<0 0 0>, rot=<0 0 0 1>}
 deform_upperarm_up_base_l {parent=upperarm_twist_01_l, loc=<-8.891 0.4825 3.9517>, rot=<0 0 0 -1>}
 deform_upperarm_up_l {parent=deform_upperarm_up_base_l, loc=<0 0 0>, rot=<0 0 0 1>}
 MSTR_jacket_collar_dyn_l {parent=upperarm_twist_01_l, loc=<-8.5629 -0.0594 4.7083>, rot=<-0.0664 0.0583 0.0072 -0.9961>}
 jacket_collar_dyn_l_1 {parent=MSTR_jacket_collar_dyn_l, loc=<0 0 0>, rot=<0 0 0 1>}
 jacket_collar_dyn_l_2 {parent=jacket_collar_dyn_l_1, loc=<5.9664 0 0.0001>, rot=<0 0 0 1>}
 jacket_collar_dyn_l_3 {parent=jacket_collar_dyn_l_2, loc=<4.3706 0 0.0001>, rot=<0 0 0 1>}
 upperarm_twist_02_l {parent=upperarm_l, loc=<19.7998 -0.1134 0>, rot=<0 0 0 1>}
 clavicle_l_deform_base {parent=clavicle_l, loc=<7.3262 -1.7643 4.8616>, rot=<0 0 0 1>}
 clavicle_l_deform {parent=clavicle_l_deform_base, loc=<0 0 0>, rot=<0 0 0 1>}
 clavicle_r {parent=spine_05, loc=<5.4831 -0.1053 3.2946>, rot=<0.7289 0.0465 -0.6829 0.009>}
 upperarm_r {parent=clavicle_r, loc=<-9.8658 1.2136 0.4105>, rot=<0.0164 -0.3245 -0.0269 -0.9454>}
 lowerarm_r {parent=upperarm_r, loc=<-26.465 0.8345 -0.259>, rot=<0.0036 0.0161 -0.2298 -0.9731>}
 hand_r {parent=lowerarm_r, loc=<-22.7355 0.3203 0.6106>, rot=<-0.6784 -0.0223 0.0735 -0.7307>}
 thumb_01_r {parent=hand_r, loc=<-1.3119 0.3192 2.459>, rot=<0.6546 -0.4042 0.0664 -0.6354>}
 thumb_02_r {parent=thumb_01_r, loc=<-5.9336 0.0448 -0.0401>, rot=<-0.0037 0.0213 0.298 -0.9543>}
 thumb_03_r {parent=thumb_02_r, loc=<-2.2223 -0.3682 0.0907>, rot=<0.0118 -0.013 0.0066 -0.9998>}
 index_metacarpal_r {parent=hand_r, loc=<-1.6534 -0.0181 2.9601>, rot=<0.0158 -0.0124 -0.0173 -0.9996>}
 index_01_r {parent=index_metacarpal_r, loc=<-8.3008 -0.0357 0.0524>, rot=<-0.0044 0.0485 0.0924 -0.9945>}
 index_02_r {parent=index_01_r, loc=<-3.3589 0.3003 -0.0453>, rot=<-0.002 0.0027 0.3225 -0.9466>}
 index_03_r {parent=index_02_r, loc=<-2.5578 0.123 0.0252>, rot=<-0.002 0.0067 0.2558 -0.9667>}
 middle_metacarpal_r {parent=hand_r, loc=<-2.1856 0.1726 0.5984>, rot=<0.006 0.0168 -0.0329 -0.9993>}
 middle_01_r {parent=middle_metacarpal_r, loc=<-7.3273 0.2447 -0.0129>, rot=<-0.0208 0.0042 0.0858 -0.9961>}
 middle_02_r {parent=middle_01_r, loc=<-4.0124 -0.14 -0.1175>, rot=<-0.0218 0.0258 0.3512 -0.9357>}
 middle_03_r {parent=middle_02_r, loc=<-2.6077 -0.0534 -0.1373>, rot=<0.0082 0.0188 0.1708 -0.9851>}
 ring_metacarpal_r {parent=hand_r, loc=<-2.2073 0.3627 -1.3638>, rot=<-0.0284 0.0722 -0.0458 -0.9959>}
 ring_01_r {parent=ring_metacarpal_r, loc=<-7.0297 0.6197 0.1446>, rot=<-0.0489 -0.0252 0.1892 -0.9804>}
 ring_02_r {parent=ring_01_r, loc=<-3.6171 -0.4648 0.1027>, rot=<0 0 0.2806 -0.9598>}
 ring_03_r {parent=ring_02_r, loc=<-2.4152 -0.114 -0.0199>, rot=<0 0 0.1189 -0.9929>}
 pinky_metacarpal_r {parent=hand_r, loc=<-1.7617 0.9753 -3.3814>, rot=<-0.0397 0.0814 -0.069 -0.9935>}
 pinky_01_r {parent=pinky_metacarpal_r, loc=<-7.1707 0.4832 0.1094>, rot=<-0.0654 -0.0035 0.2242 -0.9723>}
 pinky_02_r {parent=pinky_01_r, loc=<-2.8274 -0.0893 0.1043>, rot=<0.0037 0.0011 0.2854 -0.9584>}
 pinky_03_r {parent=pinky_02_r, loc=<-2.2045 -0.031 0.0078>, rot=<-0.0002 0.0098 0.1445 -0.9895>}
 deform_wrist_dn_base_r {parent=hand_r, loc=<-0.0001 1.5328 0>, rot=<0 0 0 -1>}
 deform_wrist_dn_r {parent=deform_wrist_dn_base_r, loc=<0 0 0>, rot=<0 0 0 1>}
 deform_wrist_up_base_r {parent=hand_r, loc=<-0.0002 -2.3007 -0.0001>, rot=<0 0 0 -1>}
 deform_wrist_up_r {parent=deform_wrist_up_base_r, loc=<0 0 0>, rot=<0 0 0 1>}
 weapon_r {parent=hand_r, loc=<0.1975 3.4112 -0.3811>, rot=<0 0 0 -1>}
 hand_attach_r {parent=hand_r, loc=<-8.5 3 0>, rot=<0 0 -0.7071 -0.7071>}
 lowerarm_twist_02_r {parent=lowerarm_r, loc=<-10.6707 0 0>, rot=<0 0 0 1>}
 lowerarm_twist_01_r {parent=lowerarm_r, loc=<-18.316 0 0>, rot=<0 0 0 1>}
 MSTR_jacket_r_flap {parent=lowerarm_twist_01_r, loc=<3.2065 3.5264 -1.9875>, rot=<0.1587 -0.6769 0.2387 -0.678>}
 leather_r_1 {parent=MSTR_jacket_r_flap, loc=<0 0 0>, rot=<0 0 0 1>}
 leather_r_2 {parent=leather_r_1, loc=<3.1897 -0.0402 0>, rot=<0 0 0.2047 -0.9788>}
 leather_r_3 {parent=leather_r_2, loc=<1.991 0 0>, rot=<0 0 0 1>}
 deform_elbow_in_base_r {parent=lowerarm_r, loc=<-0.1834 -3.6062 0.2572>, rot=<0 0 0 -1>}
 deform_elbow_in_r {parent=deform_elbow_in_base_r, loc=<0 0 0>, rot=<0 0 0 1>}
 deform_elbow_out_base_r {parent=lowerarm_r, loc=<0 4 -0.0001>, rot=<0 0 0 -1>}
 deform_elbow_out_r {parent=deform_elbow_out_base_r, loc=<0 0 0>, rot=<0 0 0 1>}
 upperarm_twist_01_r {parent=upperarm_r, loc=<-7.2135 0.1134 0>, rot=<0 0 0 1>}
 deform_upperarm_bk_base_r {parent=upperarm_twist_01_r, loc=<6.4053 4.3746 -0.314>, rot=<0 0 0 -1>}
 deform_upperarm_bk_r {parent=deform_upperarm_bk_base_r, loc=<0 0 0>, rot=<0 0 0 1>}
 deform_upperarm_fr_base_r {parent=upperarm_twist_01_r, loc=<5.6312 -5.092 -0.3354>, rot=<0 0 0 -1>}
 deform_upperarm_fr_r {parent=deform_upperarm_fr_base_r, loc=<0 0 0>, rot=<0 0 0 1>}
 deform_collar_for_r {parent=deform_upperarm_fr_base_r, loc=<0 0 0>, rot=<0 0 0 1>}
 deform_upperarm_up_base_r {parent=upperarm_twist_01_r, loc=<8.891 -0.4825 -3.9517>, rot=<0 0 0 -1>}
 deform_upperarm_up_r {parent=deform_upperarm_up_base_r, loc=<0 0 0>, rot=<0 0 0 1>}
 MSTR_jacket_collar_dyn_r {parent=upperarm_twist_01_r, loc=<8.5629 0.0594 -4.7084>, rot=<-0.0072 -0.9961 -0.0664 -0.0583>}
 jacket_collar_dyn_r_1 {parent=MSTR_jacket_collar_dyn_r, loc=<0 0 0>, rot=<0 0 0 1>}
 jacket_collar_dyn_r_2 {parent=jacket_collar_dyn_r_1, loc=<5.9665 0 0>, rot=<0 0 0 1>}
 jacket_collar_dyn_r_3 {parent=jacket_collar_dyn_r_2, loc=<4.3706 0 0>, rot=<0 0 0 1>}
 upperarm_twist_02_r {parent=upperarm_r, loc=<-19.7998 0.1134 0>, rot=<0 0 0 1>}
 clavicle_r_deform_base {parent=clavicle_r, loc=<-7.3262 1.7643 -4.8614>, rot=<0 -1 0 0>}
 clavicle_r_deform {parent=clavicle_r_deform_base, loc=<0 0 0>, rot=<0 0 0 1>}
 jacket_collar_l_deform_base {parent=clavicle_r, loc=<-12.8657 -7.138 0.1397>, rot=<0.5274 -0.8081 0.1822 0.1887>}
 jacket_collar_l_deform {parent=jacket_collar_l_deform_base, loc=<0 0 0>, rot=<0 0 0 1>}
 attach_backpack {parent=spine_05, loc=<-0.123 -13.473 0>, rot=<-0.0093 -0.707 0.0093 -0.707>}
 Backpack_BR {parent=attach_backpack, type=Null, loc=<0 3.5 0>, rot=<0.0436 0 0 0.999>}
 attach_cape {parent=spine_05, loc=<9.017 -11.714 0>, rot=<0.0616 -0.7044 -0.0616 -0.7044>}
 cape {parent=attach_cape, type=Null, loc=<0 4 0>, rot=<0.1045 0 0 0.9945>}
 collar_for_r_deform_base {parent=spine_05, loc=<3.8401 6.8097 5.5514>, rot=<0.2548 0.1946 -0.3878 -0.8642>}
 collar_for_r_deform {parent=collar_for_r_deform_base, loc=<0 0 0>, rot=<0 0 0 1>}
 collar_for_l_deform_base {parent=spine_05, loc=<3.8397 6.8096 -5.5514>, rot=<-0.2548 -0.1946 -0.3878 -0.8642>}
 collar_for_l_deform {parent=collar_for_l_deform_base, loc=<0 0 0>, rot=<0 0 0 -1>}
 deform_pec_aim_base_l {parent=spine_04, loc=<3.2456 9.3309 -1.2237>, rot=<0.0603 -0.4033 0.3318 -0.8507>}
 deform_pec_aim_l {parent=deform_pec_aim_base_l, loc=<0 0 0>, rot=<0 0 0 1>}
 deform_pec_base_l {parent=deform_pec_aim_l, loc=<9.3665 0 0>, rot=<0 0 0 1>}
 deform_pec_l {parent=deform_pec_base_l, loc=<0 0 0>, rot=<0 0 0 1>}
 deform_pec_aim_base_r {parent=spine_04, loc=<3.2452 9.3309 1.2237>, rot=<-0.4033 -0.0603 0.8507 0.3318>}
 deform_pec_aim_r {parent=deform_pec_aim_base_r, loc=<0 0 0>, rot=<0 0 0 1>}
 deform_pec_base_r {parent=deform_pec_aim_r, loc=<-9.3667 -0.0003 -0.0004>, rot=<0 0 0 1>}
 deform_pec_r {parent=deform_pec_base_r, loc=<0 0 0>, rot=<0 0 0 1>}
 star_r_base {parent=spine_04, loc=<1.3013 12.8288 4.186>, rot=<0.5042 0.5066 0.5164 0.4717>}
 star_r {parent=star_r_base, loc=<0 0 0>, rot=<0 0 0 1>}
 star_l_base {parent=spine_04, loc=<1.3018 12.8289 -4.186>, rot=<-0.4717 -0.5164 -0.5066 -0.5042>}
 star_l {parent=star_l_base, loc=<0 0 0>, rot=<0 0 0 1>}
 pockets_r_deform_base {parent=spine_03, loc=<2.9051 0.5627 10.53>, rot=<0.0373 0.0787 -0.1088 -0.9902>}
 pockets_r_deform {parent=pockets_r_deform_base, loc=<0 0 0>, rot=<0 0 0 1>}
 pockets_l_deform_base {parent=spine_03, loc=<2.9051 0.5627 -10.53>, rot=<0.9902 0.1088 -0.0787 -0.0373>}
 pockets_l_deform {parent=pockets_l_deform_base, loc=<0 0 0>, rot=<0 -1 0 0>}
 belt_deform_base {parent=spine_02, loc=<0.6132 9.9648 -1.3757>, rot=<0.4797 0.5195 0.5195 0.4797>}
 belt_deform {parent=belt_deform_base, loc=<0 0 0>, rot=<0 0 0 1>}
 beltSide_deform_base {parent=spine_02, loc=<0.8538 9.4338 0>, rot=<0.4797 0.5195 0.5195 0.4797>}
 beltSide_deform {parent=beltSide_deform_base, loc=<0 0 0>, rot=<0 0 0 1>}
 MSTR_greenade_back_dyn {parent=spine_01, loc=<5.6161 1.5742 10.9419>, rot=<0.2501 0.1434 0.9573 -0.0224>}
 greenade_back_dyn_1 {parent=MSTR_greenade_back_dyn, loc=<0 0 0>, rot=<0 0 0 1>}
 greenade_back_dyn_2 {parent=greenade_back_dyn_1, loc=<5 0 0>, rot=<0 0 0 1>}
 greenade_back_ring_dyn {parent=greenade_back_dyn_1, loc=<0.8341 -0.9022 2.5403>, rot=<0.1088 0.0221 0.1981 -0.9739>}
 MSTR_greenade_for_dyn {parent=spine_01, loc=<4.7726 6.4133 9.4412>, rot=<0.229 0.3108 0.9201 -0.0659>}
 greenade_for_dyn_1 {parent=MSTR_greenade_for_dyn, loc=<0 0 0>, rot=<0 0 0 1>}
 greenade_for_dyn_2 {parent=greenade_for_dyn_1, loc=<5 0 0>, rot=<0 0 0 1>}
 greenade_for_ring_dyn {parent=greenade_for_dyn_1, loc=<0.618 -0.9984 2.4638>, rot=<0.0865 -0.0862 0.2235 -0.967>}
 MSTR_jacket_flap_B_r {parent=spine_01, loc=<4.7648 6.7157 6.616>, rot=<0.2017 0.437 0.8761 -0.0294>}
 jacket_flap_B_r_1 {parent=MSTR_jacket_flap_B_r, loc=<0 0 0>, rot=<0 0 0 1>}
 jacket_flap_B_r_2 {parent=jacket_flap_B_r_1, loc=<6.1596 0 0>, rot=<0 -0.0299 0 -0.9996>}
 jacket_flap_B_r_3 {parent=jacket_flap_B_r_2, loc=<4.2907 0.0001 0.0001>, rot=<0 -0.0072 0 -1>}
 jacket_flap_B_r_4 {parent=jacket_flap_B_r_3, loc=<7.2867 0 0>, rot=<0.0023 -0.0505 0.0453 -0.9977>}
 MSTR_jacket_flap_C_r {parent=spine_01, loc=<5.8258 1.639 9.3115>, rot=<0.2482 0.0105 0.9684 -0.023>}
 jacket_flap_C_r_1 {parent=MSTR_jacket_flap_C_r, loc=<0 0 0>, rot=<0 0 0 1>}
 jacket_flap_C_r_2 {parent=jacket_flap_C_r_1, loc=<6.8551 0 -0.0004>, rot=<0 -0.0331 0 -0.9995>}
 jacket_flap_C_r_3 {parent=jacket_flap_C_r_2, loc=<4.5805 0 0.0002>, rot=<0 -0.039 0 -0.9992>}
 jacket_flap_C_r_4 {parent=jacket_flap_C_r_3, loc=<5.4782 0 0>, rot=<0 -0.0447 0 -0.999>}
 MSTR_jacket_flap_D_r {parent=spine_01, loc=<6.397 -3.3659 7.8816>, rot=<0.2837 -0.2911 0.9084 -0.0983>}
 jacket_flap_D_r_1 {parent=MSTR_jacket_flap_D_r, loc=<0 0 0>, rot=<0 0 0 1>}
 jacket_flap_D_r_2 {parent=jacket_flap_D_r_1, loc=<4.5329 0 -0.0003>, rot=<-0.0019 -0.063 0.0011 -0.998>}
 jacket_flap_D_r_3 {parent=jacket_flap_D_r_2, loc=<6.9311 0 0>, rot=<-0.0038 -0.0478 0.002 -0.9988>}
 jacket_flap_D_r_4 {parent=jacket_flap_D_r_3, loc=<5.0415 0 0.0001>, rot=<0 -0.108 0 -0.9941>}
 MSTR_jacket_flap_E_r {parent=spine_01, loc=<6.3196 -6.2228 4.1739>, rot=<0.2397 -0.4976 0.8219 -0.1394>}
 jacket_flap_E_r_1 {parent=MSTR_jacket_flap_E_r, loc=<0 0 0>, rot=<0 0 0 1>}
 jacket_flap_E_r_2 {parent=jacket_flap_E_r_1, loc=<3.9369 0 0.0001>, rot=<-0.0124 -0.042 -0.0064 -0.999>}
 jacket_flap_E_r_3 {parent=jacket_flap_E_r_2, loc=<7.1431 0 0.0003>, rot=<-0.0159 -0.0702 -0.0097 -0.9974>}
 jacket_flap_E_r_4 {parent=jacket_flap_E_r_3, loc=<5.0428 -0.0001 -0.0002>, rot=<0 0 0 1>}
 MSTR_jacket_flap_back {parent=spine_01, loc=<6.1977 -6.9047 0>, rot=<0.2173 -0.6729 0.6729 -0.2173>}
 jacket_flap_back_1 {parent=MSTR_jacket_flap_back, loc=<0 0 0>, rot=<0 0 0 1>}
 jacket_flap_back_2 {parent=jacket_flap_back_1, loc=<3.9442 0 0>, rot=<0 -0.1153 0 -0.9933>}
 jacket_flap_back_3 {parent=jacket_flap_back_2, loc=<6.9004 0 0>, rot=<0 -0.0372 0 -0.9993>}
 jacket_flap_back_4 {parent=jacket_flap_back_3, loc=<4.992 0 0>, rot=<0 0 0 1>}
 MSTR_jacket_flap_B_l {parent=spine_01, loc=<4.6431 7.0455 -6.2127>, rot=<0.0294 -0.8761 -0.437 -0.2017>}
 jacket_flap_B_l_1 {parent=MSTR_jacket_flap_B_l, loc=<0 0 0>, rot=<0 0 0 1>}
 jacket_flap_B_l_2 {parent=jacket_flap_B_l_1, loc=<6.1597 0 0>, rot=<0 -0.0299 0 -0.9996>}
 jacket_flap_B_l_3 {parent=jacket_flap_B_l_2, loc=<4.2909 0 0>, rot=<0 -0.0072 0 -1>}
 jacket_flap_B_l_4 {parent=jacket_flap_B_l_3, loc=<7.2867 0 0>, rot=<0 -0.0506 0 -0.9987>}
 jacket_B_pocket_pnt {parent=jacket_flap_B_l_1, loc=<25.91 0 -2.82>, rot=<0 0 0 1>}
 MSTR_jacket_flap_C_l {parent=spine_01, loc=<5.8254 1.639 -9.3115>, rot=<0.023 -0.9684 -0.0105 -0.2482>}
 jacket_flap_C_l_1 {parent=MSTR_jacket_flap_C_l, loc=<0 0 0>, rot=<0 0 0 1>}
 jacket_flap_C_l_2 {parent=jacket_flap_C_l_1, loc=<6.8544 0 0>, rot=<0 -0.0331 0 -0.9995>}
 jacket_flap_C_l_3 {parent=jacket_flap_C_l_2, loc=<4.5809 0 0>, rot=<0 -0.039 0 -0.9992>}
 jacket_flap_C_l_4 {parent=jacket_flap_C_l_3, loc=<5.4782 0 0>, rot=<0 -0.0572 0 -0.9984>}
 pocket_jacket_C_pnt_1 {parent=jacket_flap_C_l_2, loc=<16.85 4.05 -1.54>, rot=<0 0 0 1>, scale=<0.8 1 1>}
 pocket_jacket_C_pnt_2 {parent=jacket_flap_C_l_2, loc=<13.43 3.24 -2.84>, rot=<0 0 0 1>, scale=<0.8 1 1>}
 MSTR_jacket_flap_D_l {parent=spine_01, loc=<6.3905 -3.8031 -7.5209>, rot=<0.0936 -0.8995 0.3249 -0.2766>}
 jacket_flap_D_l_1 {parent=MSTR_jacket_flap_D_l, loc=<0 0 0>, rot=<0 0 0 1>}
 jacket_flap_D_l_2 {parent=jacket_flap_D_l_1, loc=<4.5325 0 0>, rot=<0.0019 -0.063 -0.0011 -0.998>}
 jacket_flap_D_l_3 {parent=jacket_flap_D_l_2, loc=<6.9311 0 0>, rot=<0.0038 -0.0478 -0.002 -0.9988>}
 jacket_flap_D_l_4 {parent=jacket_flap_D_l_3, loc=<5.042 0 0>, rot=<0 -0.072 0 -0.9974>}
 MSTR_jacket_flap_E_l {parent=spine_01, loc=<6.32 -6.2228 -4.1739>, rot=<0.1394 -0.8219 0.4976 -0.2397>}
 jacket_flap_E_l_1 {parent=MSTR_jacket_flap_E_l, loc=<0 0 0>, rot=<0 0 0 1>}
 jacket_flap_E_l_2 {parent=jacket_flap_E_l_1, loc=<3.937 0 0>, rot=<0.0124 -0.042 0.0064 -0.999>}
 jacket_flap_E_l_3 {parent=jacket_flap_E_l_2, loc=<7.1436 0 0>, rot=<0.0159 -0.0702 0.0097 -0.9974>}
 jacket_flap_E_l_4 {parent=jacket_flap_E_l_3, loc=<5.0425 0 0>, rot=<0 0 0 1>}
 MSTR_jacket_flap_A_r {parent=spine_01, loc=<2.2973 9.3495 0.3307>, rot=<0.1347 0.613 0.7785 0.0005>}
 jacket_flap_A_r_1 {parent=MSTR_jacket_flap_A_r, loc=<0 0 0>, rot=<0 0 0 1>}
 jacket_flap_A_r_2 {parent=jacket_flap_A_r_1, loc=<3.9871 0 0>, rot=<0 0 0 1>}
 jacket_flap_A_r_3 {parent=jacket_flap_A_r_2, loc=<4.6357 0 0>, rot=<0 0 0 1>}
 jacket_flap_A_r_4 {parent=jacket_flap_A_r_3, loc=<6.8095 0 0>, rot=<0 0 0 1>}
 MSTR_jacket_flap_A_l {parent=spine_01, loc=<2.3 9.2809 -0.1347>, rot=<0.0193 -0.7794 -0.6096 -0.1434>}
 jacket_flap_A_l_1 {parent=MSTR_jacket_flap_A_l, loc=<0 0 0>, rot=<0 0 0 1>}
 jacket_flap_A_l_2 {parent=jacket_flap_A_l_1, loc=<3.7645 0 0>, rot=<0 0 0 1>}
 jacket_flap_A_l_3 {parent=jacket_flap_A_l_2, loc=<4.3321 0 0>, rot=<0 0 0 1>}
 jacket_flap_A_l_4 {parent=jacket_flap_A_l_3, loc=<7.3076 0 0>, rot=<0 0 0 1>}
 MSTR_jacket_flap_mid_dyn {parent=spine_01, loc=<2.8453 9.4159 0.2537>, rot=<0.0828 0.7022 0.7022 0.0828>}
 jacket_flap_mid_dyn_1 {parent=MSTR_jacket_flap_mid_dyn, loc=<0 0 0>, rot=<0 0 0 1>}
 thigh_l {parent=pelvis, loc=<-1.0855 -0.2109 -8.7053>, rot=<0.0009 0.0355 0.0155 -0.9993>}
 calf_l {parent=thigh_l, loc=<-42.5284 -0.0019 -0.7297>, rot=<0.0006 -0.0161 -0.0391 -0.9991>}
 foot_l {parent=calf_l, loc=<-46.7401 -2.5014 -0.592>, rot=<0.0427 -0.0238 0.0353 -0.9982>}
 ball_l {parent=foot_l, loc=<-3.5658 10.8066 -0.1544>, rot=<0.0001 -0.0001 -0.7063 -0.7079>}
 foot_l_deform_base {parent=foot_l, loc=<0.4 0 0>, rot=<0 0 -0.649 -0.7608>}
 foot_l_deform_for_base {parent=foot_l_deform_base, loc=<7.0355 -0.0311 0.0268>, rot=<0 0 0 1>}
 foot_l_deform_for {parent=foot_l_deform_for_base, loc=<0 0 0>, rot=<0 0 0 1>}
 foot_l_deform_back_base {parent=foot_l_deform_base, loc=<-5 0 0>, rot=<0 -1 0 0>}
 foot_l_deform_back {parent=foot_l_deform_back_base, loc=<0 0 0>, rot=<0 0 0 1>}
 calf_twist_02_l {parent=calf_l, loc=<-19.1546 -1.7943 -0.2638>, rot=<0.0237 0.0095 -0.0035 -0.9997>}
 calf_twist_01_l {parent=calf_l, loc=<-35.8483 -1.9197 -0.5794>, rot=<0.0237 0.0095 -0.0035 -0.9997>}
 deform_knee_in_base_l {parent=calf_l, loc=<0.8031 -5.0231 0.2456>, rot=<0 0 0 -1>}
 deform_knee_in_l {parent=deform_knee_in_base_l, loc=<0 0 0>, rot=<0 0 0 1>}
 deform_knee_out_base_l {parent=calf_l, loc=<0 4 0>, rot=<0 0 0 -1>}
 deform_knee_out_l {parent=deform_knee_out_base_l, loc=<0 0 0>, rot=<0 0 0 1>}
 thigh_twist_01_l {parent=thigh_l, loc=<-27.952 0 0>, rot=<0.0013 0 0 -1>}
 deform_groin_base_l {parent=thigh_l, loc=<0.4801 5.6776 -0.2486>, rot=<0 0 0 -1>}
 deform_groin_l {parent=deform_groin_base_l, loc=<0 0 0>, rot=<0 0 0 1>}
 thigh_l_pocket_pnt_1 {parent=thigh_l, loc=<-23.71 7.39 0.96>, rot=<0 0 0 1>}
 thigh_l_pocket_pnt_2 {parent=thigh_l, loc=<-10.98 5.87 -8.49>, rot=<0 0 0 1>}
 thigh_r {parent=pelvis, loc=<-1.0853 -0.2109 8.7053>, rot=<0.0355 -0.0009 0.9993 0.0155>}
 calf_r {parent=thigh_r, loc=<42.5286 0.0019 0.7297>, rot=<0.0006 -0.0161 -0.0391 -0.9991>}
 foot_r {parent=calf_r, loc=<46.7403 2.5014 0.592>, rot=<0.0427 -0.0238 0.0353 -0.9982>}
 ball_r {parent=foot_r, loc=<3.5658 -10.8066 0.1544>, rot=<0.0001 -0.0001 -0.7063 -0.7079>}
 foot_r_deform_base {parent=foot_r, loc=<-0.4002 0 0>, rot=<0 0 0.7608 -0.649>}
 foot_r_deform_for_base {parent=foot_r_deform_base, loc=<7 0 0>, rot=<0 0 0 1>}
 foot_r_deform_for {parent=foot_r_deform_for_base, loc=<0 0 0>, rot=<0 0 0 1>}
 foot_r_deform_back_base {parent=foot_r_deform_base, loc=<-5 0 0>, rot=<0 -1 0 0>}
 foot_r_deform_back {parent=foot_r_deform_back_base, loc=<0 0 0>, rot=<0 0 0 1>}
 calf_twist_02_r {parent=calf_r, loc=<19.1546 1.7943 0.2638>, rot=<0.0237 0.0095 -0.0035 -0.9997>}
 calf_twist_01_r {parent=calf_r, loc=<35.8485 1.9197 0.5794>, rot=<0.0237 0.0095 -0.0035 -0.9997>}
 deform_knee_in_base_r {parent=calf_r, loc=<-0.8031 5.0231 -0.2457>, rot=<0 0 0 -1>}
 deform_knee_in_r {parent=deform_knee_in_base_r, loc=<0 0 0>, rot=<0 0 0 1>}
 deform_knee_out_base_r {parent=calf_r, loc=<0 -4 0>, rot=<0 0 0 -1>}
 deform_knee_out_r {parent=deform_knee_out_base_r, loc=<0 0 0>, rot=<0 0 0 1>}
 thigh_twist_01_r {parent=thigh_r, loc=<27.952 0 0>, rot=<0.0013 0 0 -1>}
 deform_groin_base_r {parent=thigh_r, loc=<-0.4799 -5.6776 0.2485>, rot=<0 0 0 1>}
 deform_groin_r {parent=deform_groin_base_r, loc=<0 0 0>, rot=<0 0 0 1>}
 deform_glute_base_l {parent=pelvis, loc=<-7.3466 -9.0213 -9.1245>, rot=<0.0172 0.031 -0.4477 -0.8935>}
 deform_glute_l {parent=deform_glute_base_l, loc=<0 0 0>, rot=<0 0 0 1>}
 deform_glute_base_r {parent=pelvis, loc=<-7.3464 -9.0213 9.1245>, rot=<0.031 -0.0172 0.8935 -0.4477>}
 deform_glute_r {parent=deform_glute_base_r, loc=<0 0 0>, rot=<0 0 0 1>}
 thigh_l_twist {parent=pelvis, loc=<-1.0855 -0.2109 -8.7053>, rot=<0.0009 0.0355 0.0155 -0.9993>}
 jacket_flap_B_r_aim {parent=thigh_l_twist, loc=<-36.2797 14.6061 -14.6239>, rot=<0 0 0 1>}
 jacket_flap_A_r_aim {parent=thigh_l_twist, loc=<-39.1572 17.9381 -1.3156>, rot=<-0.0168 -0.7844 -0.6069 -0.127>}
 jacket_flap_C_r_aim {parent=thigh_l_twist, loc=<-35.3738 0.5731 -23.7949>, rot=<0.0048 -0.9768 -0.0064 -0.2138>}
 thigh_l_aim_up {parent=pelvis, loc=<-1.0855 -0.2109 -8.7053>, rot=<-0.0158 -0.0318 0.4083 0.9122>}
 thigh_l_aim_up_pnt {parent=thigh_l_aim_up, loc=<3.5 0 0>, rot=<0 0 0 1>}
 thigh_r_aim_up {parent=pelvis, loc=<-1.0855 -0.2109 8.7053>, rot=<0.0141 -0.0326 0.4065 0.913>}
 thigh_r_aim_up_pnt {parent=thigh_r_aim_up, loc=<3 0 0>, rot=<0 0 0 1>}
 thigh_r_twist {parent=pelvis, loc=<-1.0855 -0.2109 8.7053>, rot=<0.0355 -0.0009 0.9993 0.0155>}
 jacket_flap_B_l_aim {parent=thigh_r_twist, loc=<36.2797 -14.6061 14.6239>, rot=<0 0 0 1>}
 jacket_flap_A_l_aim {parent=thigh_r_twist, loc=<39.1572 -17.9381 1.3156>, rot=<0 0 0 1>}
 jacket_flap_C_l_aim {parent=thigh_r_twist, loc=<35.3738 -0.5731 23.7949>, rot=<0 0 0 1>}
 attach {parent=root, loc=<0 0 0>, rot=<0 0 0 1>}
 ik_hand_root {parent=root, loc=<0 0 0>, rot=<0 0 0 1>}
 ik_hand_gun {parent=ik_hand_root, loc=<-48.1744 3.3546 104.4381>, rot=<0.6276 0.1309 0.4183 -0.6434>}
 ik_hand_r {parent=ik_hand_gun, loc=<0 0 0>, rot=<0 0 0 1>}
 ik_hand_l {parent=ik_hand_gun, loc=<59.3272 67.6997 34.3524>, rot=<-0.6158 -0.7027 -0.3565 0>}
 ik_foot_root {parent=root, loc=<0 0 0>, rot=<0 0 0 1>}
 ik_foot_l {parent=ik_foot_root, loc=<14.8262 -7.3545 8.296>, rot=<0.0357 0.7045 0.0284 -0.7082>}
 ik_foot_r {parent=ik_foot_root, loc=<-14.8262 -7.3546 8.2958>, rot=<0.7082 0.0284 -0.7045 0.0357>}
 root_fp {parent=root, loc=<0 0 0>, rot=<0 0 0 1>}
 attach_fp {parent=root_fp, loc=<0 0 0>, rot=<0 0 0 1>}
 camera_root {parent=root_fp, loc=<0 0 0>, rot=<0 0 0 1>}
 camera {parent=camera_root, loc=<0 0 0>, rot=<0 0 0 1>}
 attach_emote_01 {parent=root, loc=<0 0 0>, rot=<0 0 0 1>}
 attach_emote_02 {parent=root, loc=<0 0 0>, rot=<0 0 0 1>}
 attach_emote_03 {parent=root, loc=<0 0 0>, rot=<0 0 0 1>}
 warp_point_01 {parent=root, loc=<0 0 0>, rot=<0 0 0 1>}
 warp_point_02 {parent=root, loc=<0 0 0>, rot=<0 0 0 1>}
 warp_point_03 {parent=root, loc=<0 0 0>, rot=<0 0 0 1>}
]
variables [leg_l_forward {type=double, default=0}]
