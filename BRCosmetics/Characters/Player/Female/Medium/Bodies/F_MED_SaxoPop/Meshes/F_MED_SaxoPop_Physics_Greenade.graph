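# Physics asset for the two dangling "greenade" props, front and back, on the F_MED_SaxoPop female character body.
# Each forms a constrained chain from a small MSTR anchor capsule through a grenade capsule to a ring capsule.
# Angular limits allow wide swing with little twist.
# Two right-side jacket flap bodies, a capsule and a box, have no constraints, apparently serving as colliders for the swinging grenades.
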PhysicsAsset /BRCosmetics/Characters/Player/Female/Medium/Bodies/F_MED_SaxoPop/Meshes/F_MED_SaxoPop_Physics_Greenade
bodies {
 MSTR_greenade_back_dyn [{shapeType=Capsule, length=0.505, radius=0.505, shapeName=MSTR_greenade_back_dyn_capsule, rotation=<rot 0 -90 0>}]
 MSTR_greenade_for_dyn [{shapeType=Capsule, length=0.505, radius=0.505, shapeName=MSTR_greenade_for_dyn_capsule, rotation=<rot 0 -90 0>}]
 greenade_back_dyn_1 [{
   shapeType=Capsule
   length=7.1771
   radius=2.6287
   shapeName=greenade_back_dyn_1_capsule
   center=<6.1 0.5287 2.2027>
   rotation=<rot 0.0002 -83.129 -76.6169>
  }]
 greenade_back_ring_dyn [{shapeType=Capsule, length=1.8701, radius=1, shapeName=greenade_back_ring_dyn_capsule, center=<1.4549 -0.684 0>, rotation=<rot 90 0 0>}]
 greenade_for_dyn_1 [{
   shapeType=Capsule
   length=6.9854
   radius=2.7869
   shapeName=greenade_for_dyn_1_capsule
   center=<5.8501 0.3259 2.105>
   rotation=<rot 0.0028 -84.5294 -74.6271>
  }]
 greenade_for_ring_dyn [{shapeType=Capsule, length=1.8701, radius=1, shapeName=greenade_for_ring_dyn_capsule, center=<1.5411 -0.684 0>, rotation=<rot 90 0 0>}]
 jacket_flap_B_r_1 [{shapeType=Capsule, length=7.5092, radius=7.2048, shapeName=jacket_flap_B_r_1_capsule, center=<9.7067 0.1917 -7.0061>, rotation=<rot 90 0 0>}]
 jacket_flap_C_r_1 [{
   shapeType=Box
   extentX=16.0212
   extentY=8.1359
   extentZ=2.6283
   shapeName=jacket_flap_C_r_1_box
   center=<8.0063 1.5341 -1.6804>
   rotation=<rot -2.4924 0.0967 -6.8648>
  }]
}
constraints [
 {bone1Name=greenade_back_dyn_1, bone2Name=MSTR_greenade_back_dyn, swing1LimitDegrees=20, swing2LimitDegrees=75, twistLimitDegrees=3}
 {bone1Name=greenade_for_dyn_1, bone2Name=MSTR_greenade_for_dyn, swing1LimitDegrees=20, swing2LimitDegrees=90, twistLimitDegrees=3}
 {bone1Name=greenade_back_ring_dyn, bone2Name=greenade_back_dyn_1, swing1LimitDegrees=45, swing2LimitDegrees=15, twistLimitDegrees=5}
 {bone1Name=greenade_for_ring_dyn, bone2Name=greenade_for_dyn_1, swing1LimitDegrees=45, swing2LimitDegrees=15, twistLimitDegrees=5}
]
