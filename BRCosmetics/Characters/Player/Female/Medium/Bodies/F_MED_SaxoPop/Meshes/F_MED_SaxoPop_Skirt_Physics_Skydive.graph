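# Physics asset for the F_MED_SaxoPop female character's skirt, built on jacket_flap bones and named for skydiving.
# Seven flap chains (C, D and E per side, plus back) each have four boxes hinged to a small MSTR root body by tight swing and twist limits.
# Angularly free constraints join neighboring chain tips into an open band from C left around the back to C right.
# Pelvis, spine_01 and thigh capsules provide body shapes.
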PhysicsAsset /BRCosmetics/Characters/Player/Female/Medium/Bodies/F_MED_SaxoPop/Meshes/F_MED_SaxoPop_Skirt_Physics_Skydive
bodies {
 MSTR_jacket_flap_C_l [{shapeType=Box, extentX=1.01, extentY=1.01, extentZ=1.01, shapeName=MSTR_jacket_flap_C_l_box, rotation=<rot 0 -90 0>}]
 MSTR_jacket_flap_C_r [{shapeType=Box, extentX=1.01, extentY=1.01, extentZ=1.01, shapeName=MSTR_jacket_flap_C_r_box, rotation=<rot 0 -90 0>}]
 MSTR_jacket_flap_D_l [{shapeType=Capsule, length=0.505, radius=0.505, shapeName=MSTR_jacket_flap_D_l_capsule}]
 MSTR_jacket_flap_D_r [{shapeType=Capsule, length=0.505, radius=0.505, shapeName=MSTR_jacket_flap_D_r_capsule}]
 MSTR_jacket_flap_E_l [{shapeType=Capsule, length=0.505, radius=0.505, shapeName=MSTR_jacket_flap_E_l_capsule}]
 MSTR_jacket_flap_E_r [{shapeType=Capsule, length=0.505, radius=0.505, shapeName=MSTR_jacket_flap_E_r_capsule}]
 MSTR_jacket_flap_back [{shapeType=Capsule, length=0.505, radius=0.505, shapeName=MSTR_jacket_flap_back_capsule}]
 jacket_flap_B_l_1 [{shapeType=Capsule, length=1, radius=1, shapeName=jacket_flap_B_l_1_capsule}]
 jacket_flap_C_l_1 [{shapeType=Box, extentX=6.3724, extentY=7.4364, extentZ=1.6657, shapeName=jacket_flap_C_l_1_box, center=<3.4334 0 0>}]
 jacket_flap_C_l_2 [{shapeType=Box, extentX=4.6795, extentY=7.4364, extentZ=1.6657, shapeName=jacket_flap_C_l_2_box, center=<2.3176 0 0>}]
 jacket_flap_C_l_3 [{shapeType=Box, extentX=5.2988, extentY=8.9936, extentZ=1.6657, shapeName=jacket_flap_C_l_3_box, center=<2.8319 0 0>}]
 jacket_flap_C_l_4 [{shapeType=Box, extentX=6.0069, extentY=9.2893, extentZ=1.6657, shapeName=jacket_flap_C_l_4_box, center=<3.1252 0 0>}]
 jacket_flap_C_r_1 [{shapeType=Box, extentX=6.3724, extentY=7.4364, extentZ=1.6657, shapeName=jacket_flap_C_r_1_box, center=<3.4334 0 0>}]
 jacket_flap_C_r_2 [{shapeType=Box, extentX=4.6795, extentY=7.4364, extentZ=1.6657, shapeName=jacket_flap_C_r_2_box, center=<2.3176 0 0>}]
 jacket_flap_C_r_3 [{shapeType=Box, extentX=5.2988, extentY=8.9936, extentZ=1.6657, shapeName=jacket_flap_C_r_3_box, center=<2.8319 0 0>}]
 jacket_flap_C_r_4 [{shapeType=Box, extentX=6.0069, extentY=9.2893, extentZ=1.6657, shapeName=jacket_flap_C_r_4_box, center=<3.1252 0 0>}]
 jacket_flap_D_l_1 [{shapeType=Box, extentX=4.2783, extentY=5.644, extentZ=1.644, shapeName=jacket_flap_D_l_1_box, center=<2.1732 0 0>}]
 jacket_flap_D_l_2 [{shapeType=Box, extentX=4.8838, extentY=5.644, extentZ=2.644, shapeName=jacket_flap_D_l_2_box, center=<2.4802 0 0>}]
 jacket_flap_D_l_3 [{shapeType=Box, extentX=6.9094, extentY=6.9449, extentZ=2.644, shapeName=jacket_flap_D_l_3_box, center=<1.4805 0 0>}]
 jacket_flap_D_l_4 [{shapeType=Box, extentX=7.3669, extentY=7.9121, extentZ=2.644, shapeName=jacket_flap_D_l_4_box, center=<3.754 0 0>}]
 jacket_flap_D_r_1 [{shapeType=Box, extentX=4.2783, extentY=5.644, extentZ=1.644, shapeName=jacket_flap_D_r_1_box, center=<2.1732 0 0>}]
 jacket_flap_D_r_2 [{shapeType=Box, extentX=4.8838, extentY=5.644, extentZ=2.644, shapeName=jacket_flap_D_r_2_box, center=<2.4802 0 0>}]
 jacket_flap_D_r_3 [{shapeType=Box, extentX=6.9094, extentY=6.9449, extentZ=2.0431, shapeName=jacket_flap_D_r_3_box, center=<1.4805 0 0>}]
 jacket_flap_D_r_4 [{shapeType=Box, extentX=7.3669, extentY=7.9121, extentZ=2.0431, shapeName=jacket_flap_D_r_4_box, center=<3.754 0.1081 0>}]
 jacket_flap_E_l_1 [{shapeType=Box, extentX=3.5015, extentY=5.644, extentZ=1.644, shapeName=jacket_flap_E_l_1_box, center=<2.0168 0 0>}]
 jacket_flap_E_l_2 [{shapeType=Box, extentX=4.8838, extentY=5.644, extentZ=2.644, shapeName=jacket_flap_E_l_2_box, center=<2.4802 0 0>}]
 jacket_flap_E_l_3 [{shapeType=Box, extentX=6.9094, extentY=6.9449, extentZ=2.644, shapeName=jacket_flap_E_l_3_box, center=<1.4805 0 0>}]
 jacket_flap_E_l_4 [{shapeType=Box, extentX=7.3669, extentY=7.9121, extentZ=2.644, shapeName=jacket_flap_E_l_4_box, center=<3.754 0 0>}]
 jacket_flap_E_r_1 [{shapeType=Box, extentX=3.5015, extentY=5.644, extentZ=1.644, shapeName=jacket_flap_E_r_1_box, center=<2.0168 0 0>}]
 jacket_flap_E_r_2 [{shapeType=Box, extentX=4.8838, extentY=5.644, extentZ=2.644, shapeName=jacket_flap_E_r_2_box, center=<2.4802 0 0>}]
 jacket_flap_E_r_3 [{
   shapeType=Box
   extentX=6.9094
   extentY=6.9449
   extentZ=2.644
   shapeName=jacket_flap_E_r_3_box
   center=<1.4805 0 0>
   rotation=<rot 0 0 5.0636>
  }]
 jacket_flap_E_r_4 [{shapeType=Box, extentX=7.3669, extentY=7.9121, extentZ=2.644, shapeName=jacket_flap_E_r_4_box, center=<3.754 0 0>, rotation=<rot 0 0 3.9942>}]
 jacket_flap_back_1 [{shapeType=Box, extentX=3.5015, extentY=5.644, extentZ=1.644, shapeName=jacket_flap_back_1_box, center=<2.0168 0 0>}]
 jacket_flap_back_2 [{shapeType=Box, extentX=4.8838, extentY=5.644, extentZ=2.644, shapeName=jacket_flap_back_2_box, center=<2.4802 0 0>}]
 jacket_flap_back_3 [{shapeType=Box, extentX=6.9094, extentY=6.9449, extentZ=2.644, shapeName=jacket_flap_back_3_box, center=<1.4805 0 0>}]
 jacket_flap_back_4 [{shapeType=Box, extentX=7.3669, extentY=7.9121, extentZ=2.644, shapeName=jacket_flap_back_4_box, center=<3.754 0 0>}]
 pelvis [{shapeType=Capsule, length=8.0341, radius=13.7854, shapeName=pelvis_capsule, center=<-5.3095 -0.6003 0>, rotation=<rot 90 18.5805 0>}]
 spine_01 [{shapeType=Capsule, length=0, radius=12.5975, shapeName=spine_01_capsule, center=<-2.1569 3.3588 0>, rotation=<rot 90 0 0>}]
 thigh_l [{shapeType=Capsule, length=9.8326, radius=8.9612, shapeName=thigh_l_capsule, center=<-8.9582 2.4695 0.8599>, rotation=<rot 81.1118 180 -180>}]
 thigh_r [{shapeType=Capsule, length=8.2978, radius=10.4961, shapeName=thigh_r_capsule, center=<9.5944 0.5509 -2.2974>, rotation=<rot 90 0 0>}]
}
constraints [
 {bone1Name=jacket_flap_D_l_4, bone2Name=jacket_flap_D_l_3, swing1LimitDegrees=4, swing2LimitDegrees=4, twistLimitDegrees=1}
 {bone1Name=jacket_flap_D_l_3, bone2Name=jacket_flap_D_l_2, swing1LimitDegrees=4, swing2LimitDegrees=4, twistLimitDegrees=2}
 {bone1Name=jacket_flap_D_l_2, bone2Name=jacket_flap_D_l_1, swing1LimitDegrees=4, swing2LimitDegrees=4, twistLimitDegrees=2}
 {bone1Name=jacket_flap_E_l_4, bone2Name=jacket_flap_E_l_3, swing1LimitDegrees=4, swing2LimitDegrees=4, twistLimitDegrees=2}
 {bone1Name=jacket_flap_E_l_3, bone2Name=jacket_flap_E_l_2, swing1LimitDegrees=4, swing2LimitDegrees=4, twistLimitDegrees=2}
 {bone1Name=jacket_flap_E_l_2, bone2Name=jacket_flap_E_l_1, swing1LimitDegrees=4, swing2LimitDegrees=4, twistLimitDegrees=2}
 {bone1Name=jacket_flap_back_4, bone2Name=jacket_flap_back_3, swing1LimitDegrees=4, swing2LimitDegrees=4, twistLimitDegrees=2}
 {bone1Name=jacket_flap_back_3, bone2Name=jacket_flap_back_2, swing1LimitDegrees=4, swing2LimitDegrees=4, twistLimitDegrees=2}
 {bone1Name=jacket_flap_back_2, bone2Name=jacket_flap_back_1, swing1LimitDegrees=4, swing2LimitDegrees=4, twistLimitDegrees=2}
 {bone1Name=jacket_flap_E_r_4, bone2Name=jacket_flap_E_r_3, swing1LimitDegrees=4, swing2LimitDegrees=4, twistLimitDegrees=2}
 {bone1Name=jacket_flap_E_r_3, bone2Name=jacket_flap_E_r_2, swing1LimitDegrees=4, swing2LimitDegrees=4, twistLimitDegrees=2}
 {bone1Name=jacket_flap_E_r_2, bone2Name=jacket_flap_E_r_1, swing1LimitDegrees=4, swing2LimitDegrees=4, twistLimitDegrees=2}
 {bone1Name=jacket_flap_D_r_4, bone2Name=jacket_flap_D_r_3, swing1LimitDegrees=4, swing2LimitDegrees=4, twistLimitDegrees=2}
 {bone1Name=jacket_flap_D_r_3, bone2Name=jacket_flap_D_r_2, swing1LimitDegrees=4, swing2LimitDegrees=4, twistLimitDegrees=2}
 {bone1Name=jacket_flap_D_r_2, bone2Name=jacket_flap_D_r_1, swing1LimitDegrees=4, swing2LimitDegrees=4, twistLimitDegrees=2}
 {bone1Name=jacket_flap_D_l_1, bone2Name=MSTR_jacket_flap_D_l, swing1LimitDegrees=4, swing2LimitDegrees=4, twistLimitDegrees=5}
 {bone1Name=jacket_flap_D_r_1, bone2Name=MSTR_jacket_flap_D_r, swing1LimitDegrees=4, swing2LimitDegrees=4, twistLimitDegrees=5}
 {bone1Name=jacket_flap_E_l_1, bone2Name=MSTR_jacket_flap_E_l, swing1LimitDegrees=4, swing2LimitDegrees=4, twistLimitDegrees=5}
 {bone1Name=jacket_flap_E_r_1, bone2Name=MSTR_jacket_flap_E_r, swing1LimitDegrees=4, swing2LimitDegrees=4, twistLimitDegrees=5}
 {bone1Name=jacket_flap_back_1, bone2Name=MSTR_jacket_flap_back, swing1LimitDegrees=4, swing2LimitDegrees=4, twistLimitDegrees=5}
 {
  bone1Name=jacket_flap_E_l_4
  bone2Name=jacket_flap_D_l_4
  swing1LimitDegrees=45
  swing1Motion=Free
  swing2LimitDegrees=45
  swing2Motion=Free
  twistLimitDegrees=45
  twistMotion=Free
 }
 {
  bone1Name=jacket_flap_E_r_4
  bone2Name=jacket_flap_D_r_4
  swing1LimitDegrees=45
  swing1Motion=Free
  swing2LimitDegrees=45
  swing2Motion=Free
  twistLimitDegrees=45
  twistMotion=Free
 }
 {
  bone1Name=jacket_flap_back_4
  bone2Name=jacket_flap_E_r_4
  swing1LimitDegrees=45
  swing1Motion=Free
  swing2LimitDegrees=45
  swing2Motion=Free
  twistLimitDegrees=45
  twistMotion=Free
 }
 {
  bone1Name=jacket_flap_back_4
  bone2Name=jacket_flap_E_l_4
  swing1LimitDegrees=45
  swing1Motion=Free
  swing2LimitDegrees=45
  swing2Motion=Free
  twistLimitDegrees=45
  twistMotion=Free
 }
 {bone1Name=jacket_flap_C_l_4, bone2Name=jacket_flap_C_l_3, swing1LimitDegrees=4, swing2LimitDegrees=4, twistLimitDegrees=2}
 {bone1Name=jacket_flap_C_l_3, bone2Name=jacket_flap_C_l_2, swing1LimitDegrees=4, swing2LimitDegrees=4, twistLimitDegrees=2}
 {bone1Name=jacket_flap_C_l_2, bone2Name=jacket_flap_C_l_1, swing1LimitDegrees=4, swing2LimitDegrees=4, twistLimitDegrees=2}
 {bone1Name=jacket_flap_C_r_4, bone2Name=jacket_flap_C_r_3, swing1LimitDegrees=4, swing2LimitDegrees=4, twistLimitDegrees=2}
 {bone1Name=jacket_flap_C_r_3, bone2Name=jacket_flap_C_r_2, swing1LimitDegrees=4, swing2LimitDegrees=4, twistLimitDegrees=2}
 {bone1Name=jacket_flap_C_r_2, bone2Name=jacket_flap_C_r_1, swing1LimitDegrees=4, swing2LimitDegrees=4, twistLimitDegrees=2}
 {bone1Name=jacket_flap_C_r_1, bone2Name=MSTR_jacket_flap_C_r, swing1LimitDegrees=4, swing2LimitDegrees=4, twistLimitDegrees=2}
 {bone1Name=jacket_flap_C_l_1, bone2Name=MSTR_jacket_flap_C_l, swing1LimitDegrees=4, swing2LimitDegrees=4, twistLimitDegrees=2}
 {
  bone1Name=jacket_flap_D_l_4
  bone2Name=jacket_flap_C_l_4
  swing1LimitDegrees=45
  swing1Motion=Free
  swing2LimitDegrees=45
  swing2Motion=Free
  twistLimitDegrees=45
  twistMotion=Free
 }
 {
  bone1Name=jacket_flap_C_r_4
  bone2Name=jacket_flap_D_r_4
  swing1LimitDegrees=45
  swing1Motion=Free
  swing2LimitDegrees=45
  swing2Motion=Free
  twistLimitDegrees=45
  twistMotion=Free
 }
]
