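# Post-process style AnimBP for the SaxoPop female body cosmetic, copying the parent mesh pose and adding secondary motion to outfit bones.
# Control Rigs bracket a chain of constraint, RigidBody and AnimDynamics nodes for skirt, zipper, collars, leather and stars.
# Skydiving swaps skirt physics assets, while DBNO tilts a jacket flap.
# State-driven gravity feeds the simulations; backbling checks gate back-collar dynamics and offset the SaxoPop backpack attachment.
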AnimBlueprint /BRCosmetics/Characters/Player/Female/Medium/Bodies/F_MED_SaxoPop/Meshes/F_MED_SaxoPop_AnimBP
target_skeleton=&/F_MED_SaxoPop_Skeleton
graphs [
 AnimGraph {nodes [
   n1 = Control Rig(Source=Copy Pose From Mesh() {bUseAttachedParent=true}) {
    alphaBoolBlend {blendOption=Linear}
    controlRigAssetReference {blueprintRigClass=&/F_MED_SaxoPop_CtrlRig.F_MED_SaxoPop_CtrlRig_C}
   }
   n2 = "Constraint - dyn_pocket_zipper_base (PPPP)"!(
    ComponentPose=Local To Component(LocalPose=@n1)
    Alpha=1
    ConstraintWeights_0=0.0588
    ConstraintWeights_1:float=0.387
    ConstraintWeights_2:float=0.5026
    ConstraintWeights_3:float=0.0517
   ) -> (Pose) {
    alphaBoolBlend {blendOption=Linear}
    boneToModify {boneName=dyn_pocket_zipper_base}
    constraintSetup [
     {targetBone {boneName=pelvis}}
     {targetBone {boneName=thigh_l}}
     {targetBone {boneName=thigh_twist_01_l}}
     {targetBone {boneName=deform_groin_l}}
    ]
    constraintWeights [1,1,1,1]
   }
   n3 = Anim Dynamics - Bone: dyn_pocket_zipper_01(ComponentPose=@n2, Alpha=1, ExternalForce="0,0,0") {
    angularDampingOverride=0.9999
    angularSpringConstant=500
    bAngularSpring=true
    bOverrideAngularDamping=true
    bUseGravityOverride=true
    simulationSpace=RootRelative
    alphaBoolBlend {blendOption=Linear}
    boundBone {boneName=dyn_pocket_zipper_01}
    gravityOverride {z=-980}
    physicsBodyDefinitions [{
      boundBone {boneName=dyn_pocket_zipper_01}
      boxExtents=<10 10 22.8619>
      constraintSetup {angularLimitsMax=<-2 1.2878 0>, angularLimitsMin=<-2 -50 0>, angularTarget=<0 0 0>}
      localJointOffset=<5 0 0>
     }]
    simSpaceSettings {maxAngularAcceleration=250, maxAngularVelocity=3, simSpaceAngularAlpha=1}
   }
   n4 = RigidBody!(
    ComponentPose=@n3
    bAlphaBoolEnabled:bool=NOT Boolean(A=Get bIsSkyDiving())
    OverrideWorldGravity:Vector=Get Gravity Override()
    ExternalForce
   ) -> (Pose) {
    alphaInputType=Bool
    bOverrideWorldGravity=true
    overridePhysicsAsset=&/F_MED_SaxoPop_Skirt_Physics
    simulationSpace=BaseBoneSpace
    alphaBoolBlend {blendOption=Linear}
    baseBoneRef {boneName=pelvis}
    componentAppliedLinearAccClamp=<0 0 0>
    simSpaceSettings {maxAngularAcceleration=10, maxAngularVelocity=10, maxLinearAcceleration=150, maxLinearVelocity=200, worldAlpha=1}
   }
   n5 = RigidBody(ComponentPose=@n4, bAlphaBoolEnabled=Get bIsSkyDiving(), OverrideWorldGravity=Get Gravity Override()) {
    alphaInputType=Bool
    bOverrideWorldGravity=true
    overridePhysicsAsset=&/F_MED_SaxoPop_Skirt_Physics_Skydive
    simulationSpace=BaseBoneSpace
    alphaBoolBlend {blendOption=Linear}
    baseBoneRef {boneName=pelvis}
    componentAppliedLinearAccClamp=<0 0 0>
    simSpaceSettings {maxAngularAcceleration=10, maxAngularVelocity=10, maxLinearAcceleration=200, maxLinearVelocity=200, worldAlpha=1}
   }
   n6 = RigidBody(ComponentPose=@n5, OverrideWorldGravity=Get Gravity Override()) {
    alphaInputType=Bool
    bOverrideWorldGravity=true
    overridePhysicsAsset=&/F_MED_SaxoPop_Physics_Greenade
    simulationSpace=BaseBoneSpace
    alphaBoolBlend {blendOption=Linear}
    baseBoneRef {boneName=spine_01}
    componentAppliedLinearAccClamp=<0 0 0>
    simSpaceSettings {
     maxAngularAcceleration=20
     maxAngularVelocity=20
     maxLinearAcceleration=450
     maxLinearVelocity=500
     worldAlpha=1
     externalLinearDragV=<0.1 0.1 0.1>
     externalLinearVelocity=<0.1 0.1 0.1>
    }
   }
   n7 = Get Gravity override animDyn() -> (Gravity override animDyn:Vector)
   n8 = Anim Dynamics - Bone: collar_back_1!(ComponentPose=@n6, bAlphaBoolEnabled:bool=Get no backpack() -> (no backpack:bool), GravityOverride:Vector=@n7, ExternalForce="0,0,0") -> (Pose) {
    alphaInputType=Bool
    angularDampingOverride=1
    angularSpringConstant=1000
    bAngularSpring=true
    bGravityOverrideInSimSpace=true
    bOverrideAngularDamping=true
    bUseGravityOverride=true
    linearSpringConstant=500
    simulationSpace=RootRelative
    alphaBoolBlend {blendOption=Linear}
    boundBone {boneName=collar_back_1}
    componentAppliedLinearAccClamp=<3000 3000 3000>
    componentLinearAccScale=<0.1 0.1 0.45>
    componentLinearVelScale=<0.1 0.1 0.45>
    gravityOverride {x=-980}
    physicsBodyDefinitions [{
      boundBone {boneName=collar_back_1}
      boxExtents=<10 10 10>
      constraintSetup {angularLimitsMax=<0 0 0>, angularLimitsMin=<0 -15 0>, angularTarget=<0 0 0>}
      localJointOffset=<2 0 -1>
     }]
    relativeSpaceBone {boneName=MSTR_collar_back}
    simSpaceSettings {maxAngularAcceleration=300, maxAngularVelocity=300, simSpaceAngularAlpha=0.5}
   }
   n9 = Anim Dynamics - Bone: collar_back_2!(ComponentPose=@n8, bAlphaBoolEnabled:bool=Get no backpack(), GravityOverride:Vector=@n7, ExternalForce="0,0,0") -> (Pose) {
    alphaInputType=Bool
    angularDampingOverride=1
    angularSpringConstant=1000
    bAngularSpring=true
    bGravityOverrideInSimSpace=true
    bOverrideAngularDamping=true
    bUseGravityOverride=true
    linearSpringConstant=500
    simulationSpace=RootRelative
    alphaBoolBlend {blendOption=Linear}
    boundBone {boneName=collar_back_2}
    componentAppliedLinearAccClamp=<3000 3000 3000>
    componentLinearAccScale=<0.1 0.1 0.45>
    componentLinearVelScale=<0.1 0.1 0.45>
    gravityOverride {x=-980}
    physicsBodyDefinitions [{
      boundBone {boneName=collar_back_2}
      boxExtents=<10 10 10>
      constraintSetup {angularLimitsMax=<0 0 0>, angularLimitsMin=<0 -15 0>, angularTarget=<0 0 0>}
      localJointOffset=<2 0 -1>
     }]
    relativeSpaceBone {boneName=MSTR_collar_back}
    simSpaceSettings {maxAngularAcceleration=300, maxAngularVelocity=300, simSpaceAngularAlpha=0.5}
   }
   n10 = Anim Dynamics - Bone: leather_r_1(ComponentPose=@n9, Alpha=1, ExternalForce="0,0,0") {
    angularDampingOverride=1
    angularSpringConstant=500
    bAngularSpring=true
    bGravityOverrideInSimSpace=true
    bOverrideAngularDamping=true
    bUseGravityOverride=true
    simulationSpace=BoneRelative
    alphaBoolBlend {blendOption=Linear}
    boundBone {boneName=leather_r_1}
    componentAppliedLinearAccClamp=<3000 3000 3000>
    componentLinearAccScale=<0.1 0.3 0.3>
    componentLinearVelScale=<0.1 0.3 0.3>
    gravityOverride {y=-700, z=-800}
    physicsBodyDefinitions [{
      boundBone {boneName=leather_r_1}
      boxExtents=<10 10 20>
      constraintSetup {angularLimitsMax=<0 10 15>, angularLimitsMin=<0 -10 -20>, angularTarget=<0 0 0>}
      localJointOffset=<2 0.5 0>
     }]
    relativeSpaceBone {boneName=lowerarm_r}
    simSpaceSettings {maxAngularAcceleration=300, maxAngularVelocity=300, simSpaceAngularAlpha=0.5}
   }
   n11 = Anim Dynamics - Bone: leather_r_2(ComponentPose=@n10, Alpha=1, ExternalForce="0,0,0") {
    angularDampingOverride=1
    angularSpringConstant=500
    bAngularSpring=true
    bGravityOverrideInSimSpace=true
    bOverrideAngularDamping=true
    bUseGravityOverride=true
    simulationSpace=BoneRelative
    alphaBoolBlend {blendOption=Linear}
    boundBone {boneName=leather_r_2}
    componentAppliedLinearAccClamp=<3000 3000 3000>
    componentLinearAccScale=<0.1 0.3 0.3>
    componentLinearVelScale=<0.1 0.3 0.3>
    gravityOverride {y=-700, z=-600}
    physicsBodyDefinitions [{
      boundBone {boneName=leather_r_2}
      boxExtents=<10 10 20>
      constraintSetup {angularLimitsMax=<0 10 15>, angularLimitsMin=<0 -10 -20>, angularTarget=<0 0 0>}
      localJointOffset=<3 2 0>
     }]
    relativeSpaceBone {boneName=lowerarm_r}
    simSpaceSettings {maxAngularAcceleration=300, maxAngularVelocity=300, simSpaceAngularAlpha=0.5}
   }
   n12 = Anim Dynamics - Bone: leather_l_1(ComponentPose=@n11, Alpha=1, ExternalForce="0,0,0") {
    angularDampingOverride=0.9999
    angularSpringConstant=5000
    bAngularSpring=true
    bGravityOverrideInSimSpace=true
    bOverrideAngularDamping=true
    bUseGravityOverride=true
    simulationSpace=BoneRelative
    alphaBoolBlend {blendOption=Linear}
    boundBone {boneName=leather_l_1}
    componentAppliedLinearAccClamp=<3000 3000 3000>
    componentLinearAccScale=<0.1 0.3 0.3>
    componentLinearVelScale=<0.1 0.3 0.3>
    gravityOverride {y=700, z=980}
    physicsBodyDefinitions [{
      boundBone {boneName=leather_l_1}
      boxExtents=<10 10 20>
      constraintSetup {angularLimitsMax=<0 5 15>, angularLimitsMin=<0 -20 -20>, angularTarget=<0 0 0>}
      localJointOffset=<2 0.5 0>
     }]
    relativeSpaceBone {boneName=lowerarm_l}
    simSpaceSettings {maxAngularAcceleration=300, maxAngularVelocity=300, simSpaceAngularAlpha=0.5}
   }
   n13 = Anim Dynamics - Bone: leather_l_2(ComponentPose=@n12, Alpha=1, ExternalForce="0,0,0") {
    angularDampingOverride=0.9999
    angularSpringConstant=500
    bAngularSpring=true
    bGravityOverrideInSimSpace=true
    bOverrideAngularDamping=true
    bUseGravityOverride=true
    simulationSpace=BoneRelative
    alphaBoolBlend {blendOption=Linear}
    boundBone {boneName=leather_l_2}
    componentAppliedLinearAccClamp=<3000 3000 3000>
    componentLinearAccScale=<0.1 0.3 0.3>
    componentLinearVelScale=<0.1 0.3 0.3>
    gravityOverride {y=700, z=980}
    physicsBodyDefinitions [{
      boundBone {boneName=leather_l_2}
      boxExtents=<10 10 20>
      constraintSetup {angularLimitsMax=<0 5 15>, angularLimitsMin=<0 -20 -20>, angularTarget=<0 0 0>}
      localJointOffset=<3 2 0>
     }]
    relativeSpaceBone {boneName=lowerarm_l}
    simSpaceSettings {maxAngularAcceleration=300, maxAngularVelocity=300, simSpaceAngularAlpha=0.5}
   }
   n14 = Anim Dynamics - Bone: jacket_collar_dyn_r_1(ComponentPose=@n13, Alpha=1, ExternalForce="0,0,0") {
    angularDampingOverride=1
    angularSpringConstant=250
    bAngularSpring=true
    bGravityOverrideInSimSpace=true
    bLinearSpring=true
    bOverrideAngularDamping=true
    bUseGravityOverride=true
    linearSpringConstant=250
    alphaBoolBlend {blendOption=Linear}
    boundBone {boneName=jacket_collar_dyn_r_1}
    componentAppliedLinearAccClamp=<3000 3000 3000>
    componentLinearAccScale=<0.6 0.6 0.6>
    componentLinearVelScale=<0.6 0.6 0.6>
    gravityOverride {z=-980}
    physicsBodyDefinitions [{
      boundBone {boneName=jacket_collar_dyn_r_1}
      boxExtents=<10 10 10>
      constraintSetup {angularLimitsMax=<1 0 2>, angularLimitsMin=<-1 -7 -2>, angularTarget=<0 0 0>}
      localJointOffset=<0.5 0 0>
     }]
    simSpaceSettings {maxAngularAcceleration=300, maxAngularVelocity=300, simSpaceAngularAlpha=0.3}
   }
   n15 = Anim Dynamics - Bone: jacket_collar_dyn_r_2(ComponentPose=@n14, Alpha=1, ExternalForce="0,0,0") {
    angularDampingOverride=1
    angularSpringConstant=250
    bAngularSpring=true
    bGravityOverrideInSimSpace=true
    bLinearSpring=true
    bOverrideAngularDamping=true
    bUseGravityOverride=true
    linearSpringConstant=250
    alphaBoolBlend {blendOption=Linear}
    boundBone {boneName=jacket_collar_dyn_r_2}
    componentAppliedLinearAccClamp=<3000 3000 3000>
    componentLinearAccScale=<0.6 0.6 0.6>
    componentLinearVelScale=<0.6 0.6 0.6>
    gravityOverride {z=-980}
    physicsBodyDefinitions [{
      boundBone {boneName=jacket_collar_dyn_r_2}
      boxExtents=<10 10 10>
      constraintSetup {angularLimitsMax=<1 0 2>, angularLimitsMin=<-1 -15 -2>, angularTarget=<0 0 0>}
      localJointOffset=<0.5 0 0>
     }]
    simSpaceSettings {maxAngularAcceleration=300, maxAngularVelocity=300, simSpaceAngularAlpha=0.3}
   }
   n16 = Anim Dynamics - Bone: jacket_collar_dyn_l_1(ComponentPose=@n15, Alpha=1, ExternalForce="0,0,0") {
    angularDampingOverride=0.9999
    angularSpringConstant=250
    bAngularSpring=true
    bGravityOverrideInSimSpace=true
    bLinearSpring=true
    bOverrideAngularDamping=true
    bUseGravityOverride=true
    linearSpringConstant=250
    alphaBoolBlend {blendOption=Linear}
    boundBone {boneName=jacket_collar_dyn_l_1}
    componentAppliedLinearAccClamp=<3000 3000 3000>
    componentLinearAccScale=<0.6 0.6 0.6>
    componentLinearVelScale=<0.6 0.6 0.6>
    gravityOverride {z=-980}
    physicsBodyDefinitions [{
      boundBone {boneName=jacket_collar_dyn_l_1}
      boxExtents=<10 10 10>
      constraintSetup {angularLimitsMax=<1 0 2>, angularLimitsMin=<-1 -7 -2>, angularTarget=<0 0 0>}
      localJointOffset=<0.4 0 0>
     }]
    simSpaceSettings {maxAngularAcceleration=300, maxAngularVelocity=300, simSpaceAngularAlpha=0.3}
   }
   n17 = Anim Dynamics - Bone: jacket_collar_dyn_l_2(ComponentPose=@n16, Alpha=1, ExternalForce="0,0,0") {
    angularDampingOverride=0.9999
    angularSpringConstant=250
    bAngularSpring=true
    bGravityOverrideInSimSpace=true
    bLinearSpring=true
    bOverrideAngularDamping=true
    bUseGravityOverride=true
    linearSpringConstant=250
    alphaBoolBlend {blendOption=Linear}
    boundBone {boneName=jacket_collar_dyn_l_2}
    componentAppliedLinearAccClamp=<3000 3000 3000>
    componentLinearAccScale=<0.6 0.6 0.6>
    componentLinearVelScale=<0.6 0.6 0.6>
    gravityOverride {z=-980}
    physicsBodyDefinitions [{
      boundBone {boneName=jacket_collar_dyn_l_2}
      boxExtents=<10 10 10>
      constraintSetup {angularLimitsMax=<1 0 2>, angularLimitsMin=<-1 -15 -2>, angularTarget=<0 0 0>}
      localJointOffset=<0.4 0 0>
     }]
    simSpaceSettings {maxAngularAcceleration=300, maxAngularVelocity=300, simSpaceAngularAlpha=0.3}
   }
   n18 = Anim Dynamics - Bone: star_l(ComponentPose=@n17, Alpha=1, ExternalForce="0,0,0") {
    bUseGravityOverride=true
    alphaBoolBlend {blendOption=Linear}
    boundBone {boneName=star_l}
    componentAppliedLinearAccClamp=<3000 3000 3000>
    componentLinearAccScale=<0.5 0.5 0.5>
    componentLinearVelScale=<0.5 0.5 0.5>
    gravityOverride {z=-980}
    physicsBodyDefinitions [{
      boundBone {boneName=star_l}
      boxExtents=<5 5 10>
      constraintSetup {angularLimitsMax=<50 0 0>, angularLimitsMin=<-50 -50 0>, angularTarget=<0 0 0>}
      localJointOffset=<0 0 -1>
     }]
    simSpaceSettings {maxAngularAcceleration=300, maxAngularVelocity=300, simSpaceAngularAlpha=1}
   }
   n19 = Anim Dynamics - Bone: star_r(ComponentPose=@n18, Alpha=1, ExternalForce="0,0,0") {
    bUseGravityOverride=true
    alphaBoolBlend {blendOption=Linear}
    boundBone {boneName=star_r}
    componentAppliedLinearAccClamp=<3000 3000 3000>
    componentLinearAccScale=<0.5 0.5 0.5>
    componentLinearVelScale=<0.5 0.5 0.5>
    gravityOverride {z=-980}
    physicsBodyDefinitions [{
      boundBone {boneName=star_r}
      boxExtents=<5 5 10>
      constraintSetup {angularLimitsMax=<50 0 0>, angularLimitsMin=<-50 -50 0>, angularTarget=<0 0 0>}
      localJointOffset=<0 0 -1.3>
     }]
    simSpaceSettings {maxAngularAcceleration=300, maxAngularVelocity=300, simSpaceAngularAlpha=1}
   }
   n20 = "Transform (Modify) Bone - Bone: jacket_flap_back_2"!(ComponentPose=@n19, bAlphaBoolEnabled:bool=Get bIsDBNO(), Translation="0,0,0", Rotation="0,0,-8", Scale="1,1,1") -> (Pose) {
    alphaInputType=Bool
    rotationMode=BMM_Additive
    alphaBoolBlend {blendOption=Linear}
    boneToModify {boneName=jacket_flap_back_2}
    rotation {roll=-8}
   }
   n21 = "Transform (Modify) Bone - Bone: attach_backpack"!(
    ComponentPose=@n20
    bAlphaBoolEnabled:bool=Get Is Saxo Pop Back Closer() -> (Is Saxo Pop Back Closer:bool)
    Translation="1.175,1.420,0.344"
    Rotation="0,0,0"
    Scale="1,1,1"
   ) -> (Pose) {
    alphaInputType=Bool
    translationMode=BMM_Additive
    translationSpace=BCS_BoneSpace
    alphaBoolBlend {blendOption=Linear}
    boneToModify {boneName=attach_backpack}
    translation=<1.175 1.42 0.344>
   }
   n22 = Control Rig(Source=Component To Local(ComponentPose=@n21)) {
    alphaBoolBlend {blendOption=Linear}
    controlRigAssetReference {blueprintRigClass=&/F_MED_SaxoPop_CtrlRig_clavicles.F_MED_SaxoPop_CtrlRig_clavicles_C}
   }
   Output Pose(Result=@n22) {layerGroup=None}
  ]}
 EventGraph {nodes [
   Event BlueprintUpdateAnimation()
   > n1 = Gravity Override State Function DEPRECATED(
    CharacterPartAnimInstance=Self-Reference()
    GameTimeInSeconds=GetGameTimeInSeconds()
    GravityOverrideParams=Get Gravity Override Params()
    RBAN Params=Get RBAN Params()
   )
   > n2 = Set Gravity Override(Gravity Override=@n1)
   > n3 = Set Gravity override animDyn(Gravity override animDyn:Vector=@n2) -> (Output_Get:Vector)
   > Branch(Condition=Get bIsSkydiveFloating())
   > Set Gravity override animDyn(Gravity override animDyn=Rotate Vector Around Axis!(InVect:Vector=@n3, AngleDeg:float=-105, Axis:Vector="-1,0,0") -> (ReturnValue:Vector))
   n4 = Event OnCharacterCustomizationLoaded()
   Set Is Saxo Pop Back Closer(execute=[@n4.then, @n10.then], Is Saxo Pop Back Closer:bool=false) -> (Output_Get:bool)
   > n5 = Sequence()
   > n6 = Branch(Condition=Get bIsInFrontEnd())
   > Set Is frontend alpha(Is frontend alpha:double=0.5) -> (Output_Get:double)
   Set Is frontend alpha(execute=@n6.else, Is frontend alpha=1)
   n7 = For Each Loop with Break!(Exec:Exec=@n5.then_1, Array:Array of Custom Character Part Object References=@n4.CustomParts, Break:Exec=@n9.then) -> (LoopBody:Exec, Array Element:Custom Character Part Object Reference, Array Index:int, Completed:Exec)
   n8 = CheckForBackbling(LongBackblingThreshold=0)
   Set no backpack(
    execute=@n7.Completed
    no backpack:bool=NOT Boolean(A=OR Boolean!(A=@n8, B=@n8.bIsBackblingACape, C:bool=@n8.bIsBackblingLong) -> (ReturnValue))
   ) -> (Output_Get:bool)
   Branch(execute=@n7.LoopBody, Condition="Equal (Object)"(A=@n7, B=/BRCosmetics/Characters/CharacterParts/Backpacks/CP_Backpack_SaxoPop))
   > n9 = Set Is Saxo Pop Back Closer(Is Saxo Pop Back Closer=true)
   n10 = Event BlueprintInitializeAnimation()
  ]}
]
variables [
 Gravity Override {type=Vector}
 Gravity override animDyn {type=Vector}
 Gravity Override Params {type=GravityOverrideParamsStruct, default {skydive_Parachuting {gravityOverride=<0 -400 -980>}}}
 in_jump {type=number}
 Is frontend alpha {type=number}
 Is Saxo Pop Back Closer {type=boolean}
 no backpack {type=boolean}
 RBAN Params {
  type=RBANParamsStruct
  default {
   dBNO {simSpaceSettings {
     maxAngularAcceleration=1000
     maxAngularVelocity=3
     maxLinearAcceleration=300
     maxLinearVelocity=700
     velocityScaleZ=0
     externalLinearDragV=<0 0.5 0>
    }}
   dBNO_Carried {simSpaceSettings {
     maxAngularAcceleration=1000
     maxAngularVelocity=3
     maxLinearAcceleration=300
     maxLinearVelocity=700
     velocityScaleZ=0
     externalLinearDragV=<0 0.5 0>
    }}
   emoteOrMelee {simSpaceSettings {
     maxAngularAcceleration=1000
     maxAngularVelocity=3
     maxLinearAcceleration=300
     maxLinearVelocity=700
     velocityScaleZ=0
     externalLinearDragV=<0 0.5 0>
    }}
   front_End {simSpaceSettings {
     maxAngularAcceleration=1000
     maxAngularVelocity=3
     maxLinearAcceleration=300
     maxLinearVelocity=700
     velocityScaleZ=0
     externalLinearDragV=<0 0.5 0>
    }}
   onGround_Crouch_Moving {simSpaceSettings {
     maxAngularAcceleration=1000
     maxAngularVelocity=3
     maxLinearAcceleration=300
     maxLinearVelocity=700
     velocityScaleZ=0
     externalLinearDragV=<0 0.5 0>
    }}
   onGround_Falling {simSpaceSettings {
     maxAngularAcceleration=1000
     maxAngularVelocity=3
     maxLinearAcceleration=300
     maxLinearVelocity=700
     velocityScaleZ=0
     externalLinearDragV=<0 0.5 0>
    }}
   onGround_Moving {simSpaceSettings {
     maxAngularAcceleration=1000
     maxAngularVelocity=3
     maxLinearAcceleration=300
     maxLinearVelocity=700
     velocityScaleZ=0
     externalLinearDragV=<0 0.5 0>
    }}
   onGround_Standing {simSpaceSettings {
     maxAngularAcceleration=1000
     maxAngularVelocity=3
     maxLinearAcceleration=300
     maxLinearVelocity=700
     velocityScaleZ=0
     externalLinearDragV=<0 0.5 0>
    }}
   rail_Grinding {simSpaceSettings {
     maxAngularAcceleration=1000
     maxAngularVelocity=3
     maxLinearAcceleration=300
     maxLinearVelocity=700
     velocityScaleZ=0
     worldAlpha=0.1
     externalLinearDragV=<0 0.5 0>
    }}
   riding_Moving {simSpaceSettings {
     maxAngularAcceleration=1000
     maxAngularVelocity=3
     maxLinearAcceleration=300
     maxLinearVelocity=700
     velocityScaleZ=0
     externalLinearDragV=<0 0.5 0>
    }}
   riding_Moving_Targeting {simSpaceSettings {
     maxAngularAcceleration=1000
     maxAngularVelocity=3
     maxLinearAcceleration=300
     maxLinearVelocity=700
     velocityScaleZ=0
     externalLinearDragV=<0 0.5 0>
    }}
   riding_Standing {simSpaceSettings {
     maxAngularAcceleration=1000
     maxAngularVelocity=3
     maxLinearAcceleration=300
     maxLinearVelocity=700
     velocityScaleZ=0
     externalLinearDragV=<0 0.5 0>
    }}
   skydive_DivingDown {simSpaceSettings {
     maxAngularAcceleration=1000
     maxAngularVelocity=3
     maxLinearAcceleration=300
     maxLinearVelocity=700
     velocityScaleZ=0
     externalLinearDragV=<0 0.5 0>
    }}
   skydive_FloatingDown {
    componentLinearAccScale=<0 0.5 0>
    componentLinearVelScale=<0 0.5 0>
    simSpaceSettings {
     maxAngularAcceleration=1000
     maxAngularVelocity=3
     maxLinearAcceleration=300
     maxLinearVelocity=700
     velocityScaleZ=0
     externalLinearDragV=<0 0.5 0>
    }
   }
   skydive_GoingUp {simSpaceSettings {
     maxAngularAcceleration=1000
     maxAngularVelocity=3
     maxLinearAcceleration=300
     maxLinearVelocity=700
     velocityScaleZ=0
     externalLinearDragV=<0 0.5 0>
    }}
   skydive_Parachuting {
    componentLinearAccScale=<0 0.1 0.3>
    componentLinearVelScale=<0 0.1 0.3>
    simSpaceSettings {
     maxAngularAcceleration=1000
     maxAngularVelocity=3
     maxLinearAcceleration=300
     maxLinearVelocity=700
     velocityScaleZ=0
     externalLinearDragV=<0 0.5 0>
    }
   }
   skydive_Surfing {simSpaceSettings {
     maxAngularAcceleration=1000
     maxAngularVelocity=3
     maxLinearAcceleration=300
     maxLinearVelocity=700
     velocityScaleZ=0
     externalLinearDragV=<0 0.5 0>
    }}
   swimming {simSpaceSettings {
     maxAngularAcceleration=1000
     maxAngularVelocity=3
     maxLinearAcceleration=300
     maxLinearVelocity=700
     velocityScaleZ=0
     externalLinearDragV=<0 0.5 0>
    }}
   tactical_Sprint {simSpaceSettings {
     maxAngularAcceleration=1000
     maxAngularVelocity=3
     maxLinearAcceleration=300
     maxLinearVelocity=700
     velocityScaleZ=0
     externalLinearDragV=<0 0.5 0>
    }}
   ziplining {simSpaceSettings {
     maxAngularAcceleration=1000
     maxAngularVelocity=3
     maxLinearAcceleration=300
     maxLinearVelocity=700
     velocityScaleZ=0
     externalLinearDragV=<0 0.5 0>
    }}
  }
 }
]
