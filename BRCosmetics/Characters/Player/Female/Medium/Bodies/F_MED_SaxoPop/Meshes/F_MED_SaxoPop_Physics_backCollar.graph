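# Physics asset for the back collar of the F_MED_SaxoPop female medium player cosmetic body.
# A small sphere body on MSTR_collar_back roots a two-link chain of identical capsules on collar_back_1 and collar_back_2.
# Two constraints join the bodies in sequence with matching tight swing and twist limits, restricting the collar to slight bending.
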PhysicsAsset /BRCosmetics/Characters/Player/Female/Medium/Bodies/F_MED_SaxoPop/Meshes/F_MED_SaxoPop_Physics_backCollar
bodies {
 MSTR_collar_back [{shapeType=Sphere, radius=0.505, shapeName=MSTR_collar_back_sphere}]
 collar_back_1 [{shapeType=Capsule, length=1.4357, radius=1.2989, shapeName=collar_back_1_capsule, center=<1.7954 0 0>, rotation=<rot -90 0 0>}]
 collar_back_2 [{shapeType=Capsule, length=1.4357, radius=1.2989, shapeName=collar_back_2_capsule, center=<1.7954 0 0>, rotation=<rot -90 0 0>}]
}
constraints [
 {bone1Name=collar_back_2, bone2Name=collar_back_1, swing1LimitDegrees=5, swing2LimitDegrees=11, twistLimitDegrees=5}
 {bone1Name=collar_back_1, bone2Name=MSTR_collar_back, swing1LimitDegrees=5, swing2LimitDegrees=11, twistLimitDegrees=5}
]
