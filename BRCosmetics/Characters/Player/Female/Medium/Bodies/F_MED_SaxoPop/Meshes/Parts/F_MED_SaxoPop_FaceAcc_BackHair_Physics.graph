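# Physics asset for the back-hair mesh part of the medium female SaxoPop character, built entirely from capsule bodies for hair and upper-body bones.
# Back, left and right dyn_hair chains start at small MSTR roots on the head, with left and right each forking into an extra 01a link; two three-link side chains also join the head.
# Links within chains have angular limits, while chain roots and most torso joints are free.
# Head, neck, spine, clavicle and upper-arm capsules provide body volume around the hair.
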PhysicsAsset /BRCosmetics/Characters/Player/Female/Medium/Bodies/F_MED_SaxoPop/Meshes/Parts/F_MED_SaxoPop_FaceAcc_BackHair_Physics
bodies {
 clavicle_l [{shapeType=Capsule, length=1.9169, radius=8.8144, center=<6.4524 -0.0135 -2.3802>, rotation=<rot 85.606 0 0>}]
 clavicle_r [{shapeType=Capsule, length=1.6669, radius=8.5644, center=<-7.6286 0.2984 2.2116>, rotation=<rot 85.606 0 0>}]
 dyn_hair_bk_00 [{shapeType=Capsule, length=3.3125, radius=4.4375, shapeName=dyn_Hair_Back_03_capsule, center=<3.1306 -3.0467 0>, rotation=<rot 90 0.0001 0>}]
 dyn_hair_bk_01 [{shapeType=Capsule, length=4.25, radius=4.75, shapeName=dyn_Hair_Back_04_capsule, center=<3.5858 -3.2594 0>, rotation=<rot 90 8.4376 0>}]
 dyn_hair_bk_02 [{shapeType=Capsule, length=5.5, radius=6.0414, shapeName=dyn_hair_bk_02_capsule, center=<1.834 -2.5911 0>, rotation=<rot -90 5.6253 0>}]
 dyn_hair_bk_MSTR [{shapeType=Capsule, length=0.505, radius=0.505, shapeName=dyn_Hair_Back_MSTR_capsule}]
 dyn_hair_l_00 [{shapeType=Capsule, length=0.3125, radius=5.3125, shapeName=dyn_Hair_BackL_03_capsule, center=<3.1306 -0.7302 0>, rotation=<rot 90 0.0001 0>}]
 dyn_hair_l_01 [{shapeType=Capsule, length=6.125, radius=4.25, shapeName=dyn_Hair_BackL_04_capsule, center=<5.8375 -1.4314 0>, rotation=<rot 90 8.4379 0>}]
 dyn_hair_l_01a [{shapeType=Capsule, length=0, radius=4.25, shapeName=dyn_hair_l_01a_capsule, center=<-1.6843 -0.2608 0>}]
 dyn_hair_l_MSTR [{shapeType=Capsule, length=0.505, radius=0.505, shapeName=dyn_Hair_BackL_MSTR_capsule}]
 dyn_hair_r_00 [{
   shapeType=Capsule
   length=4.5
   radius=5.3125
   shapeName=dyn_Hair_BackL_03_capsule
   center=<6.6405 -2.2487 -2.0107>
   rotation=<rot 90 5.6252 0>
  }]
 dyn_hair_r_01 [{shapeType=Capsule, length=5.375, radius=5, shapeName=dyn_Hair_BackL_04_capsule, center=<9.8116 -1.6848 0>, rotation=<rot 90 22.5002 0>}]
 dyn_hair_r_01a [{
   shapeType=Capsule
   length=5.5
   radius=4
   shapeName=dyn_hair_r_01a_capsule
   center=<0.9794 0.8198 -0.0066>
   rotation=<rot -1.8883 42.1533 7.7097>
  }]
 dyn_hair_r_MSTR [{shapeType=Capsule, length=0.505, radius=0.505, shapeName=dyn_Hair_BackR_MSTR_capsule}]
 hair_side_l_1 [{shapeType=Capsule, length=3.5404, radius=2.4829, shapeName=hair_side_l_1_capsule, center=<3.9982 0 1.891>, rotation=<rot 64.6875 180 180>}]
 hair_side_l_2 [{
   shapeType=Capsule
   length=3.2724
   radius=3.1544
   shapeName=hair_side_r_2_capsule
   center=<2.4698 1.0521 -1.0463>
   rotation=<rot 70.3124 0 5.625>
  }]
 hair_side_l_3 [{
   shapeType=Capsule
   length=1.6384
   radius=3.1404
   shapeName=hair_side_r_3_capsule
   center=<3.2344 0.9373 -4.1281>
   rotation=<rot 73.125 0.0001 0.0001>
  }]
 hair_side_r_1 [{shapeType=Capsule, length=2.5018, radius=2.3242, shapeName=hair_side_r_1_capsule, center=<3.9982 0 1.891>, rotation=<rot 64.6875 -180 -180>}]
 hair_side_r_2 [{shapeType=Capsule, length=2.7724, radius=3.6544, shapeName=hair_side_r_2_capsule, center=<3.4242 0 3.1047>, rotation=<rot 64.6875 -180 -180>}]
 hair_side_r_3 [{
   shapeType=Capsule
   length=0.8884
   radius=3.8904
   shapeName=hair_side_r_3_capsule
   center=<2.6365 0 4.7701>
   rotation=<rot 64.6875 -179.9999 -180>
  }]
 head [{shapeType=Capsule, length=3.0796, radius=6.7593, shapeName=head_capsule, center=<2.3025 0.855 -0.5075>, rotation=<rot 90 0 5.0001>}]
 neck_01 [{shapeType=Capsule, length=4.755, radius=5.0675, shapeName=neck_01_capsule, center=<0 1.3702 0>, rotation=<rot 0 -90.0003 -90>}]
 neck_02 [{shapeType=Capsule, length=6.5, radius=5.5, shapeName=neck_02_capsule, center=<-2.1964 0 0>, rotation=<rot -90 -0.0001 0>}]
 spine_04 [{shapeType=Capsule, length=4.5941, radius=10.0751, shapeName=spine_04_capsule, center=<2.5892 0.4958 0>, rotation=<rot 0 -92.6847 0>}]
 spine_05 [{shapeType=Capsule, length=15.7816, radius=10.0126, shapeName=spine_04_capsule, center=<2.1177 0.7682 0>, rotation=<rot 0 -87.0597 0>}]
 upperarm_l [{shapeType=Capsule, length=13.4229, radius=4.1429, shapeName=upperarm_r_capsule, center=<5.2485 1.0639 -0.2913>, rotation=<rot 0 -90 90>}]
 upperarm_r [{shapeType=Capsule, length=19.1729, radius=5.8929, shapeName=upperarm_r_capsule, center=<-11.013 1.25 -1.2498>, rotation=<rot 0 -90 90>}]
}
constraints [
 {
  bone1Name=head
  bone2Name=neck_02
  swing1LimitDegrees=45
  swing1Motion=Free
  swing2LimitDegrees=45
  swing2Motion=Free
  twistLimitDegrees=45
  twistMotion=Free
 }
 {
  bone1Name=spine_05
  bone2Name=spine_04
  swing1LimitDegrees=45
  swing1Motion=Free
  swing2LimitDegrees=45
  swing2Motion=Free
  twistLimitDegrees=45
  twistMotion=Free
 }
 {
  bone1Name=clavicle_l
  bone2Name=spine_05
  swing1LimitDegrees=45
  swing1Motion=Free
  swing2LimitDegrees=45
  swing2Motion=Free
  twistLimitDegrees=45
  twistMotion=Free
 }
 {
  bone1Name=upperarm_l
  bone2Name=clavicle_l
  swing1LimitDegrees=45
  swing1Motion=Free
  swing2LimitDegrees=45
  swing2Motion=Free
  twistLimitDegrees=45
  twistMotion=Free
 }
 {
  bone1Name=clavicle_r
  bone2Name=spine_05
  swing1LimitDegrees=45
  swing1Motion=Free
  swing2LimitDegrees=45
  swing2Motion=Free
  twistLimitDegrees=45
  twistMotion=Free
 }
 {
  bone1Name=upperarm_r
  bone2Name=clavicle_r
  swing1LimitDegrees=45
  swing1Motion=Free
  swing2LimitDegrees=45
  swing2Motion=Free
  twistLimitDegrees=45
  twistMotion=Free
 }
 {
  bone1Name=dyn_hair_bk_MSTR
  bone2Name=head
  swing1LimitDegrees=45
  swing1Motion=Free
  swing2LimitDegrees=45
  swing2Motion=Free
  twistLimitDegrees=45
  twistMotion=Free
 }
 {bone1Name=dyn_hair_bk_00, bone2Name=dyn_hair_bk_MSTR, swing1LimitDegrees=50, swing2LimitDegrees=15, twistLimitDegrees=10}
 {bone1Name=dyn_hair_bk_01, bone2Name=dyn_hair_bk_00, swing1LimitDegrees=50, swing2LimitDegrees=15, twistLimitDegrees=10}
 {
  bone1Name=dyn_hair_l_MSTR
  bone2Name=head
  swing1LimitDegrees=45
  swing1Motion=Free
  swing2LimitDegrees=45
  swing2Motion=Free
  twistLimitDegrees=45
  twistMotion=Free
 }
 {bone1Name=dyn_hair_l_00, bone2Name=dyn_hair_l_MSTR, swing1LimitDegrees=50, swing2LimitDegrees=15, twistLimitDegrees=10}
 {bone1Name=dyn_hair_l_01, bone2Name=dyn_hair_l_00, swing1LimitDegrees=50, swing2LimitDegrees=15, twistLimitDegrees=10}
 {
  bone1Name=dyn_hair_r_MSTR
  bone2Name=head
  swing1LimitDegrees=45
  swing1Motion=Free
  swing2LimitDegrees=45
  swing2Motion=Free
  twistLimitDegrees=45
  twistMotion=Free
 }
 {bone1Name=dyn_hair_r_00, bone2Name=dyn_hair_r_MSTR, swing1LimitDegrees=50, swing2LimitDegrees=15, twistLimitDegrees=10}
 {bone1Name=dyn_hair_r_01, bone2Name=dyn_hair_r_00, swing1LimitDegrees=50, swing2LimitDegrees=15, twistLimitDegrees=10}
 {
  bone1Name=neck_01
  bone2Name=spine_05
  swing1LimitDegrees=45
  swing1Motion=Free
  swing2LimitDegrees=45
  swing2Motion=Free
  twistLimitDegrees=45
  twistMotion=Free
 }
 {bone1Name=dyn_hair_r_01a, bone2Name=dyn_hair_r_00, swing1LimitDegrees=10, swing2LimitDegrees=10, twistLimitDegrees=10}
 {bone1Name=dyn_hair_l_01a, bone2Name=dyn_hair_l_00, swing1LimitDegrees=10, swing2LimitDegrees=10, twistLimitDegrees=10}
 {bone1Name=dyn_hair_bk_02, bone2Name=dyn_hair_bk_01, swing1LimitDegrees=45, swing2LimitDegrees=15, twistLimitDegrees=10}
 {bone1Name=hair_side_r_2, bone2Name=hair_side_r_1, swing1LimitDegrees=5, swing2LimitDegrees=45, twistLimitDegrees=0}
 {bone1Name=hair_side_r_3, bone2Name=hair_side_r_2, swing1LimitDegrees=5, swing2LimitDegrees=45, twistLimitDegrees=0}
 {bone1Name=hair_side_l_3, bone2Name=hair_side_l_2, swing1LimitDegrees=5, swing2LimitDegrees=45, twistLimitDegrees=0}
 {bone1Name=hair_side_l_2, bone2Name=hair_side_l_1, swing1LimitDegrees=5, swing2LimitDegrees=45, twistLimitDegrees=0}
 {
  bone1Name=hair_side_l_1
  bone2Name=head
  swing1LimitDegrees=45
  swing1Motion=Free
  swing2LimitDegrees=45
  swing2Motion=Free
  twistLimitDegrees=45
  twistMotion=Free
 }
 {
  bone1Name=hair_side_r_1
  bone2Name=head
  swing1LimitDegrees=45
  swing1Motion=Free
  swing2LimitDegrees=45
  swing2Motion=Free
  twistLimitDegrees=45
  twistMotion=Free
 }
 {
  bone1Name=neck_02
  bone2Name=neck_01
  swing1LimitDegrees=45
  swing1Motion=Locked
  swing2LimitDegrees=45
  swing2Motion=Locked
  twistLimitDegrees=45
  twistMotion=Locked
 }
]
