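# Control Rig for the SaxoPop female character's FaceAcc backpack variant, procedurally correcting hair, ear and earring bones as the body poses.
# Side hair control points blend between head, neck and upper body, take pose-reader offsets, then shape hair chains through fitted Hermite splines.
# Dynamic hair masters follow spine and head, shifted by head and right-arm pose readers, multiplied combination alphas and crouch-scaled weights.
# A ShouldersUp branch adds clavicle-driven hair offsets.
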ControlRigBlueprint /BRCosmetics/Characters/Player/Female/Medium/Bodies/F_MED_SaxoPop/Meshes/Parts/F_MED_SaxoPop_FaceAcc_CtrlRig_Backpack
graphs [
 ContainedGraph {nodes [
   n1 = RigVMFunctionEntryNode()
   SequenceExecution_12 = &RigVMFunction_Sequence(ExecuteContext=@n1.ExecuteContext)
   SequenceExecution_12_1 = &RigVMFunction_Sequence(ExecuteContext=@SequenceExecution_12.B)
   SequenceExecution_12_2 = &RigVMFunction_Sequence(ExecuteContext=@SequenceExecution_12_1.B)
   SequenceExecution_12_3 = &RigVMFunction_Sequence(ExecuteContext=@SequenceExecution_12_2.B)
   RigVMFunctionReturnNode(
    A=@SequenceExecution_12.A
    B=@SequenceExecution_12_1.A
    C=@SequenceExecution_12_2.A
    D=@SequenceExecution_12_3.A
    E=@SequenceExecution_12_3.B
   )
  ]}
 ContainedGraph {nodes [
   n1 = RigVMFunctionEntryNode()
   SequenceExecution_12 = &RigVMFunction_Sequence(ExecuteContext=@n1.ExecuteContext)
   SequenceExecution_12_1 = &RigVMFunction_Sequence(ExecuteContext=@SequenceExecution_12.B)
   SequenceExecution_12_2 = &RigVMFunction_Sequence(ExecuteContext=@SequenceExecution_12_1.B)
   SequenceExecution_12_3 = &RigVMFunction_Sequence(ExecuteContext=@SequenceExecution_12_2.B)
   SequenceExecution_1 = &RigVMFunction_Sequence(ExecuteContext=@SequenceExecution_12_3.B)
   RigVMFunctionReturnNode(
    E=@SequenceExecution_1.A
    F=@SequenceExecution_1.B
    A=@SequenceExecution_12.A
    B=@SequenceExecution_12_1.A
    C=@SequenceExecution_12_2.A
    D=@SequenceExecution_12_3.A
   )
  ]}
 ContainedGraph {nodes [
   n1 = RigVMFunctionEntryNode()
   SequenceExecution_4 = &RigVMFunction_Sequence(ExecuteContext=@n1.ExecuteContext)
   SequenceExecution_4_1 = &RigVMFunction_Sequence(ExecuteContext=@SequenceExecution_4.B)
   SequenceExecution_4_2 = &RigVMFunction_Sequence(ExecuteContext=@SequenceExecution_4_1.B)
   RigVMFunctionReturnNode(A=@SequenceExecution_4.A, B=@SequenceExecution_4_1.A, C=@SequenceExecution_4_2.A, D=@SequenceExecution_4_2.B)
  ]}
 ContainedGraph {nodes [
   n1 = RigVMFunctionEntryNode()
   SequenceExecution_3 = &RigVMFunction_Sequence(ExecuteContext=@n1.ExecuteContext)
   SequenceExecution_3_1 = &RigVMFunction_Sequence(ExecuteContext=@SequenceExecution_3.B)
   SequenceExecution_3_2 = &RigVMFunction_Sequence(ExecuteContext=@SequenceExecution_3_1.B)
   RigVMFunctionReturnNode(A=@SequenceExecution_3.A, B=@SequenceExecution_3_1.A, C=@SequenceExecution_3_2.A, D=@SequenceExecution_3_2.B)
  ]}
 ContainedGraph {nodes [
   n1 = RigVMFunctionEntryNode()
   SequenceExecution_2 = &RigVMFunction_Sequence(ExecuteContext=@n1.ExecuteContext)
   SequenceExecution_2_1 = &RigVMFunction_Sequence(ExecuteContext=@SequenceExecution_2.B)
   SequenceExecution_2_2 = &RigVMFunction_Sequence(ExecuteContext=@SequenceExecution_2_1.B)
   RigVMFunctionReturnNode(A=@SequenceExecution_2.A, B=@SequenceExecution_2_1.A, C=@SequenceExecution_2_2.A, D=@SequenceExecution_2_2.B)
  ]}
 ContainedGraph {nodes [
   n1 = RigVMFunctionEntryNode()
   SequenceExecution_7 = &RigVMFunction_Sequence(ExecuteContext=@n1.ExecuteContext)
   SequenceExecution_7_1 = &RigVMFunction_Sequence(ExecuteContext=@SequenceExecution_7.B)
   SequenceExecution_7_2 = &RigVMFunction_Sequence(ExecuteContext=@SequenceExecution_7_1.B)
   RigVMFunctionReturnNode(A=@SequenceExecution_7.A, B=@SequenceExecution_7_1.A, C=@SequenceExecution_7_2.A, D=@SequenceExecution_7_2.B)
  ]}
 ContainedGraph {nodes [
   n1 = RigVMFunctionEntryNode()
   SequenceExecution = &RigVMFunction_Sequence(ExecuteContext=@n1.ExecuteContext)
   SequenceExecution_1 = &RigVMFunction_Sequence(ExecuteContext=@SequenceExecution.B)
   SequenceExecution_2 = &RigVMFunction_Sequence(ExecuteContext=@SequenceExecution_1.B)
   RigVMFunctionReturnNode(A=@SequenceExecution.A, B=@SequenceExecution_1.A, C=@SequenceExecution_2.A, D=@SequenceExecution_2.B)
  ]}
 ContainedGraph {nodes [
   n1 = RigVMFunctionEntryNode()
   SequenceExecution_8 = &RigVMFunction_Sequence(ExecuteContext=@n1.ExecuteContext)
   SequenceExecution_8_1 = &RigVMFunction_Sequence(ExecuteContext=@SequenceExecution_8.B)
   SequenceExecution_8_2 = &RigVMFunction_Sequence(ExecuteContext=@SequenceExecution_8_1.B)
   RigVMFunctionReturnNode(A=@SequenceExecution_8.A, B=@SequenceExecution_8_1.A, C=@SequenceExecution_8_2.A, D=@SequenceExecution_8_2.B)
  ]}
 ContainedGraph {nodes [
   n1 = RigVMFunctionEntryNode()
   SequenceExecution_6_1 = &RigVMFunction_Sequence(ExecuteContext=@n1.ExecuteContext)
   SequenceExecution_6_1_1 = &RigVMFunction_Sequence(ExecuteContext=@SequenceExecution_6_1.B)
   SequenceExecution_6_1_2 = &RigVMFunction_Sequence(ExecuteContext=@SequenceExecution_6_1_1.B)
   RigVMFunctionReturnNode(A=@SequenceExecution_6_1.A, B=@SequenceExecution_6_1_1.A, C=@SequenceExecution_6_1_2.A, D=@SequenceExecution_6_1_2.B)
  ]}
 New Function_ContainedGraph {nodes [
   RigVMFunctionEntryNode()
   > RigVMFunction_Sequence = RigVMAggregateNode()
   RigVMFunction_Sequence_1 = RigVMAggregateNode(ExecuteContext=@RigVMFunction_Sequence.A)
   &RigUnit_ParentConstraint(
    AdvancedSettings=(InterpolationType=Average,RotationOrderForFilter=XZY)
    Child=(Type=Bone,Name="hair_side_r_pnt_2")
    Parents=((Item=(Type=Bone,Name="clavicle_r"),Weight=0.017),(Item=(Type=Bone,Name="head"),Weight=0.73),(Item=(Type=Bone,Name="neck_01"),Weight=0.052),(Item=(Type=Bone,Name="neck_02"),Weight=0.094))
    Weight=1
    ExecutePin=@RigVMFunction_Sequence_1.A
   )
   > &RigUnit_ParentConstraint(
    AdvancedSettings=(InterpolationType=Average,RotationOrderForFilter=XZY)
    Child=(Type=Bone,Name="hair_side_r_pnt_3")
    Parents=((Item=(Type=Bone,Name="clavicle_r"),Weight=0.138),(Item=(Type=Bone,Name="head"),Weight=0.219),(Item=(Type=Bone,Name="neck_01"),Weight=0.151),(Item=(Type=Bone,Name="neck_02"),Weight=0.141),(Item=(Type=Bone,Name="spine_05"),Weight=0.351))
    Weight=1
   )
   > &RigUnit_ParentConstraint(
    AdvancedSettings=(InterpolationType=Average,RotationOrderForFilter=XZY)
    Child=(Type=Bone,Name="hair_side_l_pnt_2")
    Parents=((Item=(Type=Bone,Name="clavicle_l"),Weight=0.017),(Item=(Type=Bone,Name="head"),Weight=0.73),(Item=(Type=Bone,Name="neck_01"),Weight=0.052),(Item=(Type=Bone,Name="neck_02"),Weight=0.094))
    Weight=1
   )
   n1 = &RigUnit_SphericalPoseReader(
    ActiveRegionScaleFactors=(PositiveWidth=2,NegativeWidth=2,PositiveHeight=1,NegativeHeight=1)
    ActiveRegionSize=0.06
    Debug=(bDrawDebug=false,bDraw2D=False,bDrawLocalAxes=False,DebugScale=25,DebugSegments=20,DebugThickness=0.25)
    DriverAxis=(X=0,Y=-1,Z=0)
    DriverItem=(Type=Bone,Name="head")
    FalloffRegionScaleFactors=(PositiveWidth=1,NegativeWidth=1,PositiveHeight=1,NegativeHeight=1)
    FalloffSize=0.42
    FlipHeightScaling=False
    FlipWidthScaling=False
    OptionalParentItem=(Type=Bone,Name="spine_05")
    RotationOffset=(X=0,Y=0,Z=0)
    ExecutePin=@RigVMFunction_Sequence_1.B
   )
   > n2 = &RigUnit_SphericalPoseReader(
    ActiveRegionScaleFactors=(PositiveWidth=1,NegativeWidth=1,PositiveHeight=1,NegativeHeight=1)
    ActiveRegionSize=0.14
    Debug=(bDrawDebug=false,bDraw2D=False,bDrawLocalAxes=False,DebugScale=25,DebugSegments=20,DebugThickness=0.25)
    DriverAxis=(X=0,Y=-1,Z=0)
    DriverItem=(Type=Bone,Name="head")
    FalloffRegionScaleFactors=(PositiveWidth=1,NegativeWidth=1,PositiveHeight=1,NegativeHeight=1)
    FalloffSize=0.42
    FlipHeightScaling=False
    FlipWidthScaling=False
    OptionalParentItem=(Type=Bone,Name="spine_05")
    RotationOffset=(X=180,Y=0,Z=0)
   )
   > RigUnit_OffsetTransformForItem = &RigUnit_OffsetTransformForItem(
    Item=(Type=Bone,Name="hair_side_l_pnt_2")
    OffsetTransform=(Rotation=(X=0,Y=0,Z=0,W=1),Translation=(X=0,Y=-3,Z=1.5),Scale3D=(X=1,Y=1,Z=1))
    bPropagateToChildren=True
    Weight=@n1.OutputParam
   )
   > &RigUnit_OffsetTransformForItem(
    Item=(Type=Bone,Name="MSTR_ear_ring_l")
    OffsetTransform=(Rotation=(X=0.0601,Y=-0.4023,Z=0.135,W=0.9035),Translation=(X=0,Y=0,Z=0),Scale3D=(X=1,Y=1,Z=1))
    bPropagateToChildren=True
    Weight=@n1.OutputParam
   )
   > &RigUnit_OffsetTransformForItem(
    Item=(Type=Bone,Name="hair_side_l_pnt_2")
    OffsetTransform=(Rotation=(X=0,Y=0,Z=0,W=1),Translation=(X=2,Y=0,Z=0),Scale3D=(X=1,Y=1,Z=1))
    bPropagateToChildren=True
    Weight=@n2.OutputParam
   )
   > &RigUnit_OffsetTransformForItem(
    Item=(Type=Bone,Name="hair_side_r_pnt_2")
    OffsetTransform=(Rotation=(X=0,Y=0,Z=0,W=1),Translation=(X=-1,Y=1,Z=-1),Scale3D=(X=1,Y=1,Z=1))
    bPropagateToChildren=True
    Weight=@n2.OutputParam
   )
   > &RigUnit_OffsetTransformForItem(
    Item=(Type=Bone,Name="MSTR_ear_ring_r")
    OffsetTransform=(Rotation=(X=-0.0542,Y=-0.3625,Z=-0.1375,W=0.9202),Translation=(X=0,Y=0,Z=0),Scale3D=(X=1,Y=1,Z=1))
    bPropagateToChildren=True
    Weight=@n2.OutputParam
   )
   > &RigUnit_OffsetTransformForItem(
    Item=(Type=Bone,Name="hair_side_r_pnt_2")
    OffsetTransform=(Rotation=(X=0,Y=0,Z=0,W=1),Translation=(X=0,Y=1,Z=-2),Scale3D=(X=1,Y=1,Z=1))
    bPropagateToChildren=True
    Weight=@n1.OutputParam
   )
   > &RigUnit_OffsetTransformForItem(
    Item=(Type=Bone,Name="hair_side_r_pnt_3")
    OffsetTransform=(Rotation=(X=0,Y=0,Z=0,W=1),Translation=(X=0,Y=1,Z=1),Scale3D=(X=1,Y=1,Z=1))
    bPropagateToChildren=True
    Weight=@n1.OutputParam
   )
   n3 = &RigUnit_SphericalPoseReader(
    ActiveRegionScaleFactors=(PositiveWidth=1,NegativeWidth=1,PositiveHeight=1.3,NegativeHeight=1.3)
    ActiveRegionSize=0.21
    Debug=(bDrawDebug=false,bDraw2D=False,bDrawLocalAxes=False,DebugScale=25,DebugSegments=20,DebugThickness=0.25)
    DriverAxis=(X=1,Y=0,Z=0)
    DriverItem=(Type=Bone,Name="head")
    FalloffRegionScaleFactors=(PositiveWidth=1,NegativeWidth=1,PositiveHeight=1,NegativeHeight=1)
    FalloffSize=0.42
    FlipHeightScaling=False
    FlipWidthScaling=False
    OptionalParentItem=(Type=Bone,Name="spine_05")
    RotationOffset=(X=-90,Y=0,Z=0)
    ExecutePin=@RigVMFunction_Sequence_1.C
   )
   > &RigUnit_OffsetTransformForItem(
    Item=(Type=Bone,Name="hair_side_r_pnt_1")
    OffsetTransform=(Rotation=(X=0,Y=0,Z=0,W=1),Translation=(X=-3,Y=0,Z=-3),Scale3D=(X=1,Y=1,Z=1))
    bPropagateToChildren=True
    Weight=@n3.OutputParam
   )
   > &RigUnit_OffsetTransformForItem(
    Item=(Type=Bone,Name="hair_side_r_pnt_2")
    OffsetTransform=(Rotation=(X=0,Y=0,Z=0,W=1),Translation=(X=-5,Y=0,Z=-2),Scale3D=(X=1,Y=1,Z=1))
    bPropagateToChildren=True
    Weight=@n3.OutputParam
   )
   > &RigUnit_OffsetTransformForItem(
    Item=(Type=Bone,Name="hair_side_r_pnt_3")
    OffsetTransform=(Rotation=(X=0,Y=0,Z=0,W=1),Translation=(X=0,Y=0,Z=-1),Scale3D=(X=1,Y=1,Z=1))
    bPropagateToChildren=True
    Weight=@n3.OutputParam
   )
   > &RigUnit_OffsetTransformForItem(
    Item=(Type=Bone,Name="ear_r_deform")
    OffsetTransform=(Rotation=(X=0,Y=0,Z=0,W=1),Translation=(X=0,Y=2.5,Z=3),Scale3D=(X=1,Y=1,Z=1))
    bPropagateToChildren=True
    Weight=@n3.OutputParam
   )
   ItemArray = RigVMDispatchNode(Value=((Type=Bone,Name="hair_side_r_pnt_1"),(Type=Bone,Name="hair_side_r_pnt_2"),(Type=Bone,Name="hair_side_r_pnt_3"),(Type=Bone,Name="hair_side_r_pnt_3"))) {template="DISPATCH_RigVMDispatch_Constant(io Value)"}
   RigVMFunction_MathTransformArrayToSRT = /Script/RigVM.RigVMFunction_MathTransformArrayToSRT(Transforms=/Script/ControlRig.RigUnit_GetTransformItemArray(Space=GlobalSpace, bInitial=False, Items=@ItemArray.Value) {is=struct RigVMUnitNode}.Transforms) {is=struct RigVMUnitNode}
   ItemArray_1 = RigVMDispatchNode(Value=((Type=Bone,Name="hair_side_r_1"),(Type=Bone,Name="hair_side_r_2"),(Type=Bone,Name="hair_side_r_3"))) {template="DISPATCH_RigVMDispatch_Constant(io Value)"}
   /Script/ControlRigSpline.RigUnit_FitChainToSplineCurveItemArray(
    Alignment=Stretched
    DebugSettings=(bEnabled=False,Scale=1,CurveColor=(R=1,G=1,B=0,A=1),SegmentsColor=(R=1,G=0,B=0,A=1),WorldOffset=(Rotation=(X=0,Y=0,Z=0,W=1),Translation=(X=0,Y=0,Z=0),Scale3D=(X=1,Y=1,Z=1)))
    Maximum=1
    Minimum=0
    PoleVectorPosition=(X=0,Y=0,Z=0)
    PrimaryAxis=(X=1,Y=0,Z=0)
    RotationEaseType=Linear
    Rotations=()
    SamplingPrecision=12
    SecondaryAxis=(X=0,Y=0,Z=0)
    Weight=1
    bPropagateToChildren=True
    Spline=/Script/ControlRigSpline.RigUnit_ControlRigSplineFromPoints(
     Compression=0
     SamplesPerSegment=12
     SplineMode=Hermite
     Stretch=0
     bClosed=False
     Points=@RigVMFunction_MathTransformArrayToSRT.Translations
    ) {is=struct RigVMUnitNode}.Spline
    Items=@ItemArray_1.Value
    ExecutePin=@RigVMFunction_Sequence_1.D
   ) {is=struct RigVMUnitNode}
   ItemArray_3 = RigVMDispatchNode(Value=((Type=Bone,Name="hair_side_l_pnt_1"),(Type=Bone,Name="hair_side_l_pnt_2"),(Type=Bone,Name="hair_side_l_pnt_3"),(Type=Bone,Name="hair_side_l_pnt_3"))) {template="DISPATCH_RigVMDispatch_Constant(io Value)"}
   n4 = &RigUnit_GetTransformItemArray(Space=GlobalSpace, bInitial=False, Items=@ItemArray_3.Value)
   ItemArray_2 = RigVMDispatchNode(Value=((Type=Bone,Name="hair_side_l_1"),(Type=Bone,Name="hair_side_l_2"),(Type=Bone,Name="hair_side_l_3"))) {template="DISPATCH_RigVMDispatch_Constant(io Value)"}
   > &RigUnit_FitChainToSplineCurveItemArray(
    Alignment=Stretched
    DebugSettings=(bEnabled=False,Scale=1,CurveColor=(R=1,G=1,B=0,A=1),SegmentsColor=(R=1,G=0,B=0,A=1),WorldOffset=(Rotation=(X=0,Y=0,Z=0,W=1),Translation=(X=0,Y=0,Z=0),Scale3D=(X=1,Y=1,Z=1)))
    Maximum=1
    Minimum=0
    PoleVectorPosition=(X=0,Y=0,Z=0)
    PrimaryAxis=(X=1,Y=0,Z=0)
    RotationEaseType=Linear
    Rotations=()
    SamplingPrecision=12
    SecondaryAxis=(X=0,Y=0,Z=0)
    Weight=1
    bPropagateToChildren=True
    Spline=&RigUnit_ControlRigSplineFromPoints(
     Compression=0
     SamplesPerSegment=12
     SplineMode=Hermite
     Stretch=0
     bClosed=False
     Points=&RigVMFunction_MathTransformArrayToSRT(Transforms=@n4.Transforms).Translations
    ).Spline
    Items=@ItemArray_2.Value
   )
   n5 = &RigUnit_SphericalPoseReader(
    ActiveRegionScaleFactors=(PositiveWidth=1.5,NegativeWidth=1.5,PositiveHeight=1.6,NegativeHeight=1)
    ActiveRegionSize=0.13
    Debug=(bDrawDebug=false,bDraw2D=False,bDrawLocalAxes=False,DebugScale=25,DebugSegments=20,DebugThickness=0.25)
    DriverAxis=(X=0,Y=0,Z=1)
    DriverItem=(Type=Bone,Name="hair_down_1")
    FalloffRegionScaleFactors=(PositiveWidth=4,NegativeWidth=4,PositiveHeight=6,NegativeHeight=1)
    FalloffSize=0.22
    FlipHeightScaling=False
    FlipWidthScaling=False
    OptionalParentItem=(Type=Bone,Name="head")
    RotationOffset=(X=35,Y=10,Z=0)
    ExecutePin=@RigVMFunction_Sequence_1.E
   )
   > &RigUnit_OffsetTransformForItem(
    Item=(Type=Bone,Name="hair_side_l_deform")
    OffsetTransform=(Rotation=(X=0,Y=0,Z=0,W=1),Translation=(X=-4,Y=0,Z=0),Scale3D=(X=1,Y=1,Z=1))
    bPropagateToChildren=True
    Weight=@n5.OutputParam
   )
   > &RigUnit_OffsetTransformForItem(
    Item=(Type=Bone,Name="hair_side_l_1")
    OffsetTransform=(Rotation=(X=0,Y=0,Z=0,W=1),Translation=(X=0,Y=-1,Z=0),Scale3D=(X=1,Y=1,Z=1))
    bPropagateToChildren=false
    Weight=@n5.OutputParam
   )
   > n6 = &RigUnit_SphericalPoseReader(
    ActiveRegionScaleFactors=(PositiveWidth=1,NegativeWidth=1,PositiveHeight=0.3,NegativeHeight=0.5)
    ActiveRegionSize=0.32
    Debug=(bDrawDebug=false,bDraw2D=False,bDrawLocalAxes=False,DebugScale=25,DebugSegments=20,DebugThickness=0.25)
    DriverAxis=(X=-1,Y=0,Z=0)
    DriverItem=(Type=Bone,Name="clavicle_r")
    FalloffRegionScaleFactors=(PositiveWidth=1,NegativeWidth=1,PositiveHeight=0.3,NegativeHeight=0.5)
    FalloffSize=0.4
    FlipHeightScaling=False
    FlipWidthScaling=False
    OptionalParentItem=(Type=Bone,Name="None")
    RotationOffset=(X=160,Y=0,Z=0)
   )
   > &RigUnit_OffsetTransformForItem(
    Item=(Type=Bone,Name="hair_side_l_1")
    OffsetTransform=(Rotation=(X=0,Y=0,Z=0,W=1),Translation=(X=0,Y=-2,Z=0),Scale3D=(X=1,Y=1,Z=1))
    bPropagateToChildren=false
    Weight=@n6.OutputParam
   )
   > &RigUnit_OffsetTransformForItem(
    Item=(Type=Bone,Name="hair_side_r_1")
    OffsetTransform=(Rotation=(X=0,Y=0,Z=0,W=1),Translation=(X=0,Y=-2,Z=0),Scale3D=(X=1,Y=1,Z=1))
    bPropagateToChildren=false
    Weight=@n6.OutputParam
   )
   RigVMFunctionReturnNode(ExecuteContext=@RigVMFunction_Sequence.B)
  ]}
 New Function_ContainedGraph {nodes [
   RigVMFunctionEntryNode()
   > &RigUnit_ParentConstraint(
    AdvancedSettings=(InterpolationType=Average,RotationOrderForFilter=XZY)
    Child=(Type=Bone,Name="dyn_hair_bk_MSTR")
    Parents=((Item=(Type=Bone,Name="spine_05"),Weight=0.7),(Item=(Type=Bone,Name="head"),Weight=0.3))
    Weight=1
   )
   > &RigUnit_ParentConstraint(
    AdvancedSettings=(InterpolationType=Average,RotationOrderForFilter=XZY)
    Child=(Type=Bone,Name="dyn_hair_l_MSTR")
    Parents=((Item=(Type=Bone,Name="spine_05"),Weight=0.4),(Item=(Type=Bone,Name="head"),Weight=0.6))
    Weight=1
   )
   > &RigUnit_ParentConstraint(
    AdvancedSettings=(InterpolationType=Average,RotationOrderForFilter=XZY)
    Child=(Type=Bone,Name="dyn_hair_r_MSTR")
    Parents=((Item=(Type=Bone,Name="spine_05"),Weight=0.4),(Item=(Type=Bone,Name="head"),Weight=0.6))
    Weight=1
   )
   > SequenceExecution_12 = RigVMAggregateNode()
   SequenceExecution_3 = RigVMAggregateNode(ExecuteContext=@SequenceExecution_12.A)
   SequenceExecution_4 = RigVMAggregateNode(ExecuteContext=@SequenceExecution_3.A)
   SequenceExecution_7 = RigVMAggregateNode(ExecuteContext=@SequenceExecution_4.A)
   n1 = &RigUnit_SphericalPoseReader(
    ActiveRegionScaleFactors=(PositiveWidth=1,NegativeWidth=1,PositiveHeight=1,NegativeHeight=1)
    ActiveRegionSize=0.065
    Debug=(bDrawDebug=false,DebugScale=25,DebugSegments=20,DebugThickness=0.25)
    DriverAxis=(X=-1,Y=0,Z=0)
    DriverItem=(Type=Bone,Name="upperarm_r")
    FalloffRegionScaleFactors=(PositiveWidth=1,NegativeWidth=1,PositiveHeight=1,NegativeHeight=1)
    FalloffSize=0.125
    OptionalParentItem=(Type=Bone,Name="None")
    RotationOffset=(X=60,Y=0,Z=65)
    ExecutePin=@SequenceExecution_7.A
   )
   > RigVMVariableNode(Value=@n1.OutputParam) {variable=upperarm_r_SPR_back_alpha}
   n2 = &RigUnit_SphericalPoseReader(
    ActiveRegionScaleFactors=(PositiveWidth=1.2,NegativeWidth=1,PositiveHeight=1,NegativeHeight=1.5)
    ActiveRegionSize=0.075
    Debug=(bDrawDebug=false,DebugScale=25,DebugSegments=20,DebugThickness=0.25)
    DriverAxis=(X=-1,Y=0,Z=0)
    DriverItem=(Type=Bone,Name="upperarm_r")
    FalloffRegionScaleFactors=(PositiveWidth=1,NegativeWidth=1,PositiveHeight=1,NegativeHeight=1)
    FalloffSize=0.2
    OptionalParentItem=(Type=Bone)
    RotationOffset=(X=-60,Y=0,Z=-15)
    ExecutePin=@SequenceExecution_7.D
   )
   > RigVMVariableNode(Value=@n2.OutputParam) {variable=upperarm_r_rifle_up_alpha}
   n3 = &RigUnit_SphericalPoseReader(
    ActiveRegionScaleFactors=(PositiveWidth=1,NegativeWidth=1,PositiveHeight=1,NegativeHeight=1)
    ActiveRegionSize=0.06
    Debug=(bDrawDebug=false,DebugScale=25,DebugSegments=20,DebugThickness=0.25)
    DriverAxis=(X=-1,Y=0,Z=0)
    DriverItem=(Type=Bone,Name="upperarm_r")
    FalloffRegionScaleFactors=(PositiveWidth=1,NegativeWidth=1,PositiveHeight=1,NegativeHeight=1)
    FalloffSize=0.15
    OptionalParentItem=(Type=Bone,Name="spine_05")
    RotationOffset=(X=-90,Y=0,Z=40)
    ExecutePin=@SequenceExecution_7.B
   )
   > RigVMVariableNode(Value=@n3.OutputParam) {variable=uppearm_r_SPR_aim_up_two_hands_alpha}
   n4 = &RigUnit_SphericalPoseReader(
    ActiveRegionScaleFactors=(PositiveWidth=2,NegativeWidth=2,PositiveHeight=1,NegativeHeight=1)
    ActiveRegionSize=0.05
    Debug=(bDrawDebug=false,DebugScale=25,DebugSegments=20,DebugThickness=0.25)
    DriverAxis=(X=-1,Y=0,Z=0)
    DriverItem=(Type=Bone,Name="upperarm_r")
    FalloffRegionScaleFactors=(PositiveWidth=1,NegativeWidth=1,PositiveHeight=0.75,NegativeHeight=0.75)
    FalloffSize=0.2
    FlipWidthScaling=false
    OptionalParentItem=(Type=Bone)
    RotationOffset=(X=15,Y=150,Z=0)
    ExecutePin=@SequenceExecution_7.C
   )
   > RigVMVariableNode(Value=@n4.OutputParam) {variable=upperarm_r_skydiving}
   SequenceExecution_2 = RigVMAggregateNode(ExecuteContext=@SequenceExecution_4.B)
   n5 = &RigUnit_SphericalPoseReader(
    ActiveRegionScaleFactors=(PositiveWidth=1,NegativeWidth=0.5,PositiveHeight=1.5,NegativeHeight=1.5)
    ActiveRegionSize=0.1
    Debug=(bDrawDebug=false,DebugScale=25,DebugSegments=20,DebugThickness=0.25)
    DriverAxis=(X=0,Y=1,Z=0)
    DriverItem=(Type=Bone,Name="head")
    FalloffRegionScaleFactors=(PositiveWidth=1,NegativeWidth=0.7,PositiveHeight=1.5,NegativeHeight=1.5)
    FalloffSize=0.4
    FlipWidthScaling=false
    OptionalParentItem=(Type=Bone,Name="spine_05")
    RotationOffset=(X=90,Y=0,Z=-80)
    ExecutePin=@SequenceExecution_2.A
   )
   > RigVMVariableNode(Value=@n5.OutputParam) {variable=head_SPR_look_up_alpha}
   n6 = &RigUnit_SphericalPoseReader(
    ActiveRegionScaleFactors=(PositiveWidth=2,NegativeWidth=0.5,PositiveHeight=2,NegativeHeight=2)
    ActiveRegionSize=0.075
    Debug=(bDrawDebug=false,DebugScale=25,DebugSegments=20,DebugThickness=0.25)
    DriverAxis=(X=0,Y=1,Z=0)
    DriverItem=(Type=Bone,Name="head")
    FalloffRegionScaleFactors=(PositiveWidth=1,NegativeWidth=0.7,PositiveHeight=1,NegativeHeight=1)
    FalloffSize=0.2
    FlipHeightScaling=false
    FlipWidthScaling=true
    OptionalParentItem=(Type=Bone)
    RotationOffset=(X=90,Y=0,Z=12.5)
    ExecutePin=@SequenceExecution_2.B
   )
   > RigVMVariableNode(Value=@n6.OutputParam) {variable=head_SPR_look_down_alpha}
   n7 = &RigUnit_SphericalPoseReader(
    ActiveRegionScaleFactors=(PositiveWidth=0.75,NegativeWidth=0.75,PositiveHeight=1,NegativeHeight=0.5)
    ActiveRegionSize=0.1
    Debug=(bDrawDebug=false,DebugScale=25,DebugSegments=20,DebugThickness=0.25)
    DriverAxis=(X=0,Y=1,Z=0)
    DriverItem=(Type=Bone,Name="head")
    FalloffRegionScaleFactors=(PositiveWidth=1,NegativeWidth=1,PositiveHeight=1,NegativeHeight=0.7)
    FalloffSize=0.2
    OptionalParentItem=(Type=Bone)
    RotationOffset=(X=120,Y=0,Z=0)
    ExecutePin=@SequenceExecution_2.C
   )
   > RigVMVariableNode(Value=@n7.OutputParam) {variable=head_SPR_look_left_alpha}
   n8 = &RigUnit_SphericalPoseReader(
    ActiveRegionScaleFactors=(PositiveWidth=0.75,NegativeWidth=0.75,PositiveHeight=1,NegativeHeight=0.5)
    ActiveRegionSize=0.1
    Debug=(bDrawDebug=false,DebugScale=25,DebugSegments=20,DebugThickness=0.25)
    DriverAxis=(X=0,Y=1,Z=0)
    DriverItem=(Type=Bone,Name="head")
    FalloffRegionScaleFactors=(PositiveWidth=1,NegativeWidth=1,PositiveHeight=1,NegativeHeight=0.7)
    FalloffSize=0.2
    FlipHeightScaling=true
    FlipWidthScaling=false
    OptionalParentItem=(Type=Bone)
    RotationOffset=(X=60,Y=0,Z=0)
    ExecutePin=@SequenceExecution_2.D
   )
   > RigVMVariableNode(Value=@n8.OutputParam) {variable=head_SPR_look_right_alpha}
   RigVMVariableNode(
    Value=&RigVMFunction_MathFloatMul(B=RigVMVariableNode() {variable=uppearm_r_SPR_aim_up_two_hands_alpha}.Value, A=RigVMVariableNode() {variable=head_SPR_look_up_alpha}.Value) {template="Multiply::Execute(in A,in B,out Result)"}.Result
    ExecuteContext=@SequenceExecution_3.B
   ) {variable=look_up_and_left_two_hands_multiplied_alpha}
   RigVMVariableNode(
    Value=&RigVMFunction_MathFloatMul(A=RigVMVariableNode() {variable=head_SPR_look_up_alpha}.Value, B=RigVMVariableNode() {variable=upperarm_r_rifle_up_alpha}.Value) {template="Multiply::Execute(in A,in B,out Result)"}.Result
    ExecuteContext=@SequenceExecution_3.C
   ) {variable=look_up_two_hands_rifle_multiplied_alpha}
   RigVMVariableNode(
    Value=&RigVMFunction_MathFloatMul(B=RigVMVariableNode() {variable=head_SPR_look_down_alpha}.Value, A=RigVMVariableNode() {variable=upperarm_r_SPR_back_alpha}.Value) {template="Multiply::Execute(in A,in B,out Result)"}.Result
    ExecuteContext=@SequenceExecution_3.D
   ) {variable=look_down_upperarm_r_back_multiplied_alpha}
   RigVMFunctionReturnNode(ExecuteContext=@SequenceExecution_12.F)
   SequenceExecution_8 = RigVMAggregateNode(ExecuteContext=@SequenceExecution_12.C)
   &RigUnit_ModifyTransforms(
    ItemToModify=((Item=(Type=Bone,Name="dyn_hair_r_MSTR"),Transform=(Rotation=(X=0,Y=0,Z=0,W=1),Translation=(X=1,Y=2,Z=-1),Scale3D=(X=1,Y=1,Z=1))))
    ItemToModify.0.Transform.Rotation=&RigVMFunction_MathQuaternionFromEuler(Euler=(X=0,Y=5,Z=-5), RotationOrder=XYZ).Result
    ExecutePin=@SequenceExecution_8.A
    Weight=RigVMVariableNode() {variable=look_up_and_left_two_hands_multiplied_alpha}.Value
   )
   n9 = &RigUnit_SphericalPoseReader(
    ActiveRegionScaleFactors=(PositiveWidth=2,NegativeWidth=0.5,PositiveHeight=2,NegativeHeight=2)
    ActiveRegionSize=0.075
    Debug=(bDrawDebug=false,DebugScale=25,DebugSegments=20,DebugThickness=0.25)
    DriverAxis=(X=1,Y=0,Z=0)
    DriverItem=(Type=Bone,Name="head")
    FalloffRegionScaleFactors=(PositiveWidth=1,NegativeWidth=0.4,PositiveHeight=1,NegativeHeight=1)
    FalloffSize=0.25
    FlipHeightScaling=false
    FlipWidthScaling=true
    OptionalParentItem=(Type=Bone,Name="spine_05")
    RotationOffset=(X=-70,Y=5,Z=0)
    ExecutePin=@SequenceExecution_8.D
   )
   n10 = &RigVMFunction_AlphaInterp(
    InterpSpeedDecreasing=10
    InterpSpeedIncreasing=10
    bClampResult=true
    Value=@n9.OutputParam
    Scale=RigVMVariableNode() {variable=crouch}.Value
   )
   > &RigUnit_ModifyTransforms(
    ItemToModify="((Item=(Type=Bone,Name=\"dyn_hair_bk_MSTR\"),Transform=(Rotation=(X=0,Y=0,Z=0,W=1),Translation=(X=3,Y=-3,Z=-2),Scale3D=(X=1,Y=1,Z=1))),(Item=(Type=Bone,Name=\"dyn_hair_l_MSTR\"),Transform=(Rotation=(X=0,Y=0,Z=0,W=1),Translation=(X=1,Y=-5,Z=-2),Scale3D=(X=1,Y=1 ...(truncated)"
    Weight=@n10.Result
    ItemToModify.0.Transform.Rotation=&RigVMFunction_MathQuaternionFromEuler(Euler=(X=0,Y=0,Z=0), RotationOrder=XYZ).Result
    ItemToModify.2.Transform.Rotation=&RigVMFunction_MathQuaternionFromEuler(Euler=(X=0,Y=0,Z=0), RotationOrder=XYZ).Result
    ItemToModify.1.Transform.Rotation=&RigVMFunction_MathQuaternionFromEuler(Euler=(X=0,Y=0,Z=0), RotationOrder=XYZ).Result
   )
   &RigUnit_ModifyTransforms(
    ItemToModify=((Item=(Type=Bone,Name="dyn_hair_bk_MSTR"),Transform=(Rotation=(X=0,Y=0,Z=0,W=1),Translation=(X=0,Y=2,Z=-2),Scale3D=(X=1,Y=1,Z=1))))
    ItemToModify.0.Transform.Rotation=&RigVMFunction_MathQuaternionFromEuler(Euler=(X=0,Y=5,Z=0), RotationOrder=XYZ).Result
    Weight=RigVMVariableNode() {variable=look_up_and_left_two_hands_multiplied_alpha}.Value
    ExecutePin=@SequenceExecution_8.B
   )
   &RigUnit_ModifyTransforms(
    ItemToModify=((Item=(Type=Bone,Name="dyn_hair_l_MSTR"),Transform=(Rotation=(X=0,Y=0,Z=0,W=1),Translation=(X=-0.5,Y=-2,Z=-4),Scale3D=(X=1,Y=1,Z=1))))
    ItemToModify.0.Transform.Rotation=&RigVMFunction_MathQuaternionFromEuler(Euler=(X=0,Y=0,Z=-5), RotationOrder=XYZ).Result
    ExecutePin=@SequenceExecution_8.C
    Weight=RigVMVariableNode() {variable=look_up_and_left_two_hands_multiplied_alpha}.Value
   )
   SequenceExecution = RigVMAggregateNode(ExecuteContext=@SequenceExecution_12.B)
   &RigUnit_ModifyTransforms(
    ItemToModify="((Item=(Type=Bone,Name=\"dyn_hair_bk_MSTR\"),Transform=(Rotation=(X=0,Y=0,Z=0,W=1),Translation=(X=1,Y=-5,Z=0),Scale3D=(X=1,Y=1,Z=1))),(Item=(Type=Bone,Name=\"dyn_hair_l_MSTR\"),Transform=(Rotation=(X=0,Y=0,Z=0,W=1),Translation=(X=-2.5,Y=-2.5,Z=0),Scale3D=(X=1, ...(truncated)"
    ItemToModify.0.Transform.Rotation=&RigVMFunction_MathQuaternionFromEuler(Euler=(X=0,Y=0,Z=-5), RotationOrder=XYZ).Result
    ItemToModify.2.Transform.Rotation=&RigVMFunction_MathQuaternionFromEuler(Euler=(X=0,Y=30,Z=-30), RotationOrder=XYZ).Result
    ItemToModify.1.Transform.Rotation=&RigVMFunction_MathQuaternionFromEuler(Euler=(X=0,Y=0,Z=-15), RotationOrder=XYZ).Result
    Weight=RigVMVariableNode() {variable=head_SPR_look_up_alpha}.Value
    ExecutePin=@SequenceExecution.A
   )
   n11 = /Script/RigVM.RigVMFunction_MathFloatMax(
    B=RigVMVariableNode() {variable=look_up_two_hands_rifle_multiplied_alpha}.Value
    A=RigVMVariableNode() {variable=look_up_and_left_two_hands_multiplied_alpha}.Value
   ) {is=struct RigVMUnitNode, template="Maximum::Execute(in A,in B,out Result)"}
   > &RigUnit_ModifyTransforms(
    ItemToModify=((Item=(Type=Bone,Name="dyn_hair_bk_MSTR"),Transform=(Rotation=(X=0,Y=0,Z=0,W=1),Translation=(X=0,Y=5,Z=0),Scale3D=(X=1,Y=1,Z=1))))
    ItemToModify.0.Transform.Rotation=&RigVMFunction_MathQuaternionFromEuler(Euler=(X=0,Y=0,Z=10), RotationOrder=XYZ).Result
    Weight=@n11.Result
   )
   > &RigUnit_ModifyTransforms(
    ItemToModify="((Item=(Type=Bone,Name=\"dyn_hair_r_MSTR\"),Transform=(Rotation=(X=0,Y=0,Z=0,W=1),Translation=(X=-3,Y=-8,Z=0),Scale3D=(X=1,Y=1,Z=1))),(Item=(Type=Bone,Name=\"dyn_hair_r_01a\"),Transform=(Rotation=(X=0,Y=0,Z=0,W=1),Translation=(X=0,Y=1,Z=1),Scale3D=(X=1,Y=1,Z=1 ...(truncated)"
    ItemToModify.0.Transform.Rotation=&RigVMFunction_MathQuaternionFromEuler(Euler=(X=0,Y=0,Z=40), RotationOrder=XYZ).Result
    Weight=@n11.Result
   )
   > &RigUnit_ModifyTransforms(
    ItemToModify=((Item=(Type=Bone,Name="dyn_hair_l_MSTR"),Transform=(Rotation=(X=0,Y=0,Z=0.0872,W=0.9962),Translation=(X=0,Y=5,Z=3),Scale3D=(X=1,Y=1,Z=1))))
    Weight=@n11.Result
   )
   SequenceExecution_6_1 = RigVMAggregateNode(ExecuteContext=@SequenceExecution_12.D)
   &RigUnit_ModifyTransforms(
    ItemToModify=((Item=(Type=Bone,Name="dyn_hair_r_MSTR"),Transform=(Rotation=(X=0,Y=0,Z=0,W=1),Translation=(X=0,Y=-1,Z=0),Scale3D=(X=1,Y=1,Z=1))))
    ItemToModify.0.Transform.Rotation=&RigVMFunction_MathQuaternionFromEuler(Euler=(X=5,Y=5,Z=0)).Result
    ExecutePin=@SequenceExecution_6_1.B
    Weight=RigVMVariableNode() {variable=head_SPR_look_down_alpha}.Value
   )
   &RigUnit_ModifyTransforms(
    ItemToModify=((Item=(Type=Bone,Name="dyn_hair_bk_MSTR"),Transform=(Rotation=(X=0,Y=0,Z=0,W=1),Translation=(X=0,Y=0,Z=0),Scale3D=(X=1,Y=1,Z=1))))
    ItemToModify.0.Transform.Rotation=&RigVMFunction_MathQuaternionFromEuler(Euler=(X=0,Y=0,Z=0)).Result
    ExecutePin=@SequenceExecution_6_1.C
    Weight=RigVMVariableNode() {variable=head_SPR_look_down_alpha}.Value
   )
   &RigUnit_ModifyTransforms(
    ItemToModify=((Item=(Type=Bone,Name="dyn_hair_l_MSTR"),Transform=(Rotation=(X=0,Y=0,Z=0,W=1),Translation=(X=0,Y=0,Z=0),Scale3D=(X=1,Y=1,Z=1))))
    ItemToModify.0.Transform.Rotation=&RigVMFunction_MathQuaternionFromEuler(Euler=(X=0,Y=0,Z=0)).Result
    Weight=RigVMVariableNode() {variable=look_down_upperarm_r_back_multiplied_alpha}.Value
    ExecutePin=@SequenceExecution_6_1.D
   )
   AlphaInterp = &RigVMFunction_AlphaInterp(
    InterpSpeedDecreasing=10
    InterpSpeedIncreasing=10
    bClampResult=true
    Value=&RigVMFunction_MathFloatMax(
     B=RigVMVariableNode() {variable=look_up_two_hands_rifle_multiplied_alpha}.Value
     A=RigVMVariableNode() {variable=look_up_and_left_two_hands_multiplied_alpha}.Value
    ) {template="Maximum::Execute(in A,in B,out Result)"}.Result
    Scale=RigVMVariableNode() {variable=crouch move}.Value
   )
   &RigUnit_ModifyTransforms(
    ItemToModify="((Item=(Type=Bone,Name=\"dyn_hair_bk_MSTR\"),Transform=(Rotation=(X=0,Y=0,Z=0,W=1),Translation=(X=4,Y=-3,Z=0),Scale3D=(X=1,Y=1,Z=1))),(Item=(Type=Bone,Name=\"dyn_hair_l_MSTR\"),Transform=(Rotation=(X=0,Y=0,Z=0,W=1),Translation=(X=3,Y=-4,Z=0),Scale3D=(X=1,Y=1,Z ...(truncated)"
    Weight=@AlphaInterp.Result
    ItemToModify.0.Transform.Rotation=&RigVMFunction_MathQuaternionFromEuler(Euler=(X=0,Y=0,Z=0), RotationOrder=XYZ).Result
    ItemToModify.2.Transform.Rotation=&RigVMFunction_MathQuaternionFromEuler(Euler=(X=0,Y=-5,Z=0), RotationOrder=XYZ).Result
    ItemToModify.1.Transform.Rotation=&RigVMFunction_MathQuaternionFromEuler(Euler=(X=0,Y=5,Z=0), RotationOrder=XYZ).Result
    ExecutePin=@SequenceExecution_12.E
   )
   RigVMVariableNode() {variable=look_up_and_left_two_hands_multiplied_alpha}
   RigVMVariableNode() {variable=look_down_upperarm_r_back_multiplied_alpha}
   RigVMVariableNode() {variable=look_down_upperarm_r_back_multiplied_alpha}
  ]}
 RigVMFunctionLibrary {nodes [
   Hair_Back = RigVMCollapseNode(ExecuteContext=())
   Hair_L_And_R = RigVMCollapseNode(ExecuteContext=())
  ]}
 RigVMFunction_Sequence_1_ContainedGraph {nodes [
   n1 = RigVMFunctionEntryNode()
   RigVMFunction_Sequence = &RigVMFunction_Sequence(ExecuteContext=@n1.ExecuteContext)
   n2 = &RigVMFunction_Sequence(ExecuteContext=@RigVMFunction_Sequence.B)
   n3 = &RigVMFunction_Sequence(ExecuteContext=@n2.B)
   n4 = &RigVMFunction_Sequence(ExecuteContext=@n3.B)
   RigVMFunctionReturnNode(A=@RigVMFunction_Sequence.A, B=@n2.A, C=@n3.A, D=@n4.A, E=@n4.B)
  ]}
 RigVMFunction_Sequence_1_ContainedGraph {nodes [
   n1 = RigVMFunctionEntryNode()
   RigVMFunction_Sequence = &RigVMFunction_Sequence(ExecuteContext=@n1.ExecuteContext)
   n2 = &RigVMFunction_Sequence(ExecuteContext=@RigVMFunction_Sequence.B)
   n3 = &RigVMFunction_Sequence(ExecuteContext=@n2.B)
   n4 = &RigVMFunction_Sequence(ExecuteContext=@n3.B)
   RigVMFunctionReturnNode(A=@RigVMFunction_Sequence.A, B=@n2.A, C=@n3.A, D=@n4.A, E=@n4.B)
  ]}
 RigVMModel {nodes [
   RigUnit_BeginExecution = &RigUnit_BeginExecution()
   SequenceExecution_1 = RigVMAggregateNode(ExecuteContext=@RigUnit_BeginExecution.ExecutePin)
   Hair_L_And_R = RigVMFunctionReferenceNode(ExecuteContext=@SequenceExecution_1.A)
   Hair_Back = RigVMFunctionReferenceNode(ExecuteContext=@SequenceExecution_1.B)
   RigVMFunction_ControlFlowBranch = &RigVMFunction_ControlFlowBranch(ExecuteContext=@SequenceExecution_1.C, Condition=RigVMVariableNode() {variable=ShouldersUp}.Value)
   n1 = &RigUnit_SphericalPoseReader(
    ActiveRegionScaleFactors=(PositiveWidth=1,NegativeWidth=1,PositiveHeight=1,NegativeHeight=1)
    ActiveRegionSize=0.1
    Debug=(bDrawDebug=True,bDraw2D=False,bDrawLocalAxes=False,DebugScale=25,DebugSegments=20,DebugThickness=0.25)
    DriverAxis=(X=-1,Y=0,Z=0)
    DriverItem=(Type=Bone,Name="clavicle_r")
    FalloffRegionScaleFactors=(PositiveWidth=1,NegativeWidth=1,PositiveHeight=1,NegativeHeight=1)
    FalloffSize=0.3
    FlipHeightScaling=False
    FlipWidthScaling=False
    OptionalParentItem=(Type=Bone,Name="None")
    RotationOffset=(X=0,Y=140,Z=0)
    ExecutePin=@RigVMFunction_ControlFlowBranch.True
   )
   > &RigUnit_ModifyTransforms(
    ItemToModify="((Item=(Type=Bone,Name=\"dyn_hair_r_MSTR\"),Transform=(Rotation=(X=0.2578,Y=-0.0842,Z=0.0226,W=0.9623),Translation=(X=-2,Y=-5,Z=0),Scale3D=(X=1,Y=1,Z=1))),(Item=(Type=Bone,Name=\"dyn_hair_bk_MSTR\"),Transform=(Rotation=(X=0,Y=0,Z=0,W=1),Translation=(X=-2,Y=0,Z ...(truncated)"
    WeightMinimum=0
    Weight=@n1.OutputParam
   )
   > n2 = &RigUnit_SphericalPoseReader(
    ActiveRegionScaleFactors=(PositiveWidth=1,NegativeWidth=1,PositiveHeight=1,NegativeHeight=1)
    ActiveRegionSize=0.1
    Debug=(bDrawDebug=True,bDraw2D=False,bDrawLocalAxes=False,DebugScale=25,DebugSegments=20,DebugThickness=0.25)
    DriverAxis=(X=1,Y=0,Z=0)
    DriverItem=(Type=Bone,Name="clavicle_l")
    FalloffRegionScaleFactors=(PositiveWidth=1,NegativeWidth=1,PositiveHeight=1,NegativeHeight=1)
    FalloffSize=0.3
    FlipHeightScaling=False
    FlipWidthScaling=False
    OptionalParentItem=(Type=Bone,Name="None")
    RotationOffset=(X=0,Y=-40,Z=0)
   )
   > &RigUnit_ModifyTransforms(
    ItemToModify=((Item=(Type=Bone,Name="dyn_hair_l_MSTR"),Transform=(Rotation=(X=-0.2578,Y=0.0842,Z=0.0226,W=0.9623),Translation=(X=-2,Y=-5,Z=0),Scale3D=(X=1,Y=1,Z=1))))
    WeightMinimum=0
    Weight=@n2.OutputParam
   )
  ]}
]
hierarchy [
 root {loc=<0 0 0>, rot=<0 0 0 1>}
 pelvis {parent=root, loc=<0 -0.5301 98.0571>, rot=<0.0127 0.707 -0.0127 -0.707>}
 spine_01 {parent=pelvis, loc=<6.1208 -0.9328 0>, rot=<0 0 0.034 -0.9994>}
 spine_02 {parent=spine_01, loc=<6.5996 -0.2095 0>, rot=<0 0 0.0238 -0.9997>}
 spine_03 {parent=spine_02, loc=<6.4914 -0.1237 0>, rot=<0 0 0.0149 -0.9999>}
 spine_04 {parent=spine_03, loc=<6.9598 0.0794 0>, rot=<0 0 -0.0132 -0.9999>}
 spine_05 {parent=spine_04, loc=<6.9987 0 0>, rot=<0 0 -0.0015 -1>}
 neck_01 {parent=spine_05, loc=<11.3138 0.1529 0>, rot=<0 0 -0.2003 -0.9797>}
 neck_02 {parent=neck_01, loc=<4.5627 0 0>, rot=<0 0 0.0504 -0.9987>}
 head {parent=neck_02, loc=<5.3914 0 0>, rot=<0 0 0.115 -0.9934>}
 MSTR_hair_A {parent=head, loc=<4.8173 5.1933 -6.9544>, rot=<0.0119 -0.9917 -0.1266 -0.0168>}
 hair_A_1 {parent=MSTR_hair_A, loc=<0 0 0>, rot=<0 0 0 1>}
 MSTR_hair_B {parent=head, loc=<6.5784 8.7766 -5.3275>, rot=<0.1339 -0.9196 -0.3553 -0.1007>}
 hair_B_1 {parent=MSTR_hair_B, loc=<0 0 0>, rot=<0 0 0 -1>}
 MSTR_hair_C {parent=head, loc=<7.4362 10.1608 2.0174>, rot=<-0.0946 -0.704 -0.6778 0.1899>}
 hair_C_1 {parent=MSTR_hair_C, loc=<0 0 0>, rot=<0 0 0 -1>}
 MSTR_hair_D {parent=head, loc=<8.8829 9.6455 -1.7174>, rot=<-0.0744 -0.7995 -0.5957 -0.0206>}
 hair_D_1 {parent=MSTR_hair_D, loc=<0 0 0>, rot=<0 0 0 1>}
 MSTR_hair_E {parent=head, loc=<4.3503 5.2282 7.2156>, rot=<-0.026 0.2577 0.964 -0.0596>}
 hair_E_1 {parent=MSTR_hair_E, loc=<0 0 0>, rot=<0 0 0 -1>}
 hair_E_2 {parent=hair_E_1, loc=<2.1291 0 0>, rot=<0 0 0.1886 -0.9821>}
 MSTR_hair_F {parent=head, loc=<4.0411 -0.3071 -9.0782>, rot=<0.377 -0.9017 -0.0187 -0.211>}
 hair_F_1 {parent=MSTR_hair_F, loc=<0 0 0>, rot=<0 0 0 -1>}
 hair_F_2 {parent=hair_F_1, loc=<2.8515 0 0>, rot=<-0.0396 0.0166 0.1589 -0.9864>}
 MSTR_hair_G {parent=head, loc=<5.7391 -3.6196 -9.1918>, rot=<0.3834 -0.9003 0.0943 -0.1829>}
 hair_G_1 {parent=MSTR_hair_G, loc=<0 0 0>, rot=<0 0 0 -1>}
 hair_G_2 {parent=hair_G_1, loc=<2.7177 0 0>, rot=<-0.0298 0.0262 0.1687 -0.9849>}
 MSTR_hair_I {parent=head, loc=<4.9165 8.2516 6.19>, rot=<0.1028 0.3638 0.7818 -0.4959>}
 hair_I_1 {parent=MSTR_hair_I, loc=<0 0 0>, rot=<0 0 0 1>}
 MSTR_hair_J {parent=head, loc=<3.9861 2.6434 -8.2828>, rot=<0.0178 -0.9914 -0.1259 0.03>}
 hair_J_1 {parent=MSTR_hair_J, loc=<0 0 0>, rot=<0 0 0 1>}
 MSTR_hair_down {parent=head, loc=<6.81 -6.7072 0>, rot=<0.1157 -0.6976 0.6976 -0.1157>}
 hair_down_1 {parent=MSTR_hair_down, loc=<0 0 0>, rot=<0 0 0 1>}
 hair_down_2 {parent=hair_down_1, loc=<4.9958 0 0>, rot=<0 0 0 1>}
 hair_down_3 {parent=hair_down_2, loc=<4.4414 0 0>, rot=<0 0 0 1>}
 hair_down_4 {parent=hair_down_3, loc=<4.712 0 0>, rot=<0 0.0163 0 -0.9999>}
 hair_down_5 {parent=hair_down_4, loc=<4.4894 0 0>, rot=<0 -0.1651 0 -0.9863>}
 hair_down_6 {parent=hair_down_5, loc=<3.6023 0 0>, rot=<0 0 0 1>}
 hair_down_4_r {parent=hair_down_4, loc=<0 -4 -1>, rot=<-0.2365 0.0454 0.5321 -0.8117>}
 hair_down_4_l {parent=hair_down_4, loc=<0 4 -1>, rot=<0.3525 0.1132 -0.6123 -0.6986>}
 hair_down_3_r_1 {parent=hair_down_3, loc=<0 -5 -1>, rot=<-0.5266 -0.119 0.5087 -0.6706>}
 hair_down_3_r_2 {parent=hair_down_3_r_1, loc=<6.7876 0 0>, rot=<-0.1261 0.1346 -0.5408 -0.8207>}
 hair_down_3_r_3 {parent=hair_down_3_r_1, loc=<4.0851 -5.4276 2.0585>, rot=<-0.5247 -0.5913 -0.5668 0.232>}
 hair_down_3_l_1 {parent=hair_down_3, loc=<0 5 -1>, rot=<0.5233 -0.144 -0.6 -0.5876>}
 hair_down_3_l_2 {parent=hair_down_3_l_1, loc=<5.8862 0 0>, rot=<0.0587 0.0449 0.4599 -0.8849>}
 MSTR_hair_M {parent=hair_down_2, loc=<1.9097 4.9435 0.1995>, rot=<0.1669 0.4001 -0.2478 -0.8664>}
 hair_M_1 {parent=MSTR_hair_M, loc=<0 0 0>, rot=<0 0 0 -1>}
 hair_down_2_r {parent=hair_down_2, loc=<0.8774 -5.4325 -0.4627>, rot=<-0.3152 -0.3424 0.5981 -0.6525>}
 MSTR_hair_H_1 {parent=head, loc=<3.2343 -2.5763 5.4758>, rot=<0.7218 -0.1563 0.6472 0.1887>}
 hair_H_1 {parent=MSTR_hair_H_1, loc=<0 0 0>, rot=<0 0 0 1>}
 hair_H_2 {parent=hair_H_1, loc=<3.5573 -0.0575 1.466>, rot=<-0.1773 -0.0644 -0.5202 -0.8329>}
 hair_H_3 {parent=hair_H_2, loc=<2.0988 0 0>, rot=<0.0394 0.1601 -0.0294 -0.9859>}
 hair_L_1 {parent=hair_H_1, loc=<3.4343 -1.0294 -2.0648>, rot=<-0.3877 -0.2389 -0.4441 -0.7716>}
 hair_K_1 {parent=hair_H_1, loc=<2.8853 -3.5351 -2.7227>, rot=<-0.3467 -0.6185 -0.1646 -0.6857>}
 ear_r_deform_base {parent=hair_H_1, loc=<1.4596 -2.6189 -1.5404>, rot=<-0.2748 -0.194 0.1051 -0.9358>}
 ear_r_deform {parent=ear_r_deform_base, loc=<0 0 0>, rot=<0 0 0 1>}
 MSTR_hair_side_l {parent=head, loc=<0.3109 -3.1906 -6.3109>, rot=<-0.0214 -0.7058 0.7069 0.0414>}
 hair_side_l {parent=MSTR_hair_side_l, loc=<0 0 0>, rot=<0 0 0 1>}
 hair_side_l_1 {parent=MSTR_hair_side_l, loc=<-0.0262 -0.0003 0.0015>, rot=<0 0 0 1>}
 hair_side_l_2 {parent=hair_side_l_1, loc=<5.2304 0 0>, rot=<-0.0036 -0.1953 0.012 -0.9807>}
 hair_side_l_3 {parent=hair_side_l_2, loc=<4.439 0 0>, rot=<0 0 0.1822 -0.9833>}
 hair_side_l_deform_base {parent=hair_side_l_1, loc=<3.3398 0.2217 0.0517>, rot=<-0.5169 -0.5304 0.4825 0.4676>}
 hair_side_l_deform {parent=hair_side_l_deform_base, loc=<0 0 0>, rot=<0 0 0 1>}
 MSTR_hair_side_r {parent=head, loc=<0.3112 -3.1906 6.3109>, rot=<0.0214 0.7058 0.7069 0.0414>}
 hair_side_r_1 {parent=MSTR_hair_side_r, loc=<0 0 0>, rot=<0 0 0 1>}
 hair_side_r_2 {parent=hair_side_r_1, loc=<5.2308 0 0>, rot=<0.0036 0.1953 0.012 -0.9807>}
 hair_side_r_3 {parent=hair_side_r_2, loc=<4.4386 0 0.0002>, rot=<0 0 0.1822 -0.9833>}
 MSTR_ear_ring_l {parent=head, loc=<0.211 0.2326 -6.8421>, rot=<0.4979 0.5021 0.5021 0.4979>}
 ear_ring_l {parent=MSTR_ear_ring_l, loc=<0 0 0>, rot=<0 0 0 1>}
 MSTR_ear_ring_r {parent=head, loc=<0.211 0.2326 6.8421>, rot=<0.4979 0.5021 0.5021 0.4979>}
 ear_ring_r {parent=MSTR_ear_ring_r, loc=<0 0 0>, rot=<0 0 0 1>}
 hair_side_l_pnt_1 {parent=head, loc=<0.3109 -3.1906 -6.3109>, rot=<-0.0214 -0.7058 0.7069 0.0414>}
 hair_side_r_pnt_1 {parent=head, loc=<0.3112 -3.1906 6.3109>, rot=<0.0214 0.7058 0.7069 0.0414>}
 dyn_hair_l_MSTR {parent=head, loc=<4.2741 -4.3095 -6.1145>, rot=<0.0919 -0.9926 0.0067 -0.0796>}
 dyn_hair_l_00 {parent=dyn_hair_l_MSTR, loc=<4.4664 -1.5204 -0.1644>, rot=<0.0181 -0.0525 0.1543 -0.9865>}
 dyn_hair_l_01 {parent=dyn_hair_l_00, loc=<8.1281 0 0>, rot=<0.0116 -0.0246 -0.0469 -0.9985>}
 dyn_hair_l_02 {parent=dyn_hair_l_01, loc=<6.1091 0 0>, rot=<0.0044 -0.0141 -0.091 -0.9957>}
 dyn_hair_l_01a {parent=dyn_hair_l_00, loc=<6.1412 3.7548 0>, rot=<0.0116 -0.0246 -0.0469 -0.9985>}
 dyn_hair_bk_MSTR {parent=head, loc=<4.4962 -4.542 0.0291>, rot=<0.1718 -0.9851 0 0>}
 dyn_hair_bk_00 {parent=dyn_hair_bk_MSTR, loc=<4.2135 -0.7738 0.0291>, rot=<0 0 0.069 -0.9976>}
 dyn_hair_bk_01 {parent=dyn_hair_bk_00, loc=<7.7147 0 0>, rot=<0 0 -0.0149 -0.9999>}
 dyn_hair_bk_02 {parent=dyn_hair_bk_01, loc=<6.6501 0 0>, rot=<0 0 -0.1461 -0.9893>}
 dyn_hair_r_MSTR {parent=head, loc=<4.2877 -3.7149 6.0243>, rot=<0.0663 -0.9912 -0.0067 0.1146>}
 dyn_hair_r_00 {parent=dyn_hair_r_MSTR, loc=<4.4758 -2.3083 0.3358>, rot=<-0.0102 0.0396 0.1262 -0.9912>}
 dyn_hair_r_01 {parent=dyn_hair_r_00, loc=<7.194 0 0>, rot=<-0.0233 0.0506 0.0526 -0.9971>}
 dyn_hair_r_02 {parent=dyn_hair_r_01, loc=<6.1472 0 0>, rot=<-0.0106 0.0328 -0.171 -0.9847>}
 dyn_hair_r_01a {parent=dyn_hair_r_00, loc=<5.2443 3.8063 -0.5712>, rot=<-0.0233 0.0506 0.0526 -0.9971>}
 hair_side_r_pnt_2 {parent=neck_02, loc=<0.0014 -1.5356 6.1632>, rot=<0.2278 0.66 0.665 0.2652>}
 hair_side_l_pnt_2 {parent=neck_02, loc=<0.0015 -1.5356 -6.1632>, rot=<-0.2278 -0.66 0.665 0.2652>}
 hair_side_r_pnt_3 {parent=neck_01, loc=<1.3652 1.6956 5.9542>, rot=<-0.1384 -0.6843 -0.5855 -0.412>}
 hair_side_l_pnt_3 {parent=neck_01, loc=<1.3649 1.6957 -5.9541>, rot=<-0.1384 -0.6843 0.5855 0.412>}
 clavicle_l {parent=spine_05, loc=<5.4831 -0.1053 -3.2946>, rot=<0.0465 -0.7289 -0.009 -0.6829>}
 upperarm_l {parent=clavicle_l, loc=<9.8658 -1.2136 -0.4105>, rot=<0.0164 -0.3245 -0.0269 -0.9454>}
 upperarm_twist_01_l {parent=upperarm_l, loc=<7.2135 -0.1134 0>, rot=<0 0 0 1>}
 deform_upperarm_bk_base_l {parent=upperarm_twist_01_l, loc=<-6.4053 -4.3746 0.314>, rot=<0 0 0 1>}
 deform_upperarm_bk_l {parent=deform_upperarm_bk_base_l, loc=<0 0 0>, rot=<0 0 0 1>}
 deform_upperarm_fr_base_l {parent=upperarm_twist_01_l, loc=<-5.6312 5.092 0.3354>, rot=<0 0 0 1>}
 deform_upperarm_fr_l {parent=deform_upperarm_fr_base_l, loc=<0 0 0>, rot=<0 0 0 1>}
 deform_upperarm_up_base_l {parent=upperarm_twist_01_l, loc=<-8.891 0.4825 3.9517>, rot=<0 0 0 1>}
 deform_upperarm_up_l {parent=deform_upperarm_up_base_l, loc=<0 0 0>, rot=<0 0 0 1>}
 MSTR_jacket_collar_dyn_l {parent=upperarm_twist_01_l, loc=<-8.5629 -0.0594 4.7083>, rot=<-0.0664 0.0583 0.0072 -0.9961>}
 jacket_collar_dyn_l_1 {parent=MSTR_jacket_collar_dyn_l, loc=<0 0 0>, rot=<0 0 0 1>}
 clavicle_hair_l {parent=clavicle_l, loc=<4.6698 -2.9304 7.5503>, rot=<-0.0372 0.6505 -0.0036 -0.7586>}
 clavicle_l_deform_base {parent=clavicle_l, loc=<7.3262 -1.7643 4.8616>, rot=<0 0 0 1>}
 clavicle_l_deform {parent=clavicle_l_deform_base, loc=<0 0 0>, rot=<0 0 0 1>}
 clavicle_r {parent=spine_05, loc=<5.4831 -0.1053 3.2946>, rot=<0.7289 0.0465 -0.6829 0.009>}
 upperarm_r {parent=clavicle_r, loc=<-9.8658 1.2136 0.4104>, rot=<0.0164 -0.3245 -0.0269 -0.9454>}
 upperarm_twist_01_r {parent=upperarm_r, loc=<-7.2135 0.1134 0>, rot=<0 0 0 1>}
 deform_upperarm_bk_base_r {parent=upperarm_twist_01_r, loc=<6.4052 4.3746 -0.3139>, rot=<0 0 0 1>}
 deform_upperarm_bk_r {parent=deform_upperarm_bk_base_r, loc=<0 0 0>, rot=<0 0 0 1>}
 deform_upperarm_fr_base_r {parent=upperarm_twist_01_r, loc=<5.6311 -5.092 -0.3353>, rot=<0 0 0 1>}
 deform_upperarm_fr_r {parent=deform_upperarm_fr_base_r, loc=<0 0 0>, rot=<0 0 0 1>}
 deform_upperarm_up_base_r {parent=upperarm_twist_01_r, loc=<8.8909 -0.4825 -3.9516>, rot=<0 0 0 1>}
 deform_upperarm_up_r {parent=deform_upperarm_up_base_r, loc=<0 0 0>, rot=<0 0 0 1>}
 MSTR_jacket_collar_dyn_r {parent=upperarm_twist_01_r, loc=<8.5628 0.0594 -4.7083>, rot=<-0.0072 -0.9961 -0.0664 -0.0583>}
 jacket_collar_dyn_r_1 {parent=MSTR_jacket_collar_dyn_r, loc=<0 0 0>, rot=<0 0 0 1>}
 clavicle_hair_r {parent=clavicle_r, loc=<-4.6699 2.9304 -7.5497>, rot=<-0.6505 -0.0372 0.7586 -0.0036>}
 clavicle_r_deform_base {parent=clavicle_r, loc=<-7.3262 1.7643 -4.8614>, rot=<0 -1 0 0>}
 clavicle_r_deform {parent=clavicle_r_deform_base, loc=<0 0 0>, rot=<0 0 0 1>}
 attach_backpack {parent=spine_05, loc=<-0.123 -13.473 0>, rot=<-0.0093 -0.707 0.0093 -0.707>}
 attach_cape {parent=spine_05, loc=<9.017 -11.714 0>, rot=<0.0616 -0.7044 -0.0616 -0.7044>}
 hair_pnt_1 {parent=spine_05, loc=<14.9036 -8.1708 0>, rot=<0.152 -0.6906 0.6906 -0.152>}
 hair_pnt_2 {parent=spine_05, loc=<10.8292 -10.0559 0>, rot=<0.0359 -0.7062 0.7062 -0.0359>}
 hair_pnt_3 {parent=spine_05, loc=<7.2455 -10.4212 0>, rot=<0.0359 -0.7062 0.7062 -0.0359>}
 hair_pnt_4 {parent=spine_05, loc=<-4.8939 -11.6589 0>, rot=<0.0359 -0.7062 0.7062 -0.0359>}
 hair_aim_up {parent=spine_05, loc=<0 -25 0>, rot=<0 0 0.024 0.9997>}
 hair_aim_for {parent=spine_05, loc=<-19.5986 -22.7877 0>, rot=<0.1407 -0.693 0.693 -0.1407>}
 clavicle_l_pnt_back {parent=spine_05, loc=<9.06 -5.94 -1.76>, rot=<-0.0465 0.7289 0.009 0.6829>}
 collar_for_r_deform_base {parent=spine_05, loc=<3.8401 6.8097 5.5514>, rot=<0.2548 0.1946 -0.3878 -0.8642>}
 collar_for_r_deform {parent=collar_for_r_deform_base, loc=<0 0 0>, rot=<0 0 0 1>}
 collar_for_l_deform_base {parent=spine_05, loc=<3.8397 6.8096 -5.5514>, rot=<-0.2548 -0.1946 -0.3878 -0.8642>}
 collar_for_l_deform {parent=collar_for_l_deform_base, loc=<0 0 0>, rot=<0 0 0 -1>}
 hair_side_r_pnt_4 {parent=spine_05, loc=<7.9182 4.0307 4.0387>, rot=<0.0014 -0.6982 -0.6562 -0.2864>}
 hair_side_l_pnt_4 {parent=spine_05, loc=<7.7004 4.128 -3.9342>, rot=<0.0014 -0.6982 0.6562 0.2864>}
]
variables [
 ShouldersUp {type=bool, default=True}
 clavicle_r_SPR_back_alpha {type=double, default=0}
 crouch {type=double, default=0}
 crouch move {type=double, default=0}
 head_SPR_look_down_alpha {type=double, default=0}
 head_SPR_look_left_alpha {type=double, default=0}
 head_SPR_look_right_alpha {type=double, default=0}
 head_SPR_look_up_alpha {type=double, default=0}
 look_down_upperarm_r_back_multiplied_alpha {type=double, default=0}
 look_up_and_left_two_hands_multiplied_alpha {type=double, default=0}
 look_up_two_hands_rifle_multiplied_alpha {type=double, default=0}
 uppearm_r_SPR_aim_up_two_hands_alpha {type=double, default=0}
 upperarm_r_SPR_back_alpha {type=double, default=0}
 upperarm_r_rifle_up_alpha {type=double, default=0}
 upperarm_r_skydiving {type=double, default=0}
]
