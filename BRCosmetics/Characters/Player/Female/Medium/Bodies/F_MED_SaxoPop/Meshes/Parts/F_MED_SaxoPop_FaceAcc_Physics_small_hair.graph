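# Physics asset for the small hair strands on the SaxoPop female medium body's face accessory mesh, giving them constrained secondary motion.
# Small capsules on MSTR_hair A–J bones root larger capsules on the strand bones.
# Constraints form one- or two-segment chains, with MSTR_hair_H_1 branching into the H, K and L strands.
# All joints share a tight twist limit, while the F, G and H-rooted strands get wider swing than the rest.
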PhysicsAsset /BRCosmetics/Characters/Player/Female/Medium/Bodies/F_MED_SaxoPop/Meshes/Parts/F_MED_SaxoPop_FaceAcc_Physics_small_hair
bodies {
 MSTR_hair_A [{shapeType=Capsule, length=1, radius=0.5, shapeName=MSTR_hair_A_capsule}]
 MSTR_hair_B [{shapeType=Capsule, length=1, radius=0.5, shapeName=MSTR_hair_B_capsule}]
 MSTR_hair_C [{shapeType=Capsule, length=1, radius=0.5, shapeName=MSTR_hair_C_capsule}]
 MSTR_hair_D [{shapeType=Capsule, length=1, radius=0.5, shapeName=MSTR_hair_D_capsule}]
 MSTR_hair_E [{shapeType=Capsule, length=1, radius=0.5, shapeName=MSTR_hair_E_capsule}]
 MSTR_hair_F [{shapeType=Capsule, length=1, radius=0.5, shapeName=MSTR_hair_F_capsule}]
 MSTR_hair_G [{shapeType=Capsule, length=1, radius=0.5, shapeName=MSTR_hair_G_capsule}]
 MSTR_hair_H_1 [{shapeType=Capsule, length=1, radius=0.5, shapeName=MSTR_hair_H_1_capsule}]
 MSTR_hair_I [{shapeType=Capsule, length=1, radius=0.5, shapeName=MSTR_hair_I_capsule}]
 MSTR_hair_J [{shapeType=Capsule, length=1, radius=0.5, shapeName=MSTR_hair_J_capsule}]
 hair_A_1 [{shapeType=Capsule, length=1, radius=1, shapeName=hair_A_1_capsule, center=<1.8063 0 0>, rotation=<rot 90 0 0>}]
 hair_B_1 [{shapeType=Capsule, length=1, radius=1, shapeName=hair_B_1_capsule, center=<1.8063 0 0>, rotation=<rot 90 0 0>}]
 hair_C_1 [{shapeType=Capsule, length=1, radius=1, shapeName=hair_C_1_capsule, center=<1.8063 0 0>, rotation=<rot 90 0 0>}]
 hair_D_1 [{shapeType=Capsule, length=1, radius=1, shapeName=hair_D_1_capsule, center=<1.8063 0 0>, rotation=<rot 90 0 0>}]
 hair_E_1 [{shapeType=Capsule, length=1, radius=1, shapeName=hair_E_1_capsule, center=<1.8063 0 0>, rotation=<rot 90 0 0>}]
 hair_E_2 [{shapeType=Capsule, length=1, radius=1, shapeName=hair_E_2_capsule, center=<1.8063 0 0>, rotation=<rot 90 0 0>}]
 hair_F_1 [{shapeType=Capsule, length=1, radius=1, shapeName=hair_F_1_capsule, center=<1.8063 0 0>, rotation=<rot 90 0 0>}]
 hair_F_2 [{shapeType=Capsule, length=1, radius=1, shapeName=hair_F_2_capsule, center=<1.8063 0 0>, rotation=<rot 90 0 0>}]
 hair_G_1 [{shapeType=Capsule, length=1, radius=1, shapeName=hair_G_1_capsule, center=<1.8063 0 0>, rotation=<rot 90 0 0>}]
 hair_G_2 [{shapeType=Capsule, length=1, radius=1, shapeName=hair_G_2_capsule, center=<1.8063 0 0>, rotation=<rot 90 0 0>}]
 hair_H_2 [{shapeType=Capsule, length=1, radius=1, shapeName=hair_H_2_capsule, center=<1.8063 0 0>, rotation=<rot 90 0 0>}]
 hair_H_3 [{shapeType=Capsule, length=1, radius=1, shapeName=hair_H_3_capsule, center=<1.8063 0 0>, rotation=<rot 90 0 0>}]
 hair_I_1 [{shapeType=Capsule, length=1, radius=1, shapeName=hair_I_1_capsule, center=<1.8063 0 0>, rotation=<rot 90 0 0>}]
 hair_J_1 [{shapeType=Capsule, length=1, radius=1, shapeName=hair_J_1_capsule, center=<1.8063 0 0>, rotation=<rot 90 0 0>}]
 hair_K_1 [{shapeType=Capsule, length=1, radius=1, shapeName=hair_K_1_capsule, center=<1.5317 0 0>, rotation=<rot 90 0 0>}]
 hair_L_1 [{shapeType=Capsule, length=1, radius=1, shapeName=hair_L_1_capsule, center=<1.3553 0 0>, rotation=<rot 90 0 0>}]
}
constraints [
 {bone1Name=hair_A_1, bone2Name=MSTR_hair_A, swing1LimitDegrees=20, swing2LimitDegrees=20, twistLimitDegrees=5}
 {bone1Name=hair_B_1, bone2Name=MSTR_hair_B, swing1LimitDegrees=20, swing2LimitDegrees=20, twistLimitDegrees=5}
 {bone1Name=hair_C_1, bone2Name=MSTR_hair_C, swing1LimitDegrees=20, swing2LimitDegrees=20, twistLimitDegrees=5}
 {bone1Name=hair_D_1, bone2Name=MSTR_hair_D, swing1LimitDegrees=20, swing2LimitDegrees=20, twistLimitDegrees=5}
 {bone1Name=hair_E_1, bone2Name=MSTR_hair_E, swing1LimitDegrees=20, swing2LimitDegrees=20, twistLimitDegrees=5}
 {bone1Name=hair_E_2, bone2Name=hair_E_1, swing1LimitDegrees=20, swing2LimitDegrees=20, twistLimitDegrees=5}
 {bone1Name=hair_F_1, bone2Name=MSTR_hair_F, swing1LimitDegrees=40, swing2LimitDegrees=40, twistLimitDegrees=5}
 {bone1Name=hair_F_2, bone2Name=hair_F_1, swing1LimitDegrees=40, swing2LimitDegrees=40, twistLimitDegrees=5}
 {bone1Name=hair_G_1, bone2Name=MSTR_hair_G, swing1LimitDegrees=40, swing2LimitDegrees=40, twistLimitDegrees=5}
 {bone1Name=hair_G_2, bone2Name=hair_G_1, swing1LimitDegrees=40, swing2LimitDegrees=40, twistLimitDegrees=5}
 {bone1Name=hair_H_2, bone2Name=MSTR_hair_H_1, swing1LimitDegrees=40, swing2LimitDegrees=40, twistLimitDegrees=5}
 {bone1Name=hair_H_3, bone2Name=hair_H_2, swing1LimitDegrees=40, swing2LimitDegrees=40, twistLimitDegrees=5}
 {bone1Name=hair_K_1, bone2Name=MSTR_hair_H_1, swing1LimitDegrees=40, swing2LimitDegrees=40, twistLimitDegrees=5}
 {bone1Name=hair_L_1, bone2Name=MSTR_hair_H_1, swing1LimitDegrees=40, swing2LimitDegrees=40, twistLimitDegrees=5}
 {bone1Name=hair_I_1, bone2Name=MSTR_hair_I, swing1LimitDegrees=20, swing2LimitDegrees=20, twistLimitDegrees=5}
 {bone1Name=hair_J_1, bone2Name=MSTR_hair_J, swing1LimitDegrees=20, swing2LimitDegrees=20, twistLimitDegrees=5}
]
