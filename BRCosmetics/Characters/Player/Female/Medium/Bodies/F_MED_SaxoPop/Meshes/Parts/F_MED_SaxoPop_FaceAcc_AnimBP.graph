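# Anim Blueprint for the SaxoPop female FaceAcc part, simulating its hair on the copied body pose.
# A Control Rig reacts to crouching and two shoulders-up emotes, then back-hair RigidBody physics follows state-driven gravity.
# Capes or long back blings switch to a tighter fixed-gravity simulation with a front-end idle bone offset.
# A small-hair RigidBody pass follows, and skydiving blends in a squashed MSTR_hair_down bone.
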AnimBlueprint /BRCosmetics/Characters/Player/Female/Medium/Bodies/F_MED_SaxoPop/Meshes/Parts/F_MED_SaxoPop_FaceAcc_AnimBP
target_skeleton=&/F_MED_SaxoPop_FaceAcc_Skeleton
graphs [
 AnimGraph {nodes [
   n1 = "Transform (Modify) Bone - Bone: MSTR_hair_down"!(
    ComponentPose=Local To Component(LocalPose=Use cached pose 'main'())
    bAlphaBoolEnabled:bool=True
    Translation="0,0,0"
    Rotation="0,0,0"
    Scale="1,0.75,1"
   ) -> (Pose) {alphaInputType=Bool, scaleMode=BMM_Additive, alphaBoolBlend {blendOption=Linear}, boneToModify {boneName=MSTR_hair_down}, scale {y=0.75}}
   Output Pose(Result=Blend Poses by bool(
     BlendPose_0=Component To Local(ComponentPose=@n1)
     BlendPose_1=Use cached pose 'main'()
     BlendTime_0=0.3
     BlendTime_1=0.3
     bActiveValue=Get bIsSkyDiving()
    ) {blendType=Linear}) {layerGroup=None}
   n2 = Control Rig!(
    Source=Copy Pose From Mesh() {bUseAttachedParent=true}
    Alpha
    crouch move:double=Get crouchmove() -> (crouchmove:double)
    crouch:double=Get crouch() -> (crouch:double)
    ShouldersUp:bool=Get IsShouldersUpEmote() -> (IsShouldersUpEmote:bool)
   ) -> (Pose) {
    alphaBoolBlend {blendOption=Linear}
    controlRigAssetReference {blueprintRigClass=&/F_MED_SaxoPop_FaceAcc_CtrlRig_Backpack.F_MED_SaxoPop_FaceAcc_CtrlRig_Backpack_C}
   }
   Save cached pose 'ControlRig'(Pose=@n2) {cacheName=ControlRig}
   n3 = RigidBody!(
    ComponentPose=Local To Component(LocalPose=Use cached pose 'ControlRig'())
    Alpha
    OverrideWorldGravity:Vector=Get Gravity Override()
    ExternalForce
   ) -> (Pose) {
    bOverrideWorldGravity=true
    overridePhysicsAsset=&/F_MED_SaxoPop_FaceAcc_BackHair_Physics
    simulationSpace=BaseBoneSpace
    alphaBoolBlend {blendOption=Linear}
    baseBoneRef {boneName=neck_01}
    componentAppliedLinearAccClamp=<0 0 0>
    simSpaceSettings {
     dampingAlpha=0.25
     maxAngularAcceleration=80
     maxAngularVelocity=20
     maxLinearAcceleration=1000
     maxLinearVelocity=500
     worldAlpha=1
     externalLinearDragV=<1 1 1>
    }
   }
   n4 = Component To Local(ComponentPose=@n3)
   n5 = "Transform (Modify) Bone - Bone: dyn_hair_l_MSTR"!(
    ComponentPose=Local To Component(LocalPose=Use cached pose 'ControlRig'())
    bAlphaBoolEnabled:bool=Get in frontend idle() -> (in frontend idle:bool)
    Translation="0,0,0"
    Rotation="18.598,23.016,2.376"
    Scale="1,1,1"
   ) -> (Pose) {
    alphaInputType=Bool
    rotationMode=BMM_Additive
    rotationSpace=BCS_BoneSpace
    scaleMode=BMM_Replace
    scaleSpace=BCS_BoneSpace
    translationMode=BMM_Additive
    translationSpace=BCS_BoneSpace
    alphaBoolBlend {blendOption=Linear}
    boneToModify {boneName=dyn_hair_l_MSTR}
    rotation=<rot 18.598 23.016 2.376>
   }
   n6 = RigidBody(ComponentPose=@n5, OverrideWorldGravity="0,0,-980") {
    bOverrideWorldGravity=true
    overridePhysicsAsset=&F_MED_SaxoPop_FaceAcc_BackHair_Physics
    simulationSpace=BaseBoneSpace
    alphaBoolBlend {blendOption=Linear}
    baseBoneRef {boneName=neck_01}
    componentAppliedLinearAccClamp=<0 0 0>
    simSpaceSettings {maxAngularAcceleration=5, maxAngularVelocity=5, maxLinearAcceleration=120, maxLinearVelocity=250, velocityScaleZ=0.5, worldAlpha=0.8}
   }
   n7 = Component To Local(ComponentPose=@n6)
   n8 = Get is_WearingCape() -> (is_WearingCape:bool)
   n9 = RigidBody!(
    ComponentPose=Local To Component(LocalPose=Blend Poses by bool(BlendPose_0=@n7, BlendPose_1=@n4, BlendTime_0=0.3, BlendTime_1=0.3, bActiveValue=@n8) {blendType=Linear})
    Alpha
    OverrideWorldGravity=Get Gravity Override()
    ExternalForce
    SimSpaceSettings:Sim Space Settings Structure=(WorldAlpha=0.8,VelocityScaleZ=0.5,DampingAlpha=1,MaxLinearVelocity=250,MaxAngularVelocity=5,MaxLinearAcceleration=120,MaxAngularAcceleration=5,ExternalLinearDragV=(X=0,Y=0,Z=0),ExternalLinearVelocity=(X=0,Y=0,Z=0),ExternalAngularVelocity=(X=0,Y=0,Z=0))
   ) -> (Pose) {
    bOverrideWorldGravity=true
    overridePhysicsAsset=&/F_MED_SaxoPop_FaceAcc_Physics_small_hair
    simulationSpace=BaseBoneSpace
    alphaBoolBlend {blendOption=Linear}
    baseBoneRef {boneName=neck_01}
    componentAppliedLinearAccClamp=<0 0 0>
    simSpaceSettings {maxAngularAcceleration=5, maxAngularVelocity=5, maxLinearAcceleration=120, maxLinearVelocity=250, velocityScaleZ=0.5, worldAlpha=0.8}
   }
   Save cached pose 'main'(Pose=Component To Local(ComponentPose=@n9)) {cacheName=main}
  ]}
 EventGraph {nodes [
   Event BlueprintUpdateAnimation()
   > n1 = Sequence!(execute) -> (then_0, then_1, then_2:Exec, then_3:Exec, then_4:Exec)
   > n2 = "Gravity Override State Function (Nativized)"!(
    execute
    CharacterPartAnimInstance:Custom Character Part Anim Instance Object Reference=Self-Reference()
    GravityOverrideParameterSet:Fort Gravity Override Parameter Set Structure=Get Gravity Override Params() -> (Gravity Override Params:Fort Gravity Override Parameter Set Structure)
    RigidBodyAnimNodeParameterSet:Fort Rigid Body Anim Node Parameter Set Structure=Get RBAN Params() -> (RBAN Params:Fort Rigid Body Anim Node Parameter Set Structure)
    DebugDraw:bool=false
    GameTimeInSeconds:float=GetGameTimeInSeconds()
    DebugJointName:Name=root
    StateLogicLayerName:Name=none
   ) -> (
    then
    OutGravityOverride:Vector
    OutCurrentState:Name=None
    OutGravityOverrideParameters:Fort Gravity Override Parameters Structure
    OutRigidBodyAnimNodeParameters:Fort Rigid Body Anim Node Parameters Structure
   )
   > Set Gravity Override(Gravity Override=@n2)
   > Set RBAN Settings Out(RBAN Settings Out=@n2.OutRigidBodyAnimNodeParameters)
   n3 = Get bIsCrouching()
   n4 = Branch(execute=@n1.then_1, Condition=AND Boolean(A=@n3, B=Get bIsMoving2D()))
   > Set crouchmove(crouchmove:double=1) -> (Output_Get:double)
   Set crouchmove(execute=@n4.else, crouchmove=0)
   Set crouch(execute=@n1.then_2, crouch:double="To Float (Boolean)"(InBool=@n3)) -> (Output_Get:double)
   n5 = Get bIsPlayingEmote()
   n6 = GetDeltaTransformFromRefPose!(self=GetOwnerSkeletalMeshComponent(), BoneName=upperarm_l, BaseName=spine_05) -> (
    ReturnValue_Location:Vector
    ReturnValue_Rotation_Roll:float
    ReturnValue_Rotation_Pitch:float
    ReturnValue_Rotation_Yaw:float
    ReturnValue_Scale:Vector
   )
   > Set in frontend idle(in frontend idle:bool=AND Boolean!(A=NOT Boolean(A=@n5), B=Get bIsInFrontEnd(), C:bool="float > float"(A=@n6.ReturnValue_Rotation_Yaw, B=-20)) -> (ReturnValue)) -> (Output_Get:bool)
   n7 = Get GetAnimInstance() -> (GetAnimInstance:Anim Instance Object Reference)
   n8 = Branch(
    execute=@n1.then_3
    Condition=OR Boolean(
     A=Montage_IsPlaying(self=@n7, Montage=/BRCosmetics/Animation/Game/MainPlayer/Emotes/Emote_Celebration_Female_M)
     B=Montage_IsPlaying(self=@n7, Montage=/BRCosmetics/Animation/Game/MainPlayer/Emotes/Mime/Emote_Mime_CMF_M)
    )
   )
   > Set IsShouldersUpEmote(IsShouldersUpEmote:bool=1) -> (Output_Get:bool)
   Set IsShouldersUpEmote(execute=@n8.else, IsShouldersUpEmote=false)
   Event OnCharacterCustomizationLoaded()
   > Set GetAnimInstance(GetAnimInstance:Anim Instance Object Reference=GetAnimInstance(self=GetOwnerSkeletalMeshComponent())) -> (Output_Get:Anim Instance Object Reference)
   n9 = CheckForBackbling()
   > n10 = Branch(Condition=OR Boolean!(A=@n9, B=@n9.bIsBackblingACape, C:bool=@n9.bIsBackblingLong) -> (ReturnValue))
   > Set is_WearingCape(is_WearingCape:bool=true) -> (Output_Get:bool)
   Set is_WearingCape(execute=@n10.else, is_WearingCape=false)
  ]}
]
variables [
 crouch {type=number}
 crouchmove {type=number}
 GetAnimInstance {type=AnimInstance, default=null}
 Gravity Override {type=Vector}
 Gravity Override Params {
  type=FortGravityOverrideParameterSet
  default=(Skydive_GoingUp=(GravityOverride=(WindFrequency=(X=1.500000,Y=7.000000,Z=8.000000),WindAmplitude=(X=200.000000,Y=-500.000000,Z=400.000000),GravityOverride=(X=0.000000,Y=-300.000000,Z=1500.000000),JointName="root")),Skydive_FloatingDown=(GravityOverride=(WindFrequency=(X=1.500000,Y=7.000000,Z=8.000000),WindAmplitude=(X=200.000000,Y=-500.000000,Z=400.000000),WindOffset=(X=0.000000,Y=-600.000000,Z=400.000000),GravityOverride=(X=0.000000,Y=-1000.000000,Z=-200.000000),JointName="root")),Skydive_DivingDown=(GravityOverride=(WindFrequency=(X=1.500000,Y=12.000000,Z=10.000000),WindAmplitude=(X=50.000000,Y=300.000000,Z=100.000000),GravityOverride=(X=0.000000,Y=-50.000000,Z=-980.000000),JointName="pelvis")),Skydive_Parachuting=(GravityOverride=(WindFrequency=(X=1.500000,Y=7.000000,Z=8.000000),WindAmplitude=(X=200.000000,Y=-500.000000,Z=400.000000),GravityOverride=(X=0.000000,Y=-800.000000,Z=-980.000000),JointName="root")),OnGround_Standing=(GravityOverride=(GravityOverride=(X=0.000000,Y=0.000000,Z=-980.000000),JointName="root")),OnGround_Moving=(GravityOverride=(GravityOverride=(X=0.000000,Y=-500.000000,Z=-980.000000),JointName="root")),OnGround_Moving_HighRotationRate=(GravityOverride=(GravityOverride=(X=0.000000,Y=0.000000,Z=-600.000000),JointName="root")),OnGround_Crouch_Moving=(GravityOverride=(GravityOverride=(X=0.000000,Y=0.000000,Z=-980.000000),JointName="root")),Emote_or_Melee=(GravityOverride=(GravityOverride=(X=0.000000,Y=0.000000,Z=-980.000000),JointName="root")),Swimming=(GravityOverride=(GravityOverride=(X=-800.000000,Y=0.000000,Z=0.000000),JointName="pelvis")),Front_End=(GravityOverride=(GravityOverride=(X=0.000000,Y=0.000000,Z=-980.000000),JointName="root")),DBNO=(GravityOverride=(GravityOverride=(X=-980.000000,Y=400.000000,Z=0.000000),JointName="pelvis")),DBNO_Carried=(GravityOverride=(GravityOverride=(X=-980.000000,Y=500.000000,Z=-500.000000),JointName="pelvis")),OnGround_Falling=(GravityOverride=(GravityOverride=(X=0.000000,Y=0.000000,Z=-980.000000),JointName="root")),Ziplining=(GravityOverride=(GravityOverride=(X=0.000000,Y=0.000000,Z=-980.000000),JointName="root")),Skydive_Surfing=(GravityOverride=(WindFrequency=(X=1.500000,Y=7.000000,Z=8.000000),WindAmplitude=(X=200.000000,Y=-500.000000,Z=400.000000),GravityOverride=(X=500.000000,Y=-300.000000,Z=-980.000000),JointName="root")),Riding_Standing=(GravityOverride=(GravityOverride=(X=0.000000,Y=-350.000000,Z=-980.000000),JointName="root")),Riding_Moving=(GravityOverride=(GravityOverride=(X=0.000000,Y=-450.000000,Z=-450.000000),JointName="root")),Riding_Moving_Targeting=(GravityOverride=(GravityOverride=(X=0.000000,Y=-200.000000,Z=-980.000000),JointName="root")),Tactical_Sprint=(GravityOverride=(GravityOverride=(X=0.000000,Y=-850.000000,Z=-800.000000),JointName="root")),Rail_Grinding=(GravityOverride=(GravityOverride=(X=0.000000,Y=0.000000,Z=-980.000000),JointName="root")),)
 }
 in frontend idle {type=boolean}
 is_WearingCape {type=boolean}
 IsShouldersUpEmote {type=boolean}
 RBAN Params {
  type=FortRigidBodyAnimNodeParameterSet
  default=(Skydive_GoingUp=(RigidBodyAnimNode=(SimSpaceSettings=(WorldAlpha=1.000000,MaxLinearVelocity=500.000000,MaxAngularVelocity=3.000000,MaxLinearAcceleration=1000.000000,MaxAngularAcceleration=1.000000,ExternalLinearDragV=(X=1.000000,Y=1.000000,Z=1.000000)))),Skydive_FloatingDown=(RigidBodyAnimNode=(ComponentLinearAccScale=(X=0.000000,Y=0.500000,Z=0.000000),ComponentLinearVelScale=(X=0.000000,Y=0.500000,Z=0.000000),SimSpaceSettings=(WorldAlpha=1.000000,MaxLinearVelocity=500.000000,MaxAngularVelocity=3.000000,MaxLinearAcceleration=1000.000000,MaxAngularAcceleration=1.000000,ExternalLinearDragV=(X=1.000000,Y=1.000000,Z=1.000000)))),Skydive_DivingDown=(RigidBodyAnimNode=(ComponentLinearAccScale=(X=0.100000,Y=0.000000,Z=0.000000),ComponentLinearVelScale=(X=0.100000,Y=0.000000,Z=0.000000),SimSpaceSettings=(WorldAlpha=1.000000,MaxLinearVelocity=500.000000,MaxAngularVelocity=3.000000,MaxLinearAcceleration=1000.000000,MaxAngularAcceleration=1.000000,ExternalLinearDragV=(X=1.000000,Y=1.000000,Z=1.000000)))),Skydive_Parachuting=(RigidBodyAnimNode=(ComponentLinearAccScale=(X=0.000000,Y=0.100000,Z=0.300000),ComponentLinearVelScale=(X=0.000000,Y=0.100000,Z=0.300000),SimSpaceSettings=(WorldAlpha=1.000000,MaxLinearVelocity=500.000000,MaxAngularVelocity=3.000000,MaxLinearAcceleration=1000.000000,MaxAngularAcceleration=1.000000,ExternalLinearDragV=(X=1.000000,Y=1.000000,Z=1.000000)))),OnGround_Standing=(RigidBodyAnimNode=(ComponentLinearAccScale=(X=0.300000,Y=0.000000,Z=0.000000),ComponentLinearVelScale=(X=0.300000,Y=0.000000,Z=0.000000),SimSpaceSettings=(WorldAlpha=1.000000,DampingAlpha=0.250000,MaxLinearVelocity=500.000000,MaxAngularVelocity=20.000000,MaxLinearAcceleration=1000.000000,MaxAngularAcceleration=80.000000,ExternalLinearDragV=(X=1.000000,Y=1.000000,Z=1.000000)))),OnGround_Moving=(RigidBodyAnimNode=(ComponentLinearAccScale=(X=0.300000,Y=0.000000,Z=0.000000),ComponentLinearVelScale=(X=0.300000,Y=0.000000,Z=0.000000),SimSpaceSettings=(WorldAlpha=1.000000,DampingAlpha=0.250000,MaxLinearVelocity=500.000000,MaxAngularVelocity=20.000000,MaxLinearAcceleration=1000.000000,MaxAngularAcceleration=80.000000,ExternalLinearDragV=(X=1.000000,Y=1.000000,Z=1.000000)))),OnGround_Moving_HighRotationRate=(RigidBodyAnimNode=(ComponentLinearAccScale=(X=0.300000,Y=0.300000,Z=0.000000),ComponentLinearVelScale=(X=0.300000,Y=0.300000,Z=0.000000),SimSpaceSettings=(WorldAlpha=1.000000,DampingAlpha=0.500000,MaxLinearVelocity=500.000000,MaxAngularVelocity=3.000000,MaxLinearAcceleration=1000.000000,MaxAngularAcceleration=1.000000,ExternalLinearDragV=(X=1.000000,Y=1.000000,Z=1.000000)))),OnGround_Crouch_Moving=(RigidBodyAnimNode=(SimSpaceSettings=(WorldAlpha=1.000000,MaxLinearVelocity=500.000000,MaxAngularVelocity=3.000000,MaxLinearAcceleration=1000.000000,MaxAngularAcceleration=1.000000,ExternalLinearDragV=(X=1.000000,Y=1.000000,Z=1.000000)))),Emote_or_Melee=(RigidBodyAnimNode=(SimSpaceSettings=(WorldAlpha=1.000000,DampingAlpha=0.250000,MaxLinearVelocity=500.000000,MaxAngularVelocity=3.000000,MaxLinearAcceleration=1000.000000,MaxAngularAcceleration=1.000000,ExternalLinearDragV=(X=1.000000,Y=1.000000,Z=1.000000)))),Swimming=(RigidBodyAnimNode=(SimSpaceSettings=(WorldAlpha=1.000000,MaxLinearVelocity=500.000000,MaxAngularVelocity=3.000000,MaxLinearAcceleration=1000.000000,MaxAngularAcceleration=1.000000,ExternalLinearDragV=(X=1.000000,Y=1.000000,Z=1.000000)))),Front_End=(RigidBodyAnimNode=(SimSpaceSettings=(WorldAlpha=1.000000,MaxLinearVelocity=500.000000,MaxAngularVelocity=3.000000,MaxLinearAcceleration=1000.000000,MaxAngularAcceleration=1.000000,ExternalLinearDragV=(X=1.000000,Y=1.000000,Z=1.000000)))),DBNO=(RigidBodyAnimNode=(SimSpaceSettings=(WorldAlpha=1.000000,MaxLinearVelocity=500.000000,MaxAngularVelocity=3.000000,MaxLinearAcceleration=1000.000000,MaxAngularAcceleration=1.000000,ExternalLinearDragV=(X=1.000000,Y=1.000000,Z=1.000000)))),DBNO_Carried=(RigidBodyAnimNode=(SimSpaceSettings=(WorldAlpha=1.000000,MaxLinearVelocity=500.000000,MaxAngularVelocity=3.000000,MaxLinearAcceleration=1000.000000,MaxAngularAcceleration=1.000000,ExternalLinearDragV=(X=1.000000,Y=1.000000,Z=1.000000)))),OnGround_Falling=(RigidBodyAnimNode=(SimSpaceSettings=(WorldAlpha=1.000000,DampingAlpha=0.500000,MaxLinearVelocity=500.000000,MaxAngularVelocity=20.000000,MaxLinearAcceleration=1000.000000,MaxAngularAcceleration=80.000000,ExternalLinearDragV=(X=1.000000,Y=1.000000,Z=1.000000)))),Ziplining=(RigidBodyAnimNode=(SimSpaceSettings=(WorldAlpha=1.000000,MaxLinearVelocity=500.000000,MaxAngularVelocity=3.000000,MaxLinearAcceleration=1000.000000,MaxAngularAcceleration=1.000000,ExternalLinearDragV=(X=1.000000,Y=1.000000,Z=1.000000)))),Skydive_Surfing=(RigidBodyAnimNode=(SimSpaceSettings=(WorldAlpha=1.000000,MaxLinearVelocity=500.000000,MaxAngularVelocity=3.000000,MaxLinearAcceleration=1000.000000,MaxAngularAcceleration=1.000000,ExternalLinearDragV=(X=1.000000,Y=1.000000,Z=1.000000)))),Riding_Standing=(RigidBodyAnimNode=(SimSpaceSettings=(WorldAlpha=1.000000,MaxLinearVelocity=500.000000,MaxAngularVelocity=3.000000,MaxLinearAcceleration=1000.000000,MaxAngularAcceleration=1.000000,ExternalLinearDragV=(X=1.000000,Y=1.000000,Z=1.000000)))),Riding_Moving=(RigidBodyAnimNode=(SimSpaceSettings=(WorldAlpha=1.000000,MaxLinearVelocity=500.000000,MaxAngularVelocity=3.000000,MaxLinearAcceleration=1000.000000,MaxAngularAcceleration=1.000000,ExternalLinearDragV=(X=1.000000,Y=1.000000,Z=1.000000)))),Riding_Moving_Targeting=(RigidBodyAnimNode=(SimSpaceSettings=(WorldAlpha=1.000000,MaxLinearVelocity=500.000000,MaxAngularVelocity=3.000000,MaxLinearAcceleration=1000.000000,MaxAngularAcceleration=1.000000,ExternalLinearDragV=(X=1.000000,Y=1.000000,Z=1.000000)))),Tactical_Sprint=(RigidBodyAnimNode=(ComponentLinearAccScale=(X=0.300000,Y=0.000000,Z=0.000000),ComponentLinearVelScale=(X=0.300000,Y=0.000000,Z=0.000000),SimSpaceSettings=(WorldAlpha=1.000000,MaxLinearVelocity=500.000000,MaxAngularVelocity=3.000000,MaxLinearAcceleration=1000.000000,MaxAngularAcceleration=1.000000,ExternalLinearDragV=(X=1.000000,Y=1.000000,Z=1.000000)))),Rail_Grinding=(RigidBodyAnimNode=(SimSpaceSettings=(WorldAlpha=1.000000,MaxLinearVelocity=500.000000,MaxAngularVelocity=3.000000,MaxLinearAcceleration=1000.000000,MaxAngularAcceleration=1.000000,ExternalLinearDragV=(X=1.000000,Y=1.000000,Z=1.000000)))),)
 }
 RBAN Settings Out {type=FortRigidBodyAnimNodeParameters}
]
